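# Material function for the GreenHazard character body that draws three colored layers of animated bubbles with fake depth, outputting color and a combined mask.
# Each layer pushes the pre-skinned position along the camera ray for parallax, then triplanar-samples a shared bubble texture with panning, so the pattern follows the skinned mesh.
# A time-driven sine adds per-layer wobble.
# Colors composite over black, nearer layers occluding deeper ones, and the mask takes their maximum.
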MaterialFunction /BRCosmetics/Characters/Player/Male/Medium/Bodies/M_MED_GreenHazard/Materials/MF_DepthBubbles
graph {nodes [
  n1 = LinearInterpolate(
   B=VectorParameter() {group=Bubbles, parameterName=Layer1_Color, sortPriority=9, defaultValue=<rgba 0.44 0.8 0.08 1>}.RGB
   Alpha=NamedRerouteUsage()
  )
  n2 = VectorParameter() {group=Bubbles, parameterName=Layer2_Color, sortPriority=19, defaultValue=<rgba 0.12 0.3 0.03 1>}
  n3 = Multiply(A=NamedRerouteUsage(), B=OneMinus(None=NamedRerouteUsage()))
  n4 = VectorParameter() {group=Bubbles, parameterName=Layer3_Color, sortPriority=29, defaultValue=<rgba 0.055 0.15 0.0075 1>}
  n5 = NamedRerouteUsage()
  n6 = OneMinus(None=NamedRerouteUsage())
  FunctionOutput(None=LinearInterpolate(A=LinearInterpolate(A=@n1, B=@n2.RGB, Alpha=@n3), B=@n4.RGB, Alpha=Multiply(A=@n5, B=@n6)))
  n7 = Max(A=NamedRerouteUsage(), B=NamedRerouteUsage())
  n8 = NamedRerouteUsage()
  FunctionOutput(None=NamedRerouteDeclaration(None=Max(A=@n7, B=@n8)) {name=Bubbles_Mask, nodeColor=<rgba 0.7344 1 0 1>}) {outputName=Mask, sortPriority=10}
  n9 = TextureObjectParameter() {
   group=Bubbles
   parameterName=BubbleMask
   samplerType=SAMPLERTYPE_Masks
   sortPriority=-1
   texture=/Game/Characters/Player/Male/Large/Bodies/M_LRG_Slurp_Monster/Textures/T_SlurpBubble
  }
  NamedRerouteDeclaration(None=@n9) {name=BubblesTex, nodeColor=<rgba 1 0.8672 0 1>}
  n10 = NamedRerouteUsage()
  n11 = CameraPositionWS()
  n12 = Normalize(VectorInput=Subtract(A=@n10, B=TransformPosition(None=@n11) {transformSourceType=TRANSFORMPOSSOURCE_World}))
  n13 = ScalarParameter() {defaultValue=10, group=Bubbles, parameterName=BubblesParallaxDepth_Layer1, sortPriority=10}
  n14 = VectorParameter() {group=Bubbles, parameterName=BubblesPositionOffset_Layer1, sortPriority=10, defaultValue=<rgba 0 0 0 1>}
  n15 = Add(A=Add(A=@n10, B=Multiply(A=@n12, B=@n13)), B=@n14.RGB)
  n16 = ScalarParameter() {group=Bubbles, parameterName=BubblesSpeedX_Layer1, sortPriority=10}
  n17 = ScalarParameter() {defaultValue=0.05, group=Bubbles, parameterName=BubblesSpeedY_Layer1, sortPriority=10}
  n18 = MaterialFunctionCall!(
   TextureObject=NamedRerouteUsage()
   TextureSize=Multiply(A=NamedRerouteUsage(), B=ScalarParameter() {defaultValue=10, group=Bubbles, parameterName=BubblesSize_Layer1, sortPriority=10})
   WorldPosition=Add(A=NamedRerouteDeclaration(None=@n15) {name=BubbleWorldPosParallax, nodeColor=<rgba 0 1 0.3203 1>}, B=NamedRerouteUsage())
   Export Float 4
   In=Multiply(A=NamedRerouteUsage(), B=AppendVector(A=@n16, B=@n17))
   World Space Normal=NamedRerouteUsage()
   ProjectionTransitionContrast=NamedRerouteUsage()
  ) {materialFunction=/Game/Athena/SafeZone/Preskinned_WP_Texture_ScalePan}
  n19 = MaterialFunctionCall(Float3=@n18.XY Texture) {materialFunction=&BreakOutFloat3Components}
  n20 = MaterialFunctionCall(f3=@n18.XY Texture) {materialFunction=&AddComponents}
  n21 = ScalarParameter() {group=Bubbles, parameterName=BubblesAmount_Layer1, sortPriority=10}
  NamedRerouteDeclaration(None=Multiply(
    A=LinearInterpolate(A=@n19.R, B=@n20.f3, Alpha=@n21)
    B=ScalarParameter() {defaultValue=0.6, group=Bubbles, parameterName=BubblesOpacity_Layer1, sortPriority=10}
   )) {name=Layer1_Mask, nodeColor=<rgba 0 1 0.2266 1>}
  NamedRerouteDeclaration(None=ScalarParameter() {defaultValue=1, group=Bubbles, parameterName=BubblesSpeed_Global, sortPriority=0}) {name=GlobalSpeed, nodeColor=<rgba 1 0 0.1641 1>}
  n22 = Multiply(
   A=ComponentMask(None=NamedRerouteUsage()) {b=true}
   B=ScalarParameter() {defaultValue=0.1, group=Bubbles, parameterName=BubblesDistortTiling_Global, sortPriority=0}
  )
  n23 = Frac(None=Multiply(A=Time(), B=ScalarParameter() {defaultValue=0.15, group=Bubbles, parameterName=BubblesDistortSpeed_Global, sortPriority=0}) {constB=0.5})
  NamedRerouteDeclaration(None=Sine(None=Add(A=@n22, B=@n23))) {name=GlobalDistort, nodeColor=<rgba 1 0 0.0469 1>}
  n24 = VectorParameter() {group=Bubbles, parameterName=BubblesDistortAmplitude_Layer1, sortPriority=10, defaultValue=<rgba 0.2 0.2 0 1>}
  NamedRerouteDeclaration(None=Multiply(A=@n24.RGB, B=NamedRerouteUsage())) {name=BubbleDistort_1, nodeColor=<rgba 1 0 0.3281 1>}
  n25 = VectorParameter() {group=Bubbles, parameterName=BubblesDistortAmplitude_Layer2, sortPriority=20, defaultValue=<rgba 0 0 0 1>}
  NamedRerouteDeclaration(None=Multiply(A=@n25.RGB, B=NamedRerouteUsage())) {name=BubbleDistort_2, nodeColor=<rgba 1 0 0.3281 1>}
  n26 = NamedRerouteUsage()
  n27 = CameraPositionWS()
  n28 = Normalize(VectorInput=Subtract(A=@n26, B=TransformPosition(None=@n27) {transformSourceType=TRANSFORMPOSSOURCE_World}))
  n29 = ScalarParameter() {defaultValue=20, group=Bubbles, parameterName=BubblesParallaxDepth_Layer2, sortPriority=20}
  n30 = VectorParameter() {group=Bubbles, parameterName=BubblesPositionOffset_Layer2, sortPriority=20, defaultValue=<rgba 0 0 0 1>}
  n31 = Add(A=Add(A=@n26, B=Multiply(A=@n28, B=@n29)), B=@n30.RGB)
  n32 = ScalarParameter() {defaultValue=-0.02, group=Bubbles, parameterName=BubblesSpeedX_Layer2, sortPriority=20}
  n33 = ScalarParameter() {defaultValue=0.07, group=Bubbles, parameterName=BubblesSpeedY_Layer2, sortPriority=20}
  n34 = MaterialFunctionCall(
   TextureObject=NamedRerouteUsage()
   TextureSize=Multiply(A=NamedRerouteUsage(), B=ScalarParameter() {defaultValue=6, group=Bubbles, parameterName=BubblesSize_Layer2, sortPriority=20})
   WorldPosition=Add(A=NamedRerouteDeclaration(None=@n31) {name=BubbleWorldPosParallax_1, nodeColor=<rgba 0 1 0.3203 1>}, B=NamedRerouteUsage())
   In=Multiply(A=NamedRerouteUsage(), B=AppendVector(A=@n32, B=@n33))
   World Space Normal=NamedRerouteUsage()
   ProjectionTransitionContrast=NamedRerouteUsage()
  ) {materialFunction=&Preskinned_WP_Texture_ScalePan}
  n35 = MaterialFunctionCall(Float3=@n34.XY Texture) {materialFunction=&BreakOutFloat3Components}
  n36 = MaterialFunctionCall(f3=@n34.XY Texture) {materialFunction=&AddComponents}
  n37 = ScalarParameter() {defaultValue=0.5, group=Bubbles, parameterName=BubblesAmount_Layer2, sortPriority=20}
  NamedRerouteDeclaration(None=Multiply(
    A=LinearInterpolate(A=@n35.R, B=@n36.f3, Alpha=@n37)
    B=ScalarParameter() {defaultValue=0.6, group=Bubbles, parameterName=BubblesOpacity_Layer2, sortPriority=20}
   )) {name=Layer2_Mask, nodeColor=<rgba 1 0 0.0937 1>}
  n38 = VectorParameter() {group=Bubbles, parameterName=BubblesDistortAmplitude_Layer3, sortPriority=30, defaultValue=<rgba 0 0 0 1>}
  NamedRerouteDeclaration(None=Multiply(A=@n38.RGB, B=NamedRerouteUsage())) {name=BubbleDistort_3, nodeColor=<rgba 1 0 0.3281 1>}
  n39 = NamedRerouteUsage()
  n40 = CameraPositionWS()
  n41 = Normalize(VectorInput=Subtract(A=@n39, B=TransformPosition(None=@n40) {transformSourceType=TRANSFORMPOSSOURCE_World}))
  n42 = ScalarParameter() {defaultValue=30, group=Bubbles, parameterName=BubblesParallaxDepth_Layer3, sortPriority=30}
  n43 = VectorParameter() {group=Bubbles, parameterName=BubblesPositionOffset_Layer3, sortPriority=30, defaultValue=<rgba 0 0 0 1>}
  n44 = Add(A=Add(A=@n39, B=Multiply(A=@n41, B=@n42)), B=@n43.RGB)
  n45 = ScalarParameter() {defaultValue=0.02, group=Bubbles, parameterName=BubblesSpeedX_Layer3, sortPriority=30}
  n46 = ScalarParameter() {defaultValue=0.1, group=Bubbles, parameterName=BubblesSpeedY_Layer3, sortPriority=30}
  n47 = MaterialFunctionCall(
   TextureObject=NamedRerouteUsage()
   TextureSize=Multiply(A=NamedRerouteUsage(), B=ScalarParameter() {defaultValue=5, group=Bubbles, parameterName=BubblesSize_Layer3, sortPriority=30})
   WorldPosition=Add(A=NamedRerouteDeclaration(None=@n44) {name=BubbleWorldPosParallax_2, nodeColor=<rgba 0 1 0.3203 1>}, B=NamedRerouteUsage())
   In=Multiply(A=NamedRerouteUsage(), B=AppendVector(A=@n45, B=@n46))
   World Space Normal=NamedRerouteUsage()
   ProjectionTransitionContrast=NamedRerouteUsage()
  ) {materialFunction=&Preskinned_WP_Texture_ScalePan}
  n48 = MaterialFunctionCall(Float3=@n47.XY Texture) {materialFunction=&BreakOutFloat3Components}
  n49 = MaterialFunctionCall(f3=@n47.XY Texture) {materialFunction=&AddComponents}
  n50 = ScalarParameter() {defaultValue=1, group=Bubbles, parameterName=BubblesAmount_Layer3, sortPriority=30}
  NamedRerouteDeclaration(None=Multiply(
    A=LinearInterpolate(A=@n48.R, B=@n49.f3, Alpha=@n50)
    B=ScalarParameter() {defaultValue=0.3, group=Bubbles, parameterName=BubblesOpacity_Layer3, sortPriority=30}
   )) {name=Layer3_Mask, nodeColor=<rgba 0 0.4141 1 1>}
  NamedRerouteDeclaration(None=@n7) {name=Layer1+2_Mask, nodeColor=<rgba 0.9453 0 1 1>}
  NamedRerouteDeclaration(None=VertexInterpolator(VS=LocalPosition() {localOrigin=InstancePreSkinning}.XYZ).PS) {name=PreSkinnedPos, nodeColor=<rgba 0.2422 0 1 1>}
  NamedRerouteDeclaration(None=VertexInterpolator(VS=PreSkinnedNormal()).PS) {name=PreSkinnedNormal, nodeColor=<rgba 0.0781 1 0 1>}
  NamedRerouteDeclaration(None=ScalarParameter() {defaultValue=1, group=Bubbles, parameterName=BubblesSize_Global, sortPriority=0}) {name=GlobalSize, nodeColor=<rgba 1 0 0.3984 1>}
  NamedRerouteDeclaration(None=ScalarParameter() {defaultValue=2, parameterName=BubblesTriPlanarBlendSharpness_Global}) {name=BlendSharpness, nodeColor=<rgba 0 0.7187 1 1>}
 ]}
