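# Material function for the GreenHazard male body that outputs lower-arm U and V gradients and a torso height gradient.
# Gradients derive from interpolated pre-skinned local position and hardcoded body landmarks, so they stay fixed to the mesh during animation.
# Arm V repeats along a fixed axis, arm U repeats an Arctangent2Fast angle around the arm, and torso rescales Z.
# Several branches, including seam offset and segment projection, feed no output.
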MaterialFunction /BRCosmetics/Characters/Player/Male/Medium/Bodies/M_MED_GreenHazard/Materials/MF_GreenHazard_PreSkinnedGradients
graph {nodes [
  n1 = Subtract(A=NamedRerouteUsage(), B=NamedRerouteUsage())
  n2 = Normalize(VectorInput=VectorParameter() {defaultValue=<rgba 0.1 0.8 0.5 1>}.RGB)
  n3 = Normalize(VectorInput=CrossProduct(A=Normalize(VectorInput=@n1), B=@n2))
  n4 = DotProduct(A=@n3, B=NamedRerouteUsage())
  n5 = Add(A=@n4) {constB=-10}
  n6 = Constant() {desc=Radius, r=1}
  n7 = DotProduct(A=@n2, B=NamedRerouteUsage())
  n8 = Divide(A=Arctangent2Fast(Y=Divide(A=@n5, B=@n6), X=Divide(A=@n7, B=@n6)), B=Divide() {constA=710, constB=113, desc=Tau})
  Add(A=@n8) {constB=0.5, desc=Offset seam to inner arm}
  ComponentMask(None=@n3) {b=true}
  n9 = NamedRerouteUsage()
  n10 = Subtract(A=@n9, B=NamedRerouteUsage())
  n11 = Subtract(A=@n10, B=Subtract(A=@n9, B=NamedRerouteUsage()))
  Divide(
   A=DotProduct(A=@n10, B=Normalize(VectorInput=@n11))
   B=MaterialFunctionCall(Vector 3=@n11) {materialFunction=/Engine/Functions/Engine_MaterialFunctions02/Utility/VectorLength}.V3 Length
  )
  Frac(None=@n4)
  n12 = Normalize(VectorInput=Constant4Vector() {constant=<rgba 0.6 -0.4 0.6 1>}.RGB)
  n13 = NamedRerouteUsage()
  n14 = Add(A=Divide(A=DotProduct(A=@n12, B=@n13)) {constB=16}) {constB=0.82}
  FunctionOutput(None=NamedRerouteDeclaration(None=Frac(None=@n14)) {name=LowerArm_VGradient, nodeColor=<rgba 0 0.0391 1 1>}) {outputName=LowerArm_V}
  n15 = Multiply(A=@n8) {constB=100}
  FunctionOutput(None=NamedRerouteDeclaration(None=Frac(None=@n15)) {name=LowerArm_UGradient, nodeColor=<rgba 0.5234 1 0 1>}) {outputName=LowerArm_U}
  n16 = Subtract(A=ComponentMask(None=NamedRerouteUsage()) {b=true}, B=Constant() {r=107})
  n17 = Constant() {r=36}
  FunctionOutput(None=NamedRerouteDeclaration(None=Divide(A=@n16, B=@n17)) {name=TorsoZGradient, nodeColor=<rgba 0.6875 1 0 1>}) {outputName=TorsoGradient}
  NamedRerouteDeclaration(None=VertexInterpolator(VS=LocalPosition() {localOrigin=InstancePreSkinning}.XYZ).PS) {name=PreSkinnedPos, nodeColor=<rgba 1 0.2812 0 1>}
  NamedRerouteDeclaration(None=Constant4Vector() {constant=<rgba -49.9905 1.5256 106 1>}.RGB) {name=Wrist_PreSkinnedPos, nodeColor=<rgba 0.6875 0 1 1>}
  NamedRerouteDeclaration(None=Constant4Vector() {constant=<rgba -41.6309 -4.6707 113.6576 1>}.RGB) {name=Elbow_PreSkinnedPos, nodeColor=<rgba 0.7578 1 0 1>}
 ]}
