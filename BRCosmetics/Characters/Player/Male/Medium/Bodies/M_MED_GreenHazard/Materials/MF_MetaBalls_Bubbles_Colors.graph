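# Material function that colors the metaball bubbles in the GreenHazard character body materials.
# HeightToNormalSmooth turns the Height input into normals, whose tangent-space XY samples a lava lamp matcap and is also output as Bubble Warp.
# Matcap plus Fresnel scales a vertical-gradient reflection tint.
# This adds to bubble and scattering colors blended by world normal Z and the gradient, producing Bubble Colors.
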MaterialFunction /BRCosmetics/Characters/Player/Male/Medium/Bodies/M_MED_GreenHazard/Materials/MF_MetaBalls_Bubbles_Colors
graph {nodes [
  n1 = VectorParameter() {group=Bubbles - Color, parameterName=Reflection Color Up, sortPriority=6, defaultValue=<rgba 0.1667 0.0078 0 1>}
  n2 = VectorParameter() {group=Bubbles - Color, parameterName=Reflection Color Down, sortPriority=7, defaultValue=<rgba 1.5 0.0882 0.0136 1>}
  n3 = LinearInterpolate(
   A=@n1.RGB
   B=@n2.RGB
   Alpha=NamedRerouteDeclaration(None=FunctionInput() {inputName=VerticalGradient, inputType=FunctionInput_Scalar, sortPriority=10}) {name=VerticalGradient, nodeColor=<rgba 0.8516 1 0 1>}
  )
  n4 = MaterialFunctionCall!(Height=FunctionInput() {inputName=Height, inputType=FunctionInput_Scalar}, Absolute World Position, World Space Vertex Normals) {materialFunction=/Engine/Functions/Engine_MaterialFunctions03/Procedurals/HeightToNormalSmooth}
  n5 = ComponentMask(None=Transform(None=@n4.WorldSpaceNormal) {transformSourceType=TRANSFORMSOURCE_World, transformType=TRANSFORM_Tangent}) {g=true, r=true}
  n6 = TextureSampleParameter2D(UVs=Add(A=Multiply(A=@n5) {constB=0.5}) {constB=0.5}) {
   group=Bubbles - Color
   parameterName=Matcap
   sortPriority=-1
   texture=/Game/Environments/Asteria/Sets/Lair/Materials/LavaLamp/T_LavaLamp_Matcap
  }
  n7 = Add(
   A=Multiply(A=@n6.R, B=ScalarParameter() {defaultValue=1.4, group=Bubbles - Color, parameterName=Matcap Intensity, sortPriority=9})
   B=Multiply(
    A=FunctionInput() {inputName=Fresnel, inputType=FunctionInput_Scalar}
    B=ScalarParameter() {defaultValue=1.4, group=Bubbles - Color, parameterName=Fresnel Intensity, sortPriority=8}
   )
  )
  n8 = VectorParameter() {group=Bubbles - Color, parameterName=Scattering Color Down, sortPriority=5, defaultValue=<rgba 1 0.1238 0.0142 0.5>}
  n9 = VectorParameter() {group=Bubbles - Color, parameterName=Scattering Color Up, sortPriority=4, defaultValue=<rgba 0.1042 0 0.0052 1>}
  n10 = VectorParameter() {group=Bubbles - Color, parameterName=Bubbles Color Down, sortPriority=3, defaultValue=<rgba 1.8 0.1041 0 1>}
  n11 = VectorParameter() {group=Bubbles - Color, parameterName=Bubbles Color Up, sortPriority=2, defaultValue=<rgba 0.224 0 0.0045 1>}
  n12 = ComponentMask(None=@n4.WorldSpaceNormal) {b=true}
  n13 = ScalarParameter() {defaultValue=0.6, group=Bubbles - Color, parameterName=Bubbles Color Shift, sortPriority=8}
  n14 = LinearInterpolate(A=@n9.RGB, B=LinearInterpolate(A=@n10.RGB, B=@n11.RGB, Alpha=Add(A=@n12, B=@n13) {constB=0.5}))
  n15 = OneMinus(None=Multiply(A=@n8.A, B=NamedRerouteUsage()))
  FunctionOutput(None=Add(A=Multiply(A=@n3, B=@n7), B=LinearInterpolate(A=@n8.RGB, B=@n14, Alpha=@n15))) {outputName=Bubble Colors, sortPriority=1}
  FunctionOutput(None=@n5) {outputName=Bubble Warp, sortPriority=2}
 ]}
