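# Material function that renders animated metaball-style bubbles in foreground and background layers for the GreenHazard character body.
# Each layer multiplies two panning projections of a grayscale texture in pre-skinned space, with view-direction parallax and sine-driven distortion.
# The patterns become occludable masks and pseudo fresnel/height inputs for MF_MetaBalls_Bubbles_Colors.
# It outputs composited color, with a tinted background, plus a Fresnel edge-faded bubble mask.
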MaterialFunction /BRCosmetics/Characters/Player/Male/Medium/Bodies/M_MED_GreenHazard/Materials/MF_MetaBalls_Bubbles
graph {nodes [
  n1 = NamedRerouteUsage()
  n2 = ScalarParameter() {defaultValue=0.5, group=Bubbles - Background, parameterName=BubblesOpacity_Background, sortPriority=0}
  n3 = NamedRerouteUsage()
  n4 = Multiply(A=@n3, B=@n3)
  n5 = Multiply(A=@n4, B=@n4)
  n6 = Saturate(None=Multiply(A=OneMinus(None=@n5)) {constB=7})
  n7 = FunctionInput(Preview=Constant() {r=1}) {bUsePreviewValueAsDefault=true, inputName=Background Occlusion Mask, inputType=FunctionInput_Scalar, sortPriority=10, previewValue=<1 0 0 1>}
  n8 = NamedRerouteDeclaration(None=Multiply(A=Multiply(A=@n2, B=@n6), B=@n7)) {name=Background_Mask, nodeColor=<rgba 0.6172 1 0 1>}
  n9 = NamedRerouteUsage()
  FunctionOutput(None=LinearInterpolate(A=Multiply(A=@n1, B=@n8), B=NamedRerouteUsage(), Alpha=Ceil(None=@n9)))
  n10 = ScalarParameter() {defaultValue=0.3, group=Bubbles - Global, parameterName=BubbleEdgeFade_Exponent}
  n11 = NamedRerouteUsage()
  n12 = NamedRerouteUsage()
  n13 = Multiply(A=OneMinus(None=Fresnel(ExponentIn=@n10) {baseReflectFraction=0}), B=Multiply(A=Max(A=@n11, B=@n12)) {constB=2})
  FunctionOutput(None=NamedRerouteDeclaration(None=Saturate(None=@n13)) {name=Bubbles_Mask, nodeColor=<rgba 0 0.8828 1 1>}) {outputName=Mask, sortPriority=10}
  n14 = TextureObjectParameter() {
   group=Bubbles - Global
   parameterName=BubbleMask
   samplerType=SAMPLERTYPE_LinearGrayscale
   sortPriority=-1
   texture=/Game/Environments/Asteria/Sets/Lair/Materials/LavaLamp/T_LavaLamp_Background
  }
  NamedRerouteDeclaration(None=@n14) {name=BubblesTex, nodeColor=<rgba 1 0.8672 0 1>}
  n15 = FunctionInput(Preview=Constant() {r=1}) {bUsePreviewValueAsDefault=true, inputName=Foreground Occlusion Mask, inputType=FunctionInput_Scalar, sortPriority=10, previewValue=<1 0 0 1>}
  NamedRerouteDeclaration(None=@n15) {name=ForegroundOcclusion, nodeColor=<rgba 1 0 0.2578 1>}
  NamedRerouteDeclaration(None=ScalarParameter() {defaultValue=1, group=Bubbles - Global, parameterName=BubblesSpeed_Global, sortPriority=0}) {name=GlobalSpeed, nodeColor=<rgba 1 0 0.1641 1>}
  n16 = Multiply(
   A=ComponentMask(None=NamedRerouteUsage()) {b=true}
   B=ScalarParameter() {defaultValue=0.1, group=Bubbles - Global, parameterName=BubblesDistortTiling_Global, sortPriority=0}
  )
  n17 = Frac(None=Multiply(A=Time(), B=ScalarParameter() {defaultValue=0.15, group=Bubbles - Global, parameterName=BubblesDistortSpeed_Global, sortPriority=0}) {constB=0.5})
  NamedRerouteDeclaration(None=Sine(None=Add(A=@n16, B=@n17))) {name=GlobalDistort, nodeColor=<rgba 1 0 0.0469 1>}
  n18 = VectorParameter() {group=Bubbles - Foreground, parameterName=BubblesDistortAmplitude_Foreground, sortPriority=3, defaultValue=<rgba 0.2 0.2 0 1>}
  NamedRerouteDeclaration(None=Multiply(A=@n18.RGB, B=NamedRerouteUsage())) {name=BubbleDistort_1, nodeColor=<rgba 1 0 0.3281 1>}
  n19 = Subtract(A=ComponentMask(None=NamedRerouteUsage()) {r=true}) {constB=-30}
  NamedRerouteDeclaration(None=Saturate(None=Ceil(None=@n19))) {name=Arm/Torso, nodeColor=<rgba 1 0 0.1641 1>}
  n20 = NamedRerouteUsage()
  n21 = Multiply(
   A=NamedRerouteUsage()
   B=ScalarParameter() {defaultValue=80, group=Bubbles - Background, parameterName=BubblesSize_1_Background, sortPriority=10}
  )
  n22 = NamedRerouteUsage()
  n23 = Subtract(A=@n22, B=TransformPosition(None=CameraPositionWS()) {transformSourceType=TRANSFORMPOSSOURCE_World})
  n24 = ScalarParameter() {defaultValue=30, group=Bubbles - Background, parameterName=BubblesParallaxDepth_Background, sortPriority=3}
  n25 = FunctionInput() {bUsePreviewValueAsDefault=true, inputName=Background Parallax Offset, inputType=FunctionInput_Scalar}
  n26 = Add(A=@n22, B=Multiply(A=Normalize(VectorInput=@n23), B=Add(A=@n24, B=@n25)))
  n27 = VectorParameter() {group=Bubbles - Background, parameterName=BubblesPositionOffset_1_Background, sortPriority=30, defaultValue=<rgba 0 0 0 1>}
  n28 = Add(
   A=NamedRerouteDeclaration(None=Add(A=@n26, B=@n27.RGB)) {name=BubbleWorldPosParallax_2, nodeColor=<rgba 0 1 0.3203 1>}
   B=NamedRerouteUsage()
  )
  n29 = Multiply(
   A=NamedRerouteUsage()
   B=AppendVector(
    A=ScalarParameter() {defaultValue=0.02, group=Bubbles - Background, parameterName=BubblesSpeedX_1_Background, sortPriority=20}
    B=ScalarParameter() {defaultValue=0.0346, group=Bubbles - Background, parameterName=BubblesSpeedY_1_Background, sortPriority=20}
   )
  )
  n30 = NamedRerouteUsage()
  n31 = NamedRerouteUsage()
  n32 = NamedRerouteUsage()
  n33 = Multiply(
   A=NamedRerouteUsage()
   B=ScalarParameter() {defaultValue=50, group=Bubbles - Background, parameterName=BubblesSize_2_Background, sortPriority=11}
  )
  n34 = NamedRerouteUsage()
  n35 = Subtract(A=@n34, B=TransformPosition(None=CameraPositionWS()) {transformSourceType=TRANSFORMPOSSOURCE_World})
  n36 = ScalarParameter() {defaultValue=30, group=Bubbles - Background, parameterName=BubblesParallaxDepth_Background, sortPriority=3}
  n37 = Add(A=@n34, B=Multiply(A=Normalize(VectorInput=@n35), B=Add(A=@n36, B=@n25)))
  n38 = VectorParameter() {group=Bubbles - Background, parameterName=BubblesPositionOffset_2_Background, sortPriority=31, defaultValue=<rgba 0 0 0 1>}
  n39 = Add(
   A=NamedRerouteDeclaration(None=Add(A=@n37, B=@n38.RGB)) {name=BubbleWorldPosParallax_3, nodeColor=<rgba 0 1 0.3203 1>}
   B=NamedRerouteUsage()
  )
  n40 = Multiply(
   A=NamedRerouteUsage()
   B=AppendVector(
    A=ScalarParameter() {defaultValue=0.02, group=Bubbles - Background, parameterName=BubblesSpeedX_2_Background, sortPriority=21}
    B=ScalarParameter() {defaultValue=0.0952, group=Bubbles - Background, parameterName=BubblesSpeedY_2_Background, sortPriority=21}
   )
  )
  n41 = NamedRerouteUsage()
  n42 = NamedRerouteUsage()
  n43 = Multiply(
   A=ComponentMask(None=MaterialFunctionCall!(
     TextureObject=@n20
     TextureSize=@n21
     WorldPosition=@n28
     Export Float 4
     In=@n29
     World Space Normal=@n30
     ProjectionTransitionContrast=@n31
    ) {materialFunction=/Game/Athena/SafeZone/Preskinned_WP_Texture_ScalePan}.XYZ Texture) {r=true}
   B=ComponentMask(None=MaterialFunctionCall(TextureObject=@n32, TextureSize=@n33, WorldPosition=@n39, In=@n40, World Space Normal=@n41, ProjectionTransitionContrast=@n42) {materialFunction=&Preskinned_WP_Texture_ScalePan}.XYZ Texture) {r=true}
  )
  NamedRerouteDeclaration(None=Saturate(None=Multiply(A=@n43) {constB=2.5})) {name=Background_Pattern, nodeColor=<rgba 0 1 0.1094 1>}
  n44 = VectorParameter() {group=Bubbles - Background, parameterName=BubblesDistortAmplitude_Background, sortPriority=4, defaultValue=<rgba 0 0 0 1>}
  NamedRerouteDeclaration(None=Multiply(A=@n44.RGB, B=NamedRerouteUsage())) {name=BubbleDistort_3, nodeColor=<rgba 1 0 0.3281 1>}
  n45 = NamedRerouteUsage()
  n46 = Multiply(
   A=NamedRerouteUsage()
   B=ScalarParameter() {defaultValue=100, group=Bubbles - Foreground, parameterName=BubblesSize_1_Foreground, sortPriority=10}
  )
  n47 = NamedRerouteUsage()
  n48 = Subtract(A=@n47, B=TransformPosition(None=CameraPositionWS()) {transformSourceType=TRANSFORMPOSSOURCE_World})
  n49 = ScalarParameter() {defaultValue=5, group=Bubbles - Foreground, parameterName=BubblesParallaxDepth_Foreground, sortPriority=2}
  n50 = FunctionInput() {bUsePreviewValueAsDefault=true, inputName=Foreground Parallax Offset, inputType=FunctionInput_Scalar}
  n51 = Add(A=@n47, B=Multiply(A=Normalize(VectorInput=@n48), B=Add(A=@n49, B=@n50)))
  n52 = VectorParameter() {group=Bubbles - Foreground, parameterName=BubblesPositionOffset_1_Foregrund, sortPriority=30, defaultValue=<rgba 0 0 0 1>}
  n53 = Add(A=NamedRerouteDeclaration(None=Add(A=@n51, B=@n52.RGB)) {name=BubbleWorldPosParallax, nodeColor=<rgba 0 1 0.3203 1>}, B=NamedRerouteUsage())
  n54 = Multiply(
   A=NamedRerouteUsage()
   B=AppendVector(
    A=ScalarParameter() {defaultValue=0.01, group=Bubbles - Foreground, parameterName=BubblesSpeedX_1_Foreground, sortPriority=20}
    B=ScalarParameter() {defaultValue=0.0357, group=Bubbles - Foreground, parameterName=BubblesSpeedY_1_Foreground, sortPriority=20}
   )
  )
  n55 = NamedRerouteUsage()
  n56 = NamedRerouteUsage()
  n57 = NamedRerouteUsage()
  n58 = Multiply(
   A=NamedRerouteUsage()
   B=ScalarParameter() {defaultValue=70, group=Bubbles - Foreground, parameterName=BubblesSize_2_Foreground, sortPriority=11}
  )
  n59 = NamedRerouteUsage()
  n60 = Subtract(A=@n59, B=TransformPosition(None=CameraPositionWS()) {transformSourceType=TRANSFORMPOSSOURCE_World})
  n61 = ScalarParameter() {defaultValue=5, group=Bubbles - Foreground, parameterName=BubblesParallaxDepth_Foreground, sortPriority=2}
  n62 = Add(A=@n59, B=Multiply(A=Normalize(VectorInput=@n60), B=Add(A=@n61, B=@n50)))
  n63 = VectorParameter() {group=Bubbles - Foreground, parameterName=BubblesPositionOffset_2_Foreground, sortPriority=31, defaultValue=<rgba 0 0 0 1>}
  n64 = Add(
   A=NamedRerouteDeclaration(None=Add(A=@n62, B=@n63.RGB)) {name=BubbleWorldPosParallax_1, nodeColor=<rgba 0 1 0.3203 1>}
   B=NamedRerouteUsage()
  )
  n65 = Multiply(
   A=NamedRerouteUsage()
   B=AppendVector(
    A=ScalarParameter() {defaultValue=-0.02, group=Bubbles - Foreground, parameterName=BubblesSpeedX_2_Foreground, sortPriority=21}
    B=ScalarParameter() {defaultValue=0.1465, group=Bubbles - Foreground, parameterName=BubblesSpeedY_2_Foreground, sortPriority=21}
   )
  )
  n66 = NamedRerouteUsage()
  n67 = NamedRerouteUsage()
  n68 = Multiply(
   A=ComponentMask(None=MaterialFunctionCall(TextureObject=@n45, TextureSize=@n46, WorldPosition=@n53, In=@n54, World Space Normal=@n55, ProjectionTransitionContrast=@n56) {materialFunction=&Preskinned_WP_Texture_ScalePan}.XY Texture) {r=true}
   B=ComponentMask(None=MaterialFunctionCall(TextureObject=@n57, TextureSize=@n58, WorldPosition=@n64, In=@n65, World Space Normal=@n66, ProjectionTransitionContrast=@n67) {materialFunction=&Preskinned_WP_Texture_ScalePan}.XY Texture) {r=true}
  )
  NamedRerouteDeclaration(None=Saturate(None=Multiply(A=@n68) {constB=2.5})) {name=Foreground_Pattern, nodeColor=<rgba 1 0 0.1069 1>}
  n69 = ScalarParameter() {defaultValue=1, group=Bubbles - Foreground, parameterName=BubblesOpacity_Foreground, sortPriority=0}
  n70 = NamedRerouteUsage()
  n71 = Multiply(A=@n70, B=@n70)
  n72 = Multiply(A=@n71, B=@n71)
  n73 = Saturate(None=Multiply(A=OneMinus(None=@n72)) {constB=7})
  NamedRerouteDeclaration(None=Multiply(A=Multiply(A=@n69, B=@n73), B=@n15)) {name=Foreground_Mask, nodeColor=<rgba 0.6172 1 0 1>}
  n74 = OneMinus(None=@n70)
  n75 = ScalarParameter() {defaultValue=1, group=Bubbles - Foreground, parameterName=BubblesNormalIntensity_Foreground, sortPriority=1}
  NamedRerouteDeclaration(None=MaterialFunctionCall!(Fresnel=@n72, Height=Multiply(A=@n74, B=@n75), VerticalGradient=NamedRerouteUsage()) {materialFunction=&/MF_MetaBalls_Bubbles_Colors}.Bubble Colors) {name=Foreground_Color, nodeColor=<rgba 1 0 0.9375 1>}
  n76 = Multiply(
   A=OneMinus(None=@n3)
   B=ScalarParameter() {defaultValue=1, group=Bubbles - Background, parameterName=BubblesNormalIntensity_Background, sortPriority=2}
  )
  n77 = NamedRerouteDeclaration(None=FunctionInput() {inputName=VerticalGradient, inputType=FunctionInput_Scalar}) {name=VerticalGradient, nodeColor=<rgba 1 0.7969 0 1>}
  n78 = VectorParameter() {group=Bubbles - Background, parameterName=Background Tint, sortPriority=1, defaultValue=<rgba 0.5 0.5 0.5 1>}
  NamedRerouteDeclaration(None=Multiply(
    A=MaterialFunctionCall(Fresnel=@n5, Height=@n76, VerticalGradient=@n77) {materialFunction=&MF_MetaBalls_Bubbles_Colors}.Bubble Colors
    B=@n78.RGB
   )) {name=Background_Color, nodeColor=<rgba 1 0 0.9375 1>}
  NamedRerouteDeclaration(None=VertexInterpolator(VS=LocalPosition() {localOrigin=InstancePreSkinning}.XYZ).PS) {name=PreSkinnedPos, nodeColor=<rgba 0.2422 0 1 1>}
  NamedRerouteDeclaration(None=VertexInterpolator(VS=PreSkinnedNormal()).PS) {name=PreSkinnedNormal, nodeColor=<rgba 0.0781 1 0 1>}
  NamedRerouteDeclaration(None=ScalarParameter() {defaultValue=1, group=Bubbles - Global, parameterName=BubblesSize_Global, sortPriority=0}) {name=GlobalSize, nodeColor=<rgba 1 0 0.3984 1>}
  NamedRerouteDeclaration(None=ScalarParameter() {defaultValue=2, group=Bubbles - Global, parameterName=BubblesTriPlanarBlendSharpness_Global}) {name=BlendSharpness, nodeColor=<rgba 0 0.7187 1 1>}
 ]}
