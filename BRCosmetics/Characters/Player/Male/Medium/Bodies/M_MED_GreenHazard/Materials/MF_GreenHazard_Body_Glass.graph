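# Material function for the glass body of the GreenHazard player cosmetic, layering gradient fog and metaball bubbles onto input material attributes.
# Fog density derives from scene depth behind the surface, with separate arm and torso depths split by pre-skinned local X.
# Pre-skinned gradients shape the vertical fog, while depth-based rib masks and parallax offsets feed the bubbles function.
# The results override base color, emissive and opacity.
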MaterialFunction /BRCosmetics/Characters/Player/Male/Medium/Bodies/M_MED_GreenHazard/Materials/MF_GreenHazard_Body_Glass
graph {nodes [
  n1 = NamedRerouteUsage()
  n2 = LinearInterpolate(
   A=VectorParameter() {group=Fog, parameterName=FogColor_Down, defaultValue=<rgba 0.2733 0.4 0.02 1>}.RGB
   B=VectorParameter() {group=Fog, parameterName=FogColor_Up, defaultValue=<rgba 0.005 0.1 0.005 1>}.RGB
   Alpha=NamedRerouteUsage()
  )
  n3 = NamedRerouteUsage()
  n4 = NamedRerouteUsage()
  n5 = Min(A=@n3, B=OneMinus(None=@n4))
  n6 = NamedRerouteUsage()
  n7 = LinearInterpolate(
   A=VectorParameter() {group=Fog, parameterName=FogEmissive_Down, defaultValue=<rgba 0.0683 0.1 0.005 1>}.RGB
   B=VectorParameter() {group=Fog, parameterName=FogEmissive_Up, defaultValue=<rgba 0.001 0.02 0.001 1>}.RGB
   Alpha=NamedRerouteUsage()
  )
  n8 = Max(A=NamedRerouteUsage(), B=NamedRerouteUsage())
  n9 = LinearInterpolate(A=@n8, B=Constant() {r=1}, Alpha=NamedRerouteUsage()) {constB=0.8}
  FunctionOutput(None=SetMaterialAttributes!(
    MaterialAttributes=NamedRerouteUsage()
    BaseColor=LinearInterpolate(A=@n1, B=@n2, Alpha=@n5)
    EmissiveColor=LinearInterpolate(A=@n6, B=@n7, Alpha=@n5)
    Opacity=Max(A=@n8, B=@n9)
   ) {attributeSetTypes [#1, #2, #3]})
  Multiply(A=@n3, B=@n4)
  n10 = Subtract(A=SceneDepth().R, B=Distance(A=WorldPosition().XYZ, B=CameraPositionWS()))
  n11 = VertexInterpolator(VS=LocalPosition() {localOrigin=InstancePreSkinning}.XYZ)
  n12 = Ceil(None=Subtract(A=ComponentMask(None=@n11.PS) {r=true}) {constB=-30})
  n13 = LinearInterpolate(
   A=ScalarParameter() {defaultValue=2, group=Fog, parameterName=FogDepth_Arm}
   B=ScalarParameter() {defaultValue=30, group=Fog, parameterName=FogDepth_Torso}
   Alpha=NamedRerouteDeclaration(None=Saturate(None=@n12)) {name=Arm/Torso, nodeColor=<rgba 1 0 0.1641 1>}
  )
  n14 = LinearInterpolate(
   A=ScalarParameter() {defaultValue=0.6, group=Fog, parameterName=Fog Min Density}
   B=ScalarParameter() {defaultValue=1, group=Fog, parameterName=Fog Max Density}
   Alpha=Saturate(None=Divide(A=@n10, B=@n13))
  ) {constA=0.5}
  NamedRerouteDeclaration(None=LinearInterpolate(A=Multiply(B=@n14) {constA=2}, B=@n14, Alpha=NamedRerouteUsage())) {name=FogMask, nodeColor=<rgba 1 0 0.6328 1>}
  n15 = NamedRerouteUsage()
  n16 = Constant() {r=0.7}
  n17 = Constant() {r=1}
  n18 = Saturate(None=Divide(A=@n10, B=Constant() {r=5}))
  n19 = Constant() {r=0.9}
  n20 = Constant() {r=1}
  n21 = Saturate(None=Divide(A=@n10, B=Constant() {r=12}))
  n22 = MaterialFunctionCall!(
   Background Parallax Offset=LinearInterpolate(A=ScalarParameter() {parameterName=Arm Paralax Offset}, Alpha=@n15) {constB=0}
   Foreground Parallax Offset
   VerticalGradient=OneMinus(None=NamedRerouteUsage())
   Foreground Occlusion Mask=LinearInterpolate(B=SmoothStep(Min=@n16, Max=@n17, Value=@n18) {desc=Ribs Mask}, Alpha=@n15)
   Background Occlusion Mask=LinearInterpolate(B=SmoothStep(Min=@n19, Max=@n20, Value=@n21) {desc=Ribs Mask}, Alpha=@n15) {constA=1.5}
  ) {materialFunction=&/MF_MetaBalls_Bubbles}
  NamedRerouteDeclaration(None=@n22.Mask) {name=BubblesMask, nodeColor=<rgba 0 0.2031 1 1>}
  NamedRerouteDeclaration(None=@n22.Result) {name=BubblesColor, nodeColor=<rgba 0.5234 1 0 1>}
  n23 = GetMaterialAttributes(None=NamedRerouteDeclaration(None=FunctionInput() {inputName=InMat, inputType=FunctionInput_MaterialAttributes}) {name=InMat, nodeColor=<rgba 1 0.9375 0 1>}) {attributeGetTypes [#1, #2, #3]}
  NamedRerouteDeclaration(None=@n23.Opacity) {name=InOpacity, nodeColor=<rgba 1 0 0.8437 1>}
  NamedRerouteDeclaration(None=@n23.BaseColor) {name=InBC, nodeColor=<rgba 0 1 0.8359 1>}
  NamedRerouteDeclaration(None=@n23.EmissiveColor) {name=In Emissive, nodeColor=<rgba 1 0.5156 0 1>}
  n24 = MaterialFunctionCall() {materialFunction=&/MF_GreenHazard_PreSkinnedGradients}
  n25 = LinearInterpolate(A=@n24.LowerArm_V, B=@n24.LowerArm_V, Alpha=NamedRerouteUsage())
  n26 = ScalarParameter() {defaultValue=0.1, group=Fog, parameterName=ArmFogGradient_Exp}
  n27 = ScalarParameter() {defaultValue=0.184, group=Fog, parameterName=TorsoFogGradient_Exp}
  n28 = NamedRerouteUsage()
  NamedRerouteDeclaration(None=Power(
    Base=NamedRerouteDeclaration(None=@n25) {name=VerticalGradient, nodeColor=<rgba 1 0 0.9844 1>}
    Exp=LinearInterpolate(A=@n26, B=@n27, Alpha=@n28) {constA=0.176, constB=0.2}
   )) {name=VerticalGradient_Mod, nodeColor=<rgba 0 0.9297 1 1>}
 ]}
