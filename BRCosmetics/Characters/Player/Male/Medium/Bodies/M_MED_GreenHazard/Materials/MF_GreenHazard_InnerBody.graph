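# Material function that tints the base color of the GreenHazard character's inner body green and passes the other material attributes through.
# It blends InMat's base color toward a green parameter, then multiplies by a second masked dark-green tint.
# An FX opacity mask texture supplies named reroutes for skeleton, torso shell and arm shell masks, plus a combined shells mask.
# A torso gradient color ramp from MF_GreenHazard_PreSkinnedGradients is built but left unconnected to the output.
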MaterialFunction /BRCosmetics/Characters/Player/Male/Medium/Bodies/M_MED_GreenHazard/Materials/MF_GreenHazard_InnerBody
graph {nodes [
  n1 = GetMaterialAttributes(None=NamedRerouteUsage()) {attributeGetTypes [#1]}
  n2 = VectorParameter() {parameterName=Param_1, defaultValue=<rgba 0.0834 0.5 0 1>}
  n3 = LinearInterpolate(A=@n1.BaseColor, B=@n2.RGB, Alpha=NamedRerouteUsage())
  n4 = LinearInterpolate(B=VectorParameter() {defaultValue=<rgba 0.0333 0.2 0 1>}.RGB, Alpha=NamedRerouteUsage()) {constA=1}
  FunctionOutput(None=SetMaterialAttributes!(MaterialAttributes=@n1.MaterialAttributes, BaseColor=Multiply(A=@n3, B=@n4)) {attributeSetTypes [#1]})
  n5 = MaterialFunctionCall() {materialFunction=&/MF_GreenHazard_PreSkinnedGradients}
  n6 = NamedRerouteUsage()
  n7 = Constant() {r=0.104}
  n8 = Constant() {r=0.1}
  n9 = NamedRerouteUsage()
  LinearInterpolate(
   A=VectorParameter() {parameterName=Param_2, defaultValue=<rgba 0.7478 1 0 1>}.RGB
   B=@n2.RGB
   Alpha=Power(Base=LinearInterpolate(A=@n5.TorsoGradient, B=@n5.TorsoGradient, Alpha=@n6), Exp=LinearInterpolate(A=@n7, B=@n8, Alpha=@n9))
  )
  NamedRerouteDeclaration(None=FunctionInput() {inputName=InMat, inputType=FunctionInput_MaterialAttributes}) {name=InMat, nodeColor=<rgba 0.5 1 0 1>}
  n10 = TextureSample() {
   samplerType=SAMPLERTYPE_Masks
   texture=/BRCosmetics/Characters/Player/Male/Medium/Bodies/M_MED_GreenHazard/Textures/T_GreenHazard_Body_FX_OpacityMasks
  }
  NamedRerouteDeclaration(None=@n10.R) {name=SkelMask, nodeColor=<rgba 1 0 0.8203 1>}
  NamedRerouteDeclaration(None=Max(
    A=NamedRerouteDeclaration(None=@n10.G) {name=TorsoShellMask, nodeColor=<rgba 1 0.5391 0 1>}
    B=NamedRerouteDeclaration(None=@n10.B) {name=ArmShellMask, nodeColor=<rgba 0 1 0.8594 1>}
   )) {name=ShellsMask, nodeColor=<rgba 0.2188 0 1 1>}
 ]}
