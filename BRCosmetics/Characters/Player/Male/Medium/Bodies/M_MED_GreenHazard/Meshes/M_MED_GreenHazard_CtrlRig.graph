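# Corrective Control Rig for the medium male GreenHazard cosmetic body, adjusting deform and dyn bones for collar, jacket, shoulder pad, straps, arms and feet.
# A shared DeformRig function runs first, then a sequence applies per-region parent constraints and weighted transform offsets.
# Weights come from spherical pose readers, remapped bone deltas and values stored in variables, such as thigh twist.
# An "RBAN Off" flag enables extra jacket correctives.
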ControlRigBlueprint /BRCosmetics/Characters/Player/Male/Medium/Bodies/M_MED_GreenHazard/Meshes/M_MED_GreenHazard_CtrlRig
graphs [
 Multiply_6_ContainedGraph {nodes [
   n1 = RigVMFunctionEntryNode()
   RigVMFunctionReturnNode(Result=&RigVMFunction_MathDoubleMul(B=@n1.C, A=&RigVMFunction_MathDoubleMul(A=@n1.A, B=@n1.B).Result).Result)
  ]}
 RigVMFunctionLibrary {links [], nodes []}
 RigVMFunction_Sequence_1_ContainedGraph {nodes [
   n1 = RigVMFunctionEntryNode()
   RigVMFunction_Sequence = &RigVMFunction_Sequence(ExecuteContext=@n1.ExecuteContext)
   n2 = &RigVMFunction_Sequence(ExecuteContext=@RigVMFunction_Sequence.B)
   n3 = &RigVMFunction_Sequence(ExecuteContext=@n2.B)
   n4 = &RigVMFunction_Sequence(ExecuteContext=@n3.B)
   n5 = &RigVMFunction_Sequence(ExecuteContext=@n4.B)
   n6 = &RigVMFunction_Sequence(ExecuteContext=@n5.B)
   n7 = &RigVMFunction_Sequence(ExecuteContext=@n6.B)
   n8 = &RigVMFunction_Sequence(ExecuteContext=@n7.B)
   n9 = &RigVMFunction_Sequence(ExecuteContext=@n8.B)
   RigVMFunctionReturnNode(A=@RigVMFunction_Sequence.A, B=@n2.A, C=@n3.A, D=@n4.A, E=@n5.A, F=@n6.A, G=@n7.A, H=@n8.A, I=@n9.A, J=@n9.B)
  ]}
 RigVMModel {nodes [
   Remap = &RigVMFunction_MathDoubleRemap(SourceMaximum=145, SourceMinimum=105, TargetMaximum=1, TargetMinimum=0, Value=0)
   Subtract = &RigVMFunction_MathDoubleSub(
    A=@Remap.Result
    B=&RigVMFunction_MathDoubleRemap(SourceMaximum=-120, SourceMinimum=-30, TargetMaximum=1, TargetMinimum=0, Value=RigVMVariableNode() {variable=Thigh_TwistZ_R}.Value).Result
   )
   AlphaInterp = &RigVMFunction_AlphaInterp(
    InterpSpeedDecreasing=10
    InterpSpeedIncreasing=15
    Scale=1
    bClampResult=true
    bMapRange=true
    Value=&RigVMFunction_MathDoubleClamp(Value=@Subtract.Result).Result
   )
   &RigUnit_ModifyTransforms(
    ExecutePin=()
    ItemToModify="((Item=(Type=Bone,Name=\"deform_jacket_back_alt_r\"),Transform=(Rotation=(X=0.1294,Y=-0.483,Z=0.2241,W=0.8365),Translation=(X=-10,Y=0,Z=0),Scale3D=(X=1,Y=1,Z=1))),(Item=(Type=Bone,Name=\"deform_jacket_side_alt_r\"),Transform=(Rotation=(X=0,Y=-0.0872,Z=0,W=0.99 ...(truncated)"
    WeightMinimum=0
    Weight=@AlphaInterp.Result
   )
   n1 = &RigVMFunction_MathDoubleRemap(SourceMaximum=145, SourceMinimum=105, TargetMaximum=1, TargetMinimum=0, Value=0)
   n2 = &RigVMFunction_MathDoubleRemap(SourceMaximum=-120, SourceMinimum=-30, TargetMaximum=1, TargetMinimum=0, Value=RigVMVariableNode() {variable=Thigh_TwistZ_L}.Value)
   n3 = &RigVMFunction_AlphaInterp(
    InterpSpeedDecreasing=10
    InterpSpeedIncreasing=15
    Scale=1
    bClampResult=true
    bMapRange=true
    Value=&RigVMFunction_MathDoubleClamp(Value=&RigVMFunction_MathDoubleSub(A=@n1.Result, B=@n2.Result).Result).Result
   )
   > &RigUnit_ModifyTransforms(
    ItemToModify="((Item=(Type=Bone,Name=\"deform_jacket_back_alt_l\"),Transform=(Rotation=(X=0.1485,Y=-0.554,Z=0.212,W=0.7912),Translation=(X=10,Y=0,Z=0),Scale3D=(X=1,Y=1,Z=1))),(Item=(Type=Bone,Name=\"deform_jacket_side_alt_l\"),Transform=(Rotation=(X=0,Y=-0.0872,Z=0,W=0.9962 ...(truncated)"
    WeightMinimum=0
    Weight=@n3.Result
   )
   &RigUnit_ModifyTransforms(
    ExecutePin=()
    ItemToModify="((Item=(Type=Bone,Name=\"deform_jacket_back_alt_r\"),Transform=(Rotation=(X=0.1294,Y=-0.483,Z=0.2241,W=0.8365),Translation=(X=-10,Y=0,Z=0),Scale3D=(X=1,Y=1,Z=1))),(Item=(Type=Bone,Name=\"deform_jacket_side_alt_r\"),Transform=(Rotation=(X=0,Y=-0.0872,Z=0,W=0.99 ...(truncated)"
    Weight=1
    WeightMinimum=0
   )
   > &RigUnit_ModifyTransforms(
    ItemToModify="((Item=(Type=Bone,Name=\"deform_jacket_back_alt_l\"),Transform=(Rotation=(X=0.1485,Y=-0.554,Z=0.212,W=0.7912),Translation=(X=10,Y=0,Z=0),Scale3D=(X=1,Y=1,Z=1))),(Item=(Type=Bone,Name=\"deform_jacket_side_alt_l\"),Transform=(Rotation=(X=0,Y=-0.0872,Z=0,W=0.9962 ...(truncated)"
    Weight=1
    WeightMinimum=0
   )
   RigUnit_BeginExecution = &RigUnit_BeginExecution()
   FN_DeformRig_FNC = RigVMFunctionReferenceNode(
    DeformRig_ Curve="(deformRig_l_upperarm_up_alpha_72_#1=False,deformRig_r_upperarm_up_alpha_73_#2=False,deformRig_l_upperarm_fr_alpha_74_#3=False,deformRig_r_upperarm_fr_alpha_75_CB51B8 ...(truncated)"
    DeformRig_Bool="(upperarm_1_#4=True,elbow_3_#5=True,wrist_7_#6=True,groin_8_#7=True,glute_17_#8=True,knee_9_40C408324CA261 ...(truncated)"
    DeformRig_Debug=(upperarm_debug_18_#4=False,wrist_debug_20_#6=False,lat_debug_26_#9=False,groinGlute_debug_27_#7=False)
    DeformRig_Value="(upperarm_up_18_#10=(Rotation=(X=0.0,Y=-0.0,Z=0.0,W=1.0),Translation=(X=0.0,Y=0.0,Z=8.0),Scale3D=(X=0.70,Y=1.0,Z=1.0)),upperarm_fwd_fr_19_#11=(Rotation=(X=0.0,Y=-0.0,Z=0.0,W=1.0),Translation=(X=0.0, ...(truncated)"
    ExecuteContext=@RigUnit_BeginExecution.ExecutePin
   )
   > &RigUnit_ModifyTransforms(
    ItemToModify=((Item=(Type=Bone,Name="dyn_collar_back_base"),Transform=(Rotation=(X=0,Y=0.0436,Z=0,W=0.999),Translation=(X=-0.15,Y=0,Z=0),Scale3D=(X=1,Y=1,Z=1))))
    Weight=1
    WeightMinimum=0
   )
   > &RigUnit_ModifyTransforms(
    ItemToModify=((Item=(Type=Bone,Name="dyn_collar_side_base_r"),Transform=(Rotation=(X=0,Y=0,Z=0,W=1),Translation=(X=-0.15,Y=0,Z=0),Scale3D=(X=1,Y=1,Z=1))))
    Weight=1
    WeightMinimum=0
   )
   > &RigUnit_ModifyTransforms(
    ItemToModify=((Item=(Type=Bone,Name="dyn_collar_side_base_l"),Transform=(Rotation=(X=0,Y=0,Z=0,W=1),Translation=(X=0.15,Y=0,Z=0),Scale3D=(X=1,Y=1,Z=1))))
    Weight=1
    WeightMinimum=0
   )
   > RigVMFunction_Sequence = RigVMAggregateNode()
   &RigUnit_ModifyTransforms(
    ItemToModify=((Item=(Type=Bone,Name="neck_01"),Transform=(Rotation=(X=0,Y=0,Z=0,W=1),Translation=(X=0,Y=0,Z=0),Scale3D=(X=1,Y=1,Z=1))))
    Weight=1
    WeightMinimum=0
    ExecutePin=@RigVMFunction_Sequence.A
   )
   n4 = &RigUnit_GetRelativeTransformForItem(Child=(Type=Bone,Name="head"), Parent=(Type=Bone,Name="spine_05"), bChildInitial=false, bParentInitial=false)
   n5 = &RigVMFunction_MathQuaternionToEuler(RotationOrder=ZYX, Value=@n4.RelativeTransform.Rotation)
   RigVMFunction_MathQuaternionFromEuler = &RigVMFunction_MathQuaternionFromEuler(
    Euler=(X=0,Y=0,Z=0)
    RotationOrder=XYZ
    Euler.Z=&RigVMFunction_MathDoubleClamp(Maximum=20, Value=&RigVMFunction_MathDoubleMul(B=-1, A=@n5.Result.Y).Result).Result
    Euler.Y=&RigVMFunction_MathDoubleMul(B=-0.5, A=@n5.Result.Z).Result
   )
   > &RigUnit_ModifyTransforms(
    ItemToModify="((Item=(Type=Bone,Name=\"dyn_collar_back_base\"),Transform=(Rotation=(X=0,Y=0,Z=0,W=1),Translation=(X=3,Y=0,Z=3),Scale3D=(X=1,Y=1,Z=1))),(Item=(Type=Bone,Name=\"dyn_collar_side_base_l\"),Transform=(Rotation=(X=0.0076,Y=0.0868,Z=-0.0868,W=0.9924),Translation=(X ...(truncated)"
    WeightMinimum=0
    ItemToModify.0.Transform.Translation.Z=&RigVMFunction_MathDoubleMul(B=-0.01, A=@n5.Result.Z).Result
    ItemToModify.0.Transform.Translation.X=&RigVMFunction_MathDoubleMul(B=-0.05, A=@n5.Result.Z).Result
    ItemToModify.0.Transform.Translation.Y=&RigVMFunction_MathDoubleMul(B=0.2, A=@n5.Result.X).Result
    Weight=&RigVMFunction_MathDoubleRemap(SourceMaximum=-90, SourceMinimum=0, TargetMaximum=1, TargetMinimum=0, Value=@n5.Result.Z).Result
    ItemToModify.0.Transform.Rotation=@RigVMFunction_MathQuaternionFromEuler.Result
   )
   n6 = &RigVMFunction_MathDoubleAdd(B=-3.5489, A=@n4.RelativeTransform.Translation.Y)
   > &RigUnit_ModifyTransforms(
    ItemToModify=((Item=(Type=Bone,Name="dyn_collar_back_base"),Transform=(Rotation=(X=0,Y=0,Z=0,W=1),Translation=(X=0,Y=0,Z=0),Scale3D=(X=1,Y=1,Z=1))))
    WeightMinimum=0
    ItemToModify.0.Transform.Translation.Z=&RigVMFunction_MathDoubleClamp(Maximum=2, Value=&RigVMFunction_MathDoubleMul(B=-0.5, A=@n6.Result).Result).Result
    Weight=&RigVMFunction_MathDoubleRemap(SourceMaximum=-90, SourceMinimum=0, TargetMaximum=0, TargetMinimum=1, Value=@n5.Result.Z).Result
   )
   n7 = &RigVMFunction_MathDoubleRemap(SourceMaximum=-35, SourceMinimum=0, TargetMaximum=1, TargetMinimum=0, Value=@n5.Result.Z)
   n8 = &RigVMFunction_MathDoubleMul(B=-3, A=@n5.Result.Y)
   > &RigUnit_ModifyTransforms(
    ItemToModify=((Item=(Type=Bone,Name="dyn_collar_side_base_r"),Transform=(Rotation=(X=0,Y=0,Z=0,W=1),Translation=(X=0,Y=0,Z=0),Scale3D=(X=1,Y=1,Z=1))))
    WeightMinimum=0
    Weight=@n7.Result
    ItemToModify.0.Transform.Rotation=&RigVMFunction_MathQuaternionFromEuler(Euler=(X=0,Y=25,Z=0), RotationOrder=XYZ, Euler.Y=&RigVMFunction_MathDoubleClamp(Maximum=150, Value=@n8.Result).Result).Result
   )
   n9 = &RigVMFunction_MathDoubleMul(B=3, A=@n5.Result.Y)
   > &RigUnit_ModifyTransforms(
    ItemToModify=((Item=(Type=Bone,Name="dyn_collar_side_base_l"),Transform=(Rotation=(X=0,Y=0,Z=0,W=1),Translation=(X=0,Y=0,Z=0),Scale3D=(X=1,Y=1,Z=1))))
    WeightMinimum=0
    Weight=@n7.Result
    ItemToModify.0.Transform.Rotation=&RigVMFunction_MathQuaternionFromEuler(Euler=(X=0,Y=25,Z=0), RotationOrder=XYZ, Euler.Y=&RigVMFunction_MathDoubleClamp(Maximum=150, Value=@n9.Result).Result).Result
   )
   > n10 = &RigUnit_SphericalPoseReader(
    ActiveRegionScaleFactors=(PositiveWidth=1,NegativeWidth=1,PositiveHeight=1,NegativeHeight=1)
    ActiveRegionSize=0.01
    Debug=(bDrawDebug=false,bDraw2D=False,bDrawLocalAxes=False,DebugScale=25,DebugSegments=20,DebugThickness=0.25)
    DriverAxis=(X=0,Y=1,Z=0)
    DriverItem=(Type=Bone,Name="head")
    FalloffRegionScaleFactors=(PositiveWidth=1,NegativeWidth=1,PositiveHeight=1,NegativeHeight=1)
    FalloffSize=0.2
    FlipHeightScaling=False
    FlipWidthScaling=False
    OptionalParentItem=(Type=Bone,Name="spine_05")
    RotationOffset=(X=28,Y=11,Z=0)
   )
   > n11 = &RigUnit_SphericalPoseReader(
    ActiveRegionScaleFactors=(PositiveWidth=1,NegativeWidth=5,PositiveHeight=5,NegativeHeight=2)
    ActiveRegionSize=0.04
    Debug=(bDrawDebug=false,bDraw2D=False,bDrawLocalAxes=False,DebugScale=25,DebugSegments=20,DebugThickness=0.25)
    DriverAxis=(X=0,Y=1,Z=0)
    DriverItem=(Type=Bone,Name="head")
    FalloffRegionScaleFactors=(PositiveWidth=1,NegativeWidth=3,PositiveHeight=3,NegativeHeight=1)
    FalloffSize=0.3
    FlipHeightScaling=False
    FlipWidthScaling=False
    OptionalParentItem=(Type=Bone,Name="spine_05")
    RotationOffset=(X=23,Y=-20,Z=90)
   )
   > n12 = &RigUnit_SphericalPoseReader(
    ActiveRegionScaleFactors=(PositiveWidth=1,NegativeWidth=1,PositiveHeight=1,NegativeHeight=1)
    ActiveRegionSize=0.01
    Debug=(bDrawDebug=false,bDraw2D=False,bDrawLocalAxes=False,DebugScale=25,DebugSegments=20,DebugThickness=0.25)
    DriverAxis=(X=0,Y=1,Z=0)
    DriverItem=(Type=Bone,Name="head")
    FalloffRegionScaleFactors=(PositiveWidth=1,NegativeWidth=1,PositiveHeight=1,NegativeHeight=1)
    FalloffSize=0.2
    FlipHeightScaling=False
    FlipWidthScaling=False
    OptionalParentItem=(Type=Bone,Name="spine_05")
    RotationOffset=(X=28,Y=169,Z=0)
   )
   > n13 = &RigUnit_SphericalPoseReader(
    ActiveRegionScaleFactors=(PositiveWidth=5,NegativeWidth=2,PositiveHeight=2,NegativeHeight=5)
    ActiveRegionSize=0.04
    Debug=(bDrawDebug=false,bDraw2D=False,bDrawLocalAxes=False,DebugScale=25,DebugSegments=20,DebugThickness=0.25)
    DriverAxis=(X=0,Y=1,Z=0)
    DriverItem=(Type=Bone,Name="head")
    FalloffRegionScaleFactors=(PositiveWidth=3,NegativeWidth=1,PositiveHeight=1,NegativeHeight=3)
    FalloffSize=0.3
    FlipHeightScaling=False
    FlipWidthScaling=False
    OptionalParentItem=(Type=Bone,Name="spine_05")
    RotationOffset=(X=12,Y=160,Z=0)
   )
   n14 = &RigVMFunction_MathDoubleMax(A=@n10.OutputParam, B=@n11.OutputParam)
   > &RigUnit_ModifyTransforms(
    ItemToModify=((Item=(Type=Bone,Name="dyn_collar_side_base_r"),Transform=(Rotation=(X=0,Y=0.3827,Z=0,W=0.9239),Translation=(X=2,Y=0,Z=0),Scale3D=(X=1,Y=1,Z=1))))
    WeightMinimum=0
    Weight=@n14.Result
   )
   n15 = &RigUnit_GetRelativeTransformForItem(Child=(Type=Bone,Name="head"), Parent=(Type=Bone,Name="spine_05"), bChildInitial=false, bParentInitial=false)
   > &RigUnit_ModifyTransforms(
    ItemToModify=((Item=(Type=Bone,Name="dyn_collar_fr_base_r"),Transform=(Rotation=(X=0.0831,Y=0.2996,Z=-0.0262,W=0.9501),Translation=(X=0,Y=0,Z=0),Scale3D=(X=1,Y=1,Z=1))))
    WeightMinimum=0
    Weight=&RigVMFunction_MathDoubleMax(
     A=@n14.Result
     B=&RigVMFunction_MathDoubleRemap(SourceMaximum=5, SourceMinimum=0, TargetMaximum=1, TargetMinimum=0, Value=@n15.RelativeTransform.Translation.Z).Result
    ).Result
   )
   n16 = &RigVMFunction_MathDoubleMax(A=@n12.OutputParam, B=@n13.OutputParam)
   > &RigUnit_ModifyTransforms(
    ItemToModify=((Item=(Type=Bone,Name="dyn_collar_side_base_l"),Transform=(Rotation=(X=0,Y=0.3827,Z=0,W=0.9239),Translation=(X=-2,Y=0,Z=0),Scale3D=(X=1,Y=1,Z=1))))
    WeightMinimum=0
    Weight=@n16.Result
   )
   Maximum = &RigVMFunction_MathDoubleMax(
    A=@n16.Result
    B=&RigVMFunction_MathDoubleRemap(SourceMaximum=-5, SourceMinimum=0, TargetMaximum=1, TargetMinimum=0, Value=@n15.RelativeTransform.Translation.Z).Result
   )
   > &RigUnit_ModifyTransforms(
    ItemToModify=((Item=(Type=Bone,Name="dyn_collar_fr_base_l"),Transform=(Rotation=(X=0.0831,Y=0.2996,Z=-0.0262,W=0.9501),Translation=(X=0,Y=0,Z=0),Scale3D=(X=1,Y=1,Z=1))))
    WeightMinimum=0
    Weight=@Maximum.Result
   )
   > n17 = &RigUnit_SphericalPoseReader(
    ActiveRegionScaleFactors=(PositiveWidth=3,NegativeWidth=1,PositiveHeight=1,NegativeHeight=1)
    ActiveRegionSize=0.05
    Debug=(bDrawDebug=false,bDraw2D=False,bDrawLocalAxes=False,DebugScale=25,DebugSegments=20,DebugThickness=0.25)
    DriverAxis=(X=-1,Y=0,Z=0)
    DriverItem=(Type=Bone,Name="upperarm_r")
    FalloffRegionScaleFactors=(PositiveWidth=1.5,NegativeWidth=1,PositiveHeight=1,NegativeHeight=2)
    FalloffSize=0.2
    FlipHeightScaling=False
    FlipWidthScaling=False
    OptionalParentItem=(Type=Bone,Name="spine_05")
    RotationOffset=(X=-95,Y=-35,Z=47)
   )
   > n18 = &RigUnit_SphericalPoseReader(
    ActiveRegionScaleFactors=(PositiveWidth=3,NegativeWidth=1,PositiveHeight=1,NegativeHeight=1)
    ActiveRegionSize=0.05
    Debug=(bDrawDebug=false,bDraw2D=False,bDrawLocalAxes=False,DebugScale=25,DebugSegments=20,DebugThickness=0.25)
    DriverAxis=(X=1,Y=0,Z=0)
    DriverItem=(Type=Bone,Name="upperarm_l")
    FalloffRegionScaleFactors=(PositiveWidth=1.5,NegativeWidth=1,PositiveHeight=2,NegativeHeight=1)
    FalloffSize=0.2
    FlipHeightScaling=False
    FlipWidthScaling=False
    OptionalParentItem=(Type=Bone,Name="spine_05")
    RotationOffset=(X=160,Y=200,Z=50)
   )
   > &RigUnit_ModifyTransforms(
    ItemToModify="((Item=(Type=Bone,Name=\"dyn_collar_fr_01_r\"),Transform=(Rotation=(X=0,Y=-0.3007,Z=0,W=0.9537),Translation=(X=0,Y=0,Z=-1),Scale3D=(X=1,Y=1,Z=1))),(Item=(Type=Bone,Name=\"dyn_collar_side_01_r\"),Transform=(Rotation=(X=0.1245,Y=-0.2981,Z=0.0393,W=0.9456),Transl ...(truncated)"
    WeightMinimum=0
    Weight=@n17.OutputParam
   )
   > &RigUnit_ModifyTransforms(
    ItemToModify="((Item=(Type=Bone,Name=\"dyn_collar_fr_01_l\"),Transform=(Rotation=(X=0,Y=-0.1305,Z=0,W=0.9914),Translation=(X=0,Y=0,Z=0),Scale3D=(X=1,Y=1,Z=1))),(Item=(Type=Bone,Name=\"dyn_collar_side_01_l\"),Transform=(Rotation=(X=0.1722,Y=-0.1285,Z=0.0227,W=0.9764),Transla ...(truncated)"
    WeightMinimum=0
    Weight=@n18.OutputParam
   )
   n19 = &RigUnit_GetRelativeTransformForItem(Child=(Type=Bone,Name="head"), Parent=(Type=Bone,Name="spine_05"), bChildInitial=false, bParentInitial=false)
   n20 = &RigVMFunction_MathQuaternionToEuler(RotationOrder=ZYX, Value=@n19.RelativeTransform.Rotation)
   Clamp = &RigVMFunction_MathDoubleClamp(Value=&RigVMFunction_MathDoubleSub(
     A=&RigVMFunction_MathDoubleRemap(SourceMaximum=-0.8, SourceMinimum=2.5, TargetMaximum=1, TargetMinimum=0, Value=@n19.RelativeTransform.Translation.Y).Result
     B=&RigVMFunction_MathDoubleRemap(SourceMaximum=-80, SourceMinimum=-50, TargetMaximum=1, TargetMinimum=0, Value=@n20.Result.Z).Result
    ).Result)
   > &RigUnit_ModifyTransforms(
    ItemToModify="((Item=(Type=Bone,Name=\"dyn_collar_back_base\"),Transform=(Rotation=(X=0,Y=0.1736,Z=0,W=0.9848),Translation=(X=0,Y=0,Z=0),Scale3D=(X=1,Y=1,Z=1))),(Item=(Type=Bone,Name=\"dyn_collar_side_base_l\"),Transform=(Rotation=(X=0,Y=0.0872,Z=0,W=0.9962),Translation=(X= ...(truncated)"
    WeightMinimum=0
    Weight=@Clamp.Result
   )
   > &RigUnit_ModifyTransforms(
    ItemToModify="((Item=(Type=Bone,Name=\"dyn_collar_side_base_r\"),Transform=(Rotation=(X=0,Y=-0.3827,Z=0,W=0.9239),Translation=(X=0,Y=0,Z=0),Scale3D=(X=1,Y=1,Z=1))),(Item=(Type=Bone,Name=\"dyn_collar_fr_base_r\"),Transform=(Rotation=(X=0,Y=-0.2164,Z=0,W=0.9763),Translation=( ...(truncated)"
    WeightMinimum=0
    Weight=&RigVMFunction_MathDoubleRemap(SourceMaximum=55, SourceMinimum=40, TargetMaximum=1, TargetMinimum=0, Value=RigVMVariableNode() {variable=Clavicle_Pitch_R}.Value).Result
   )
   > n21 = &RigUnit_SphericalPoseReader(
    ActiveRegionScaleFactors=(PositiveWidth=1,NegativeWidth=1,PositiveHeight=1,NegativeHeight=1)
    ActiveRegionSize=0.04
    Debug=(bDrawDebug=false,bDraw2D=False,bDrawLocalAxes=False,DebugScale=25,DebugSegments=20,DebugThickness=0.25)
    DriverAxis=(X=1,Y=0,Z=0)
    DriverItem=(Type=Bone,Name="head")
    FalloffRegionScaleFactors=(PositiveWidth=1,NegativeWidth=1,PositiveHeight=1,NegativeHeight=1)
    FalloffSize=0.15
    FlipHeightScaling=False
    FlipWidthScaling=False
    OptionalParentItem=(Type=Bone,Name="spine_05")
    RotationOffset=(X=-90,Y=0,Z=25)
   )
   > &RigUnit_ModifyTransforms(
    ItemToModify=((Item=(Type=Bone,Name="dyn_collar_back_base"),Transform=(Rotation=(X=0,Y=0.4226,Z=0,W=0.9063),Translation=(X=1.5,Y=0,Z=0),Scale3D=(X=1,Y=1,Z=1))))
    WeightMinimum=0
    Weight=@n21.OutputParam
   )
   &RigUnit_ModifyTransforms(
    ItemToModify="((Item=(Type=Bone,Name=\"spine_05\"),Transform=(Rotation=(X=0,Y=0,Z=0,W=1),Translation=(X=0,Y=0,Z=0),Scale3D=(X=1,Y=1,Z=1))),(Item=(Type=Bone,Name=\"spine_04\"),Transform=(Rotation=(X=0,Y=0,Z=0,W=1),Translation=(X=0,Y=0,Z=0),Scale3D=(X=1,Y=1,Z=1))),(Item=(Type ...(truncated)"
    Weight=1
    WeightMinimum=0
    ExecutePin=@RigVMFunction_Sequence.B
   )
   > &RigUnit_ParentConstraint(
    AdvancedSettings=(InterpolationType=Average,RotationOrderForFilter=XZY)
    Child=(Type=Bone,Name="dyn_jacket_base_fr_l")
    Parents=((Item=(Type=Bone,Name="spine_04"),Weight=1),(Item=(Type=Bone,Name="spine_03"),Weight=1),(Item=(Type=Bone,Name="spine_02"),Weight=0.5),(Item=(Type=Bone,Name="spine_01"),Weight=0.25))
    Weight=1
    bMaintainOffset=true
   )
   > &RigUnit_ParentConstraint(
    AdvancedSettings=(InterpolationType=Average,RotationOrderForFilter=XZY)
    Child=(Type=Bone,Name="dyn_jacket_base_fr_r")
    Parents=((Item=(Type=Bone,Name="spine_04"),Weight=1),(Item=(Type=Bone,Name="spine_03"),Weight=1),(Item=(Type=Bone,Name="spine_02"),Weight=0.5),(Item=(Type=Bone,Name="spine_01"),Weight=0.25))
    Weight=1
    bMaintainOffset=true
   )
   > &RigUnit_ParentConstraint(
    AdvancedSettings=(InterpolationType=Average,RotationOrderForFilter=XZY)
    Child=(Type=Bone,Name="dyn_jacket_base_side_l")
    Parents=((Item=(Type=Bone,Name="spine_03"),Weight=1),(Item=(Type=Bone,Name="spine_02"),Weight=1))
    Weight=1
    bMaintainOffset=true
   )
   > &RigUnit_ParentConstraint(
    AdvancedSettings=(InterpolationType=Average,RotationOrderForFilter=XZY)
    Child=(Type=Bone,Name="dyn_jacket_base_side_r")
    Parents=((Item=(Type=Bone,Name="spine_03"),Weight=1),(Item=(Type=Bone,Name="spine_02"),Weight=1))
    Weight=1
    bMaintainOffset=true
   )
   > &RigUnit_ParentConstraint(
    AdvancedSettings=(InterpolationType=Average,RotationOrderForFilter=XZY)
    Child=(Type=Bone,Name="dyn_jacket_base_back_l")
    Parents=((Item=(Type=Bone,Name="spine_02"),Weight=1),(Item=(Type=Bone,Name="spine_01"),Weight=0.85))
    Weight=1
    bMaintainOffset=true
   )
   > &RigUnit_ParentConstraint(
    AdvancedSettings=(InterpolationType=Average,RotationOrderForFilter=XZY)
    Child=(Type=Bone,Name="dyn_jacket_base_back_r")
    Parents=((Item=(Type=Bone,Name="spine_02"),Weight=1),(Item=(Type=Bone,Name="spine_01"),Weight=0.85))
    Weight=1
    bMaintainOffset=true
   )
   &RigUnit_ModifyTransforms(
    ItemToModify=((Item=(Type=Bone,Name="deform_pelvis"),Transform=(Rotation=(X=0,Y=0,Z=0,W=1),Translation=(X=0,Y=0,Z=0),Scale3D=(X=0.5,Y=0.5,Z=0.5))))
    Weight=1
    WeightMinimum=0
    ExecutePin=@RigVMFunction_Sequence.H
   )
   RigVMFunction_ControlFlowBranch = &RigVMFunction_ControlFlowBranch(ExecuteContext=@RigVMFunction_Sequence.I, Condition=RigVMVariableNode() {variable=RBAN Off}.Value)
   &RigUnit_ModifyTransforms(
    ItemToModify=((Item=(Type=Bone,Name="calf_r"),Transform=(Rotation=(X=0,Y=0,Z=0,W=1),Translation=(X=0,Y=0,Z=0),Scale3D=(X=1,Y=1,Z=1))))
    Weight=1
    WeightMinimum=0
    ExecutePin=@RigVMFunction_ControlFlowBranch.True
   )
   > &RigUnit_ModifyTransforms(
    ItemToModify=((Item=(Type=Bone,Name="thigh_r"),Transform=(Rotation=(X=0,Y=0,Z=0,W=1),Translation=(X=0,Y=0,Z=0),Scale3D=(X=1,Y=1,Z=1))))
    Weight=1
    WeightMinimum=0
   )
   > &RigUnit_ParentConstraint(
    AdvancedSettings=(InterpolationType=Average,RotationOrderForFilter=XZY)
    Child=(Type=Bone,Name="deform_jacket_side_alt_r")
    Parents=((Item=(Type=Bone,Name="None"),Weight=1),(Item=(Type=Bone,Name="pelvis"),Weight=0),(Item=(Type=Bone,Name="thigh_r"),Weight=1))
    Weight=1
   )
   > &RigUnit_ParentConstraint(
    AdvancedSettings=(InterpolationType=Average,RotationOrderForFilter=XZY)
    Child=(Type=Bone,Name="deform_jacket_back_alt_r")
    Parents=((Item=(Type=Bone,Name="None"),Weight=1),(Item=(Type=Bone,Name="pelvis"),Weight=0),(Item=(Type=Bone,Name="spine_01"),Weight=1),(Item=(Type=Bone,Name="thigh_r"),Weight=1))
    Weight=1
   )
   > &RigUnit_ParentConstraint(
    AdvancedSettings=(InterpolationType=Average,RotationOrderForFilter=XZY)
    Child=(Type=Bone,Name="deform_jacket_side_alt_l")
    Parents=((Item=(Type=Bone,Name="None"),Weight=1),(Item=(Type=Bone,Name="pelvis"),Weight=0),(Item=(Type=Bone,Name="thigh_l"),Weight=1))
    Weight=1
   )
   > &RigUnit_ParentConstraint(
    AdvancedSettings=(InterpolationType=Average,RotationOrderForFilter=XZY)
    Child=(Type=Bone,Name="deform_jacket_back_alt_l")
    Parents=((Item=(Type=Bone,Name="None"),Weight=1),(Item=(Type=Bone,Name="pelvis"),Weight=0),(Item=(Type=Bone,Name="spine_01"),Weight=1),(Item=(Type=Bone,Name="thigh_l"),Weight=1))
    Weight=1
   )
   > Get Delta Ref = RigVMFunctionReferenceNode(Item=(Type=Bone,Name="calf_r"))
   > Get Delta Ref_5 = RigVMFunctionReferenceNode(Item=(Type=Bone,Name="calf_l"))
   n22 = &RigVMFunction_MathDoubleRemap(SourceMaximum=135, SourceMinimum=105, TargetMaximum=1, TargetMinimum=0, Value=@"Get Delta Ref".Result.Rotation.Yaw)
   n23 = &RigVMFunction_MathDoubleRemap(SourceMaximum=-120, SourceMinimum=-30, TargetMaximum=1, TargetMinimum=0, Value=RigVMVariableNode() {variable=Thigh_TwistZ_R}.Value)
   n24 = &RigVMFunction_AlphaInterp(
    InterpSpeedDecreasing=10
    InterpSpeedIncreasing=15
    Scale=1
    bClampResult=true
    bMapRange=true
    Value=&RigVMFunction_MathDoubleClamp(Value=&RigVMFunction_MathDoubleSub(A=@n22.Result, B=@n23.Result).Result).Result
   )
   > &RigUnit_ModifyTransforms(
    ItemToModify="((Item=(Type=Bone,Name=\"deform_jacket_back_alt_r\"),Transform=(Rotation=(X=-0.0368,Y=-0.421,Z=-0.079,W=0.9029),Translation=(X=-47,Y=0,Z=0),Scale3D=(X=1,Y=1,Z=1))),(Item=(Type=Bone,Name=\"deform_jacket_side_alt_r\"),Transform=(Rotation=(X=0,Y=0,Z=0,W=1),Transl ...(truncated)"
    WeightMinimum=0
    Weight=@n24.Result
   )
   n25 = &RigVMFunction_MathDoubleRemap(SourceMaximum=135, SourceMinimum=105, TargetMaximum=1, TargetMinimum=0, Value=@"Get Delta Ref_5".Result.Rotation.Yaw)
   n26 = &RigVMFunction_MathDoubleRemap(SourceMaximum=-120, SourceMinimum=-30, TargetMaximum=1, TargetMinimum=0, Value=RigVMVariableNode() {variable=Thigh_TwistZ_L}.Value)
   n27 = &RigVMFunction_AlphaInterp(
    InterpSpeedDecreasing=10
    InterpSpeedIncreasing=15
    Scale=1
    bClampResult=true
    bMapRange=true
    Value=&RigVMFunction_MathDoubleClamp(Value=&RigVMFunction_MathDoubleSub(A=@n25.Result, B=@n26.Result).Result).Result
   )
   > &RigUnit_ModifyTransforms(
    ItemToModify="((Item=(Type=Bone,Name=\"deform_jacket_back_alt_l\"),Transform=(Rotation=(X=-0.0368,Y=-0.421,Z=-0.079,W=0.9029),Translation=(X=47,Y=0,Z=0),Scale3D=(X=1,Y=1,Z=1))),(Item=(Type=Bone,Name=\"deform_jacket_side_alt_l\"),Transform=(Rotation=(X=0,Y=0,Z=0,W=1),Transla ...(truncated)"
    WeightMinimum=0
    Weight=@n27.Result
   )
   > n28 = &RigUnit_SphericalPoseReader(
    ActiveRegionScaleFactors=(PositiveWidth=1,NegativeWidth=1,PositiveHeight=1,NegativeHeight=1)
    ActiveRegionSize=0.05
    Debug=(bDrawDebug=true,bDraw2D=False,bDrawLocalAxes=False,DebugScale=25,DebugSegments=20,DebugThickness=0.25)
    DriverAxis=(X=1,Y=0,Z=0)
    DriverItem=(Type=Bone,Name="thigh_r")
    FalloffRegionScaleFactors=(PositiveWidth=1,NegativeWidth=1,PositiveHeight=1,NegativeHeight=1)
    FalloffSize=0.4
    FlipHeightScaling=False
    FlipWidthScaling=False
    OptionalParentItem=(Type=Bone,Name="None")
    RotationOffset=(X=0,Y=-49,Z=-8)
   )
   > n29 = &RigUnit_SphericalPoseReader(
    ActiveRegionScaleFactors=(PositiveWidth=1,NegativeWidth=1,PositiveHeight=1,NegativeHeight=1)
    ActiveRegionSize=0.05
    Debug=(bDrawDebug=true,bDraw2D=False,bDrawLocalAxes=False,DebugScale=25,DebugSegments=20,DebugThickness=0.25)
    DriverAxis=(X=1,Y=0,Z=0)
    DriverItem=(Type=Bone,Name="calf_r")
    FalloffRegionScaleFactors=(PositiveWidth=1,NegativeWidth=1,PositiveHeight=1,NegativeHeight=1)
    FalloffSize=0.4
    FlipHeightScaling=False
    FlipWidthScaling=False
    OptionalParentItem=(Type=Bone,Name="None")
    RotationOffset=(X=35,Y=90,Z=0)
   )
   Multiply_1 = RigVMAggregateNode(
    C=&RigVMFunction_MathDoubleRemap(
     SourceMaximum=27
     SourceMinimum=41
     TargetMaximum=1
     TargetMinimum=0
     Value=&RigUnit_GetRelativeTransformForItem(Child=(Type=Bone,Name="pelvis"), Parent=(Type=Bone,Name="root"), bChildInitial=False, bParentInitial=False).RelativeTransform.Translation.Z
    ).Result
    A=@n28.OutputParam
    B=@n29.OutputParam
   )
   > &RigUnit_ModifyTransforms(
    ItemToModify="((Item=(Type=Bone,Name=\"deform_jacket_back_alt_r\"),Transform=(Rotation=(X=0,Y=0,Z=0,W=1),Translation=(X=10,Y=0,Z=-15),Scale3D=(X=0.5,Y=0.5,Z=0.5))),(Item=(Type=Bone,Name=\"deform_jacket_side_alt_r\"),Transform=(Rotation=(X=0,Y=0,Z=0,W=1),Translation=(X=0,Y=0 ...(truncated)"
    WeightMinimum=0
    Weight=@Multiply_1.Result
   )
   > &RigUnit_ModifyTransforms(
    ItemToModify="((Item=(Type=Bone,Name=\"deform_jacket_back_alt_l\"),Transform=(Rotation=(X=0,Y=0,Z=0,W=1),Translation=(X=0,Y=0,Z=0),Scale3D=(X=0.5,Y=0.5,Z=0.5))),(Item=(Type=Bone,Name=\"deform_jacket_side_alt_l\"),Transform=(Rotation=(X=0,Y=0,Z=0,W=1),Translation=(X=0,Y=0,Z= ...(truncated)"
    WeightMinimum=0
    Weight=@Multiply_1.Result
   )
   n30 = &RigVMFunction_ControlFlowBranch(ExecuteContext=@RigVMFunction_Sequence.J, Condition=RigVMVariableNode() {variable=RBAN Off}.Value)
   &RigUnit_ModifyTransforms(
    ItemToModify="((Item=(Type=Bone,Name=\"deform_jacket_front_alt_l\"),Transform=(Rotation=(X=0,Y=0,Z=0,W=1),Translation=(X=0,Y=0,Z=-4),Scale3D=(X=1,Y=1,Z=1))),(Item=(Type=Bone,Name=\"deform_jacket_front_alt_r\"),Transform=(Rotation=(X=0,Y=0,Z=0,W=1),Translation=(X=0,Y=0,Z=4), ...(truncated)"
    Weight=1
    WeightMinimum=0
    ExecutePin=@n30.True
   )
   &RigUnit_ModifyTransforms(
    ItemToModify=((Item=(Type=Bone,Name="clavicle_r"),Transform=(Rotation=(X=0,Y=0,Z=0,W=1),Translation=(X=0,Y=0,Z=0),Scale3D=(X=1,Y=1,Z=1))))
    Weight=1
    WeightMinimum=0
    ExecutePin=@RigVMFunction_Sequence.C
   )
   > &RigUnit_ModifyTransforms(
    ItemToModify=((Item=(Type=Bone,Name="upperarm_r"),Transform=(Rotation=(X=0,Y=0,Z=0,W=1),Translation=(X=0,Y=0,Z=0),Scale3D=(X=1,Y=1,Z=1))))
    Weight=1
    WeightMinimum=0
   )
   > &RigUnit_ParentConstraint(
    AdvancedSettings=(InterpolationType=Average,RotationOrderForFilter=XZY)
    Child=(Type=Bone,Name="dyn_shoulderpad_master_rotate_r")
    Filter=(TranslationFilter=(bX=true,bY=true,bZ=true),RotationFilter=(bX=true,bY=true,bZ=true),ScaleFilter=(bX=true,bY=true,bZ=true))
    Parents=((Item=(Type=Bone,Name="clavicle_r"),Weight=0.5),(Item=(Type=Bone,Name="upperarm_twist_01_r"),Weight=1))
    Weight=1
   )
   > &RigUnit_ParentConstraint(
    AdvancedSettings=(InterpolationType=Average,RotationOrderForFilter=XYZ)
    Child=(Type=Bone,Name="dyn_shoulderpad_master_rotate_r")
    Filter=(TranslationFilter=(bX=true,bY=true,bZ=true),RotationFilter=(bX=false,bY=false,bZ=false),ScaleFilter=(bX=false,bY=false,bZ=false))
    Parents=((Item=(Type=Bone,Name="deform_upperarm_up_r"),Weight=1))
    Weight=0.5
   )
   > &RigUnit_ParentConstraint(
    AdvancedSettings=(InterpolationType=Average,RotationOrderForFilter=XYZ)
    Child=(Type=Bone,Name="dyn_shoulderpad_master_rotate_r")
    Filter=(TranslationFilter=(bX=true,bY=true,bZ=true),RotationFilter=(bX=true,bY=true,bZ=false),ScaleFilter=(bX=false,bY=false,bZ=false))
    Parents=((Item=(Type=Bone,Name="deform_upperarm_bk_r"),Weight=0.85),(Item=(Type=Bone,Name="deform_upperarm_fr_r"),Weight=0.85))
    Weight=0.45
    bMaintainOffset=true
   )
   > Get Delta Ref_2 = RigVMFunctionReferenceNode(Item=(Type=Bone,Name="clavicle_r"))
   > RigVMVariableNode(Value=@"Get Delta Ref_2".Result.Rotation.Pitch) {variable=Clavicle_Pitch_R}
   > Get Delta Ref_1 = RigVMFunctionReferenceNode(Item=(Type=Bone,Name="upperarm_r"))
   > n31 = &RigUnit_ModifyTransforms(
    ItemToModify=((Item=(Type=Bone,Name="dyn_shoulderpad_master_rotate_r"),Transform=(Rotation=(X=-0.2164,Y=0,Z=0,W=0.9763),Translation=(X=0,Y=0,Z=0),Scale3D=(X=1,Y=1,Z=1))))
    WeightMinimum=0
    Weight=&RigVMFunction_MathDoubleRemap(SourceMaximum=120, SourceMinimum=20, TargetMaximum=1, TargetMinimum=0, Value=@"Get Delta Ref_1".Result.Rotation.Yaw).Result
   )
   n32 = &RigVMFunction_Sequence(ExecuteContext=@n31.ExecutePin)
   n33 = &RigUnit_SphericalPoseReader(
    ActiveRegionScaleFactors=(PositiveWidth=1.5,NegativeWidth=8,PositiveHeight=2,NegativeHeight=2)
    ActiveRegionSize=0.05
    Debug=(bDrawDebug=true,bDraw2D=False,bDrawLocalAxes=False,DebugScale=25,DebugSegments=20,DebugThickness=0.25)
    DriverAxis=(X=-1,Y=0,Z=0)
    DriverItem=(Type=Bone,Name="upperarm_r")
    FalloffRegionScaleFactors=(PositiveWidth=1.5,NegativeWidth=-2,PositiveHeight=1.5,NegativeHeight=1.5)
    FalloffSize=0.3
    FlipHeightScaling=False
    FlipWidthScaling=False
    OptionalParentItem=(Type=Bone,Name="None")
    RotationOffset=(X=0,Y=150,Z=0)
    ExecutePin=@n32.A
   )
   > &RigUnit_ParentConstraint(
    AdvancedSettings=(InterpolationType=Average,RotationOrderForFilter=XZY)
    Child=(Type=Bone,Name="dyn_shoulderpad_master_rotate_r")
    Filter=(TranslationFilter=(bX=false,bY=false,bZ=false),RotationFilter=(bX=true,bY=true,bZ=true),ScaleFilter=(bX=false,bY=false,bZ=false))
    Parents=((Item=(Type=Bone,Name="upperarm_twist_01_r"),Weight=1))
    bMaintainOffset=true
    Weight=&RigVMFunction_MathDoubleMul(B=0.5, A=@n33.OutputParam).Result
   )
   > &RigUnit_ModifyTransforms(
    ItemToModify="((Item=(Type=Bone,Name=\"dyn_shoulderpad_master_rotate_r\"),Transform=(Rotation=(X=0,Y=0.0872,Z=0,W=0.9962),Translation=(X=6.5,Y=0,Z=0),Scale3D=(X=1,Y=1,Z=1))),(Item=(Type=Bone,Name=\"dyn_shoulderpad_edge_up\"),Transform=(Rotation=(X=-0.0523,Y=-0.0087,Z=-0.000 ...(truncated)"
    WeightMinimum=0
    Weight=@n33.OutputParam
   )
   n34 = &RigVMFunction_MathDoubleSub(
    A=&RigVMFunction_MathDoubleMul(
     B=&RigVMFunction_MathDoubleRemap(SourceMaximum=-13, SourceMinimum=-25, TargetMaximum=1, TargetMinimum=0, Value=@"Get Delta Ref_1".Result.Rotation.Roll).Result
     A=&RigVMFunction_MathDoubleRemap(SourceMaximum=60, SourceMinimum=10, TargetMaximum=1, TargetMinimum=0, Value=@"Get Delta Ref_1".Result.Rotation.Yaw).Result
    ).Result
    B=&RigVMFunction_MathDoubleRemap(SourceMaximum=-40, SourceMinimum=-5, TargetMaximum=1, TargetMinimum=0, Value=@"Get Delta Ref_1".Result.Rotation.Pitch).Result
   )
   &RigUnit_ModifyTransforms(
    ItemToModify=((Item=(Type=Bone,Name="dyn_shoulderpad_master_rotate_r"),Transform=(Rotation=(X=-0.1305,Y=0,Z=0,W=0.9914),Translation=(X=0,Y=-0.9,Z=0),Scale3D=(X=1,Y=1.05,Z=1))))
    WeightMinimum=0
    Weight=&RigVMFunction_MathDoubleMul(
     B=&RigVMFunction_MathDoubleClamp(Value=@n34.Result).Result
     A=&RigVMFunction_MathDoubleRemap(SourceMaximum=-8, SourceMinimum=3, TargetMaximum=1, TargetMinimum=0, Value=@"Get Delta Ref_2".Result.Rotation.Yaw).Result
    ).Result
    ExecutePin=@n32.B
   )
   > &RigUnit_ModifyTransforms(
    ItemToModify=((Item=(Type=Bone,Name="dyn_shoulderpad_master_rotate_r"),Transform=(Rotation=(X=0,Y=0.1305,Z=0,W=0.9914),Translation=(X=0,Y=-1,Z=-2),Scale3D=(X=1,Y=1,Z=1))))
    WeightMinimum=0
    Weight=RigVMVariableNode() {variable=UpperArm Up}.Value
   )
   n35 = &RigVMFunction_Sequence(ExecuteContext=@RigVMFunction_Sequence.F)
   Get Delta Ref_6 = RigVMFunctionReferenceNode(Item=(Type=Bone,Name="foot_l"), ExecuteContext=@n35.A)
   > &RigUnit_ModifyTransforms(
    ItemToModify=((Item=(Type=Bone,Name="deform_foot_bk_l"),Transform=(Rotation=(X=0,Y=0.2164,Z=0,W=0.9763),Translation=(X=-2,Y=0,Z=-4.5),Scale3D=(X=1.2,Y=1,Z=1))))
    WeightMinimum=0
    Weight=&RigVMFunction_MathDoubleRemap(SourceMaximum=65, SourceMinimum=0, TargetMaximum=1, TargetMinimum=0, Value=@"Get Delta Ref_6".Result.Rotation.Yaw).Result
   )
   Get Delta Ref_7 = RigVMFunctionReferenceNode(Item=(Type=Bone,Name="foot_r"), ExecuteContext=@n35.B)
   > &RigUnit_ModifyTransforms(
    ItemToModify=((Item=(Type=Bone,Name="deform_foot_bk_r"),Transform=(Rotation=(X=0,Y=0.2164,Z=0,W=0.9763),Translation=(X=1.9,Y=0,Z=5),Scale3D=(X=1.2,Y=1,Z=1))))
    WeightMinimum=0
    Weight=&RigVMFunction_MathDoubleRemap(SourceMaximum=65, SourceMinimum=0, TargetMaximum=1, TargetMinimum=0, Value=@"Get Delta Ref_7".Result.Rotation.Yaw).Result
   )
   &RigUnit_ParentConstraint(
    AdvancedSettings=(InterpolationType=Average,RotationOrderForFilter=XZY)
    Child=(Type=Bone,Name="dyn_strap_side_base_01_l")
    Parents=((Item=(Type=Bone,Name="spine_04"),Weight=1),(Item=(Type=Bone,Name="spine_05"),Weight=1),(Item=(Type=Bone,Name="clavicle_l"),Weight=0.5))
    Weight=1
    ExecutePin=@RigVMFunction_Sequence.G
   )
   > &RigUnit_ParentConstraint(
    AdvancedSettings=(InterpolationType=Average,RotationOrderForFilter=XZY)
    Child=(Type=Bone,Name="dyn_strap_side_base_02_l")
    Parents=((Item=(Type=Bone,Name="spine_04"),Weight=1),(Item=(Type=Bone,Name="spine_05"),Weight=0.7),(Item=(Type=Bone,Name="clavicle_l"),Weight=0.25))
    Weight=1
   )
   > &RigUnit_ParentConstraint(
    AdvancedSettings=(InterpolationType=Average,RotationOrderForFilter=XZY)
    Child=(Type=Bone,Name="dyn_strap_side_base_01_r")
    Parents=((Item=(Type=Bone,Name="spine_04"),Weight=1),(Item=(Type=Bone,Name="spine_05"),Weight=1),(Item=(Type=Bone,Name="clavicle_r"),Weight=0.5))
    Weight=1
   )
   > &RigUnit_ParentConstraint(
    AdvancedSettings=(InterpolationType=Average,RotationOrderForFilter=XZY)
    Child=(Type=Bone,Name="dyn_strap_side_base_02_r")
    Parents=((Item=(Type=Bone,Name="spine_04"),Weight=1),(Item=(Type=Bone,Name="spine_05"),Weight=0.7),(Item=(Type=Bone,Name="clavicle_r"),Weight=0.25))
    Weight=1
   )
   n36 = &RigVMFunction_Sequence(ExecuteContext=@RigVMFunction_Sequence.E)
   n37 = &RigUnit_SphericalPoseReader(
    ActiveRegionScaleFactors=(PositiveWidth=1,NegativeWidth=1,PositiveHeight=1,NegativeHeight=1)
    ActiveRegionSize=0.1
    Debug=(bDrawDebug=false,bDraw2D=False,bDrawLocalAxes=False,DebugScale=25,DebugSegments=20,DebugThickness=0.25)
    DriverAxis=(X=-1,Y=0,Z=0)
    DriverItem=(Type=Bone,Name="upperarm_r")
    FalloffRegionScaleFactors=(PositiveWidth=1,NegativeWidth=1,PositiveHeight=1,NegativeHeight=1)
    FalloffSize=0.2
    FlipHeightScaling=False
    FlipWidthScaling=False
    OptionalParentItem=(Type=Bone,Name="None")
    RotationOffset=(X=-109,Y=0,Z=62)
    ExecutePin=@n36.A
   )
   > n38 = &RigUnit_SphericalPoseReader(
    ActiveRegionScaleFactors=(PositiveWidth=1,NegativeWidth=1,PositiveHeight=1,NegativeHeight=1)
    ActiveRegionSize=0.1
    Debug=(bDrawDebug=false,bDraw2D=False,bDrawLocalAxes=False,DebugScale=25,DebugSegments=20,DebugThickness=0.25)
    DriverAxis=(X=0,Y=0,Z=-1)
    DriverItem=(Type=Bone,Name="upperarm_r")
    FalloffRegionScaleFactors=(PositiveWidth=1,NegativeWidth=1,PositiveHeight=1,NegativeHeight=1)
    FalloffSize=0.2
    FlipHeightScaling=False
    FlipWidthScaling=False
    OptionalParentItem=(Type=Bone,Name="None")
    RotationOffset=(X=-152.5,Y=-4,Z=-57)
   )
   n39 = &RigVMFunction_MathDoubleMul(A=@n38.OutputParam, B=@n37.OutputParam)
   > &RigUnit_ModifyTransforms(
    ItemToModify=((Item=(Type=Bone,Name="upperarm_twist_01_r"),Transform=(Rotation=(X=0,Y=-0.0872,Z=0,W=0.9962),Translation=(X=0,Y=-4,Z=0.65),Scale3D=(X=1,Y=1,Z=1.2))))
    WeightMinimum=0
    Weight=@n39.Result
   )
   > &RigUnit_ModifyTransforms(
    ItemToModify=((Item=(Type=Bone,Name="deform_upperarm_up_r"),Transform=(Rotation=(X=0.0435,Y=-0.0697,Z=0.003,W=0.9966),Translation=(X=5.2,Y=-2.6917,Z=1.36),Scale3D=(X=1,Y=3,Z=1))))
    WeightMinimum=0
    Weight=@n39.Result
   )
   > &RigUnit_ModifyTransforms(
    ItemToModify=((Item=(Type=Bone,Name="deform_upperarm_bk_r"),Transform=(Rotation=(X=0,Y=-0.0297,Z=0,W=0.9996),Translation=(X=-6.768,Y=3.7619,Z=-0.4019),Scale3D=(X=1,Y=1,Z=1))))
    WeightMinimum=0
    Weight=@n39.Result
   )
   > &RigUnit_ModifyTransforms(
    ItemToModify=((Item=(Type=Bone,Name="deform_upperarm_fr_r"),Transform=(Rotation=(X=0,Y=0,Z=0,W=1),Translation=(X=3,Y=-4,Z=1.6),Scale3D=(X=1,Y=1,Z=1))))
    WeightMinimum=0
    Weight=@n39.Result
   )
   > RigVMVariableNode(Value=@n39.Result) {variable=UpperArm Up}
   n40 = &RigUnit_SphericalPoseReader(
    ActiveRegionScaleFactors=(PositiveWidth=1,NegativeWidth=1,PositiveHeight=1,NegativeHeight=1)
    ActiveRegionSize=0.1
    Debug=(bDrawDebug=false,bDraw2D=False,bDrawLocalAxes=False,DebugScale=25,DebugSegments=20,DebugThickness=0.25)
    DriverAxis=(X=1,Y=0,Z=0)
    DriverItem=(Type=Bone,Name="upperarm_l")
    FalloffRegionScaleFactors=(PositiveWidth=1,NegativeWidth=1,PositiveHeight=1,NegativeHeight=1)
    FalloffSize=0.2
    FlipHeightScaling=False
    FlipWidthScaling=False
    OptionalParentItem=(Type=Bone,Name="None")
    RotationOffset=(X=-6,Y=-8,Z=41)
    ExecutePin=@n36.B
   )
   > n41 = &RigUnit_SphericalPoseReader(
    ActiveRegionScaleFactors=(PositiveWidth=1,NegativeWidth=1,PositiveHeight=1,NegativeHeight=1)
    ActiveRegionSize=0.1
    Debug=(bDrawDebug=false,bDraw2D=False,bDrawLocalAxes=False,DebugScale=25,DebugSegments=20,DebugThickness=0.25)
    DriverAxis=(X=0,Y=0,Z=1)
    DriverItem=(Type=Bone,Name="upperarm_l")
    FalloffRegionScaleFactors=(PositiveWidth=1,NegativeWidth=1,PositiveHeight=1,NegativeHeight=1)
    FalloffSize=0.2
    FlipHeightScaling=False
    FlipWidthScaling=False
    OptionalParentItem=(Type=Bone,Name="None")
    RotationOffset=(X=-41,Y=-100,Z=0)
   )
   n42 = &RigVMFunction_MathDoubleMul(A=@n41.OutputParam, B=@n40.OutputParam)
   > &RigUnit_ModifyTransforms(
    ItemToModify=((Item=(Type=Bone,Name="upperarm_twist_01_l"),Transform=(Rotation=(X=-0.0431,Y=-0.1563,Z=-0.0068,W=0.9867),Translation=(X=0,Y=3.5,Z=-1.2),Scale3D=(X=1,Y=1.5,Z=1.5))))
    WeightMinimum=0
    Weight=@n42.Result
   )
   > &RigUnit_ModifyTransforms(
    ItemToModify=((Item=(Type=Bone,Name="deform_upperarm_up_l"),Transform=(Rotation=(X=0,Y=0,Z=0,W=1),Translation=(X=0,Y=-3,Z=2.5),Scale3D=(X=1.5,Y=1.2,Z=1))))
    WeightMinimum=0
    Weight=@n42.Result
   )
   > &RigUnit_ModifyTransforms(
    ItemToModify=((Item=(Type=Bone,Name="deform_upperarm_bk_l"),Transform=(Rotation=(X=-0.0858,Y=-0.173,Z=-0.0151,W=0.9811),Translation=(X=0,Y=0,Z=-3),Scale3D=(X=1,Y=1,Z=1))))
    WeightMinimum=0
    Weight=@n42.Result
   )
   n43 = &RigVMFunction_Sequence(ExecuteContext=@RigVMFunction_Sequence.D)
   Get Delta Ref_3 = RigVMFunctionReferenceNode(Item=(Type=Bone,Name="thigh_l"), ExecuteContext=@n43.A)
   > RigVMVariableNode(Value=@"Get Delta Ref_3".Result.Rotation.Yaw) {variable=Thigh_TwistZ_L}
   > &RigUnit_ModifyTransforms(
    ItemToModify=((Item=(Type=Bone,Name="dyn_jacket_base_fr_l"),Transform=(Rotation=(X=0,Y=0,Z=0.4226,W=0.9063),Translation=(X=0,Y=0,Z=0),Scale3D=(X=1,Y=1,Z=1))))
    WeightMinimum=0
    Weight=&RigVMFunction_MathDoubleRemap(SourceMaximum=-95, SourceMinimum=0, TargetMaximum=1, TargetMinimum=0, Value=@"Get Delta Ref_3".Result.Rotation.Yaw).Result
   )
   Get Delta Ref_4 = RigVMFunctionReferenceNode(Item=(Type=Bone,Name="thigh_r"), ExecuteContext=@n43.B)
   > RigVMVariableNode(Value=@"Get Delta Ref_4".Result.Rotation.Yaw) {variable=Thigh_TwistZ_R}
   n44 = &RigUnit_GetTransform(Item=(Type=Bone,Name="upperarm_r"), Space=LocalSpace, bInitial=true)
   n45 = &RigUnit_GetTransform(Item=(Type=Bone,Name="upperarm_r"), Space=LocalSpace, bInitial=false)
   Dot___1 = /Script/RigVM.RigVMFunction_MathVectorDot(
    A=/Script/RigVM.RigVMFunction_MathQuaternionRotateVector(Vector=(X=1,Y=0,Z=0), Transform=@n44.Transform.Rotation) {is=struct RigVMUnitNode, template="Rotate Vector::Execute(in Transform,in Vector,out Result)"}.Result
    B=&RigVMFunction_MathQuaternionRotateVector(Vector=(X=1,Y=0,Z=0), Transform=@n45.Transform.Rotation) {template="Rotate Vector::Execute(in Transform,in Vector,out Result)"}.Result
   ) {is=struct RigVMUnitNode, template="Dot,|::Execute(in A,in B,out Result)"}
   Acos = /Script/RigVM.RigVMFunction_MathDoubleAcos(Value=@Dot___1.Result) {is=struct RigVMUnitNode, template="Acos::Execute(in Value,out Result)"}
   Degrees = /Script/RigVM.RigVMFunction_MathDoubleDeg(Value=@Acos.Result) {is=struct RigVMUnitNode, template="Degrees::Execute(in Value,out Result)"}
   Dot___2 = &RigVMFunction_MathVectorDot(
    B=&RigVMFunction_MathQuaternionRotateVector(Vector=(X=1,Y=0,Z=0), Transform=@n45.Transform.Rotation) {template="Rotate Vector::Execute(in Transform,in Vector,out Result)"}.Result
    A=&RigVMFunction_MathQuaternionRotateVector(Vector=(X=0,Y=1,Z=0), Transform=@n44.Transform.Rotation) {template="Rotate Vector::Execute(in Transform,in Vector,out Result)"}.Result
   ) {template="Dot,|::Execute(in A,in B,out Result)"}
   &RigUnit_ModifyTransforms(
    ExecutePin=()
    ItemToModify=((Item=(Type=Bone,Name="dyn_jacket_base_fr_r"),Transform=(Rotation=(X=0,Y=0,Z=0.1305,W=0.9914),Translation=(X=0,Y=0,Z=0),Scale3D=(X=1,Y=1,Z=1))))
    WeightMinimum=0
    Weight=&RigVMFunction_MathDoubleRemap(SourceMaximum=-95, SourceMinimum=0, TargetMaximum=1, TargetMinimum=0, Value=@"Get Delta Ref_4".Result.Rotation.Yaw).Result
   )
   &RigUnit_ModifyTransforms(
    ExecutePin=()
    ItemToModify=((Item=(Type=Bone,Name="deform_jacket_back_alt_r"),Transform=(Rotation=(X=0,Y=-0.0872,Z=0,W=0.9962),Translation=(X=0,Y=0,Z=0),Scale3D=(X=1,Y=1,Z=1))))
    Weight=1
    WeightMinimum=0
   )
   n46 = &RigUnit_SphericalPoseReader(
    ActiveRegionScaleFactors=(PositiveWidth=1,NegativeWidth=1,PositiveHeight=1,NegativeHeight=1)
    ActiveRegionSize=0.04
    Debug=(bDrawDebug=true,bDraw2D=False,bDrawLocalAxes=False,DebugScale=25,DebugSegments=20,DebugThickness=0.25)
    DriverAxis=(X=1,Y=0,Z=0)
    DriverItem=(Type=Bone,Name="calf_r")
    ExecutePin=()
    FalloffRegionScaleFactors=(PositiveWidth=1,NegativeWidth=1,PositiveHeight=1,NegativeHeight=1)
    FalloffSize=0.4
    FlipHeightScaling=False
    FlipWidthScaling=False
    OptionalParentItem=(Type=Bone,Name="pelvis")
    RotationOffset=(X=91,Y=0,Z=-17)
   )
   n47 = &RigUnit_SphericalPoseReader(
    ActiveRegionScaleFactors=(PositiveWidth=1,NegativeWidth=1,PositiveHeight=1,NegativeHeight=1)
    ActiveRegionSize=0.04
    Debug=(bDrawDebug=true,bDraw2D=False,bDrawLocalAxes=False,DebugScale=25,DebugSegments=20,DebugThickness=0.25)
    DriverAxis=(X=1,Y=0,Z=0)
    DriverItem=(Type=Bone,Name="thigh_r")
    ExecutePin=()
    FalloffRegionScaleFactors=(PositiveWidth=1,NegativeWidth=1,PositiveHeight=1,NegativeHeight=1)
    FalloffSize=0.4
    FlipHeightScaling=False
    FlipWidthScaling=False
    OptionalParentItem=(Type=Bone,Name="pelvis")
    RotationOffset=(X=-75,Y=0,Z=25)
   )
   Multiply = &RigVMFunction_MathDoubleMul(A=@n46.OutputParam, B=@n47.OutputParam)
   &RigVMFunction_MathDoubleRemap(SourceMaximum=-120, SourceMinimum=0, TargetMaximum=0, TargetMinimum=1, Value=RigVMVariableNode() {variable=Thigh_TwistZ_R}.Value)
   &RigVMFunction_MathDoubleRemap(SourceMaximum=-120, SourceMinimum=0, TargetMaximum=0, TargetMinimum=1, Value=RigVMVariableNode() {variable=Thigh_TwistZ_L}.Value)
  ]}
]
hierarchy [
 root {loc=<0 0 0>, rot=<0 0 0 1>}
 pelvis {parent=root, loc=<0 -0.7071 93.1191>, rot=<0.0127 0.707 -0.0127 -0.707>}
 spine_01 {parent=pelvis, loc=<6.7494 -0.6388 0>, rot=<0 0 0.034 -0.9994>}
 spine_02 {parent=spine_01, loc=<7.2518 -0.035 0>, rot=<0 0 0.0238 -0.9997>}
 spine_03 {parent=spine_02, loc=<6.0104 0 0>, rot=<0 0 0.0149 -0.9999>}
 spine_04 {parent=spine_03, loc=<8.1024 0 0>, rot=<0 0 -0.0132 -0.9999>}
 spine_05 {parent=spine_04, loc=<10.0146 0 0>, rot=<0 0 -0.0015 -1>}
 neck_01 {parent=spine_05, loc=<11.3138 0.1529 0>, rot=<0 0 -0.2003 -0.9797>}
 neck_02 {parent=neck_01, loc=<4.5627 0 0>, rot=<0 0 0.0504 -0.9987>}
 head {parent=neck_02, loc=<5.3914 0 0>, rot=<0 0 0.115 -0.9934>}
 attach_cape {parent=spine_05, loc=<9.0174 -11.7139 0>, rot=<0.0616 -0.7044 -0.0616 -0.7044>}
 cape {parent=attach_cape, type=Null, loc=<0 0.031 1.4083>, rot=<0.031 0 0 0.9995>}
 attach_backpack {parent=spine_05, loc=<-0.1222 -13.4729 0>, rot=<-0.0093 -0.707 0.0093 -0.707>}
 Backpack_BR {parent=attach_backpack, type=Null, loc=<0 2.9629 0>, rot=<0 0 0 1>}
 clavicle_l {parent=spine_05, loc=<4.3575 -0.1466 -2.3223>, rot=<0.0465 -0.7289 -0.009 -0.6829>}
 upperarm_l {parent=clavicle_l, loc=<15.5111 0 0>, rot=<0.0226 -0.3234 -0.0088 -0.9459>}
 lowerarm_l {parent=upperarm_l, loc=<26.6514 0 0>, rot=<0.0391 -0.0099 -0.2453 -0.9686>}
 hand_l {parent=lowerarm_l, loc=<22.3491 0 0>, rot=<-0.7081 -0.0507 0.0975 -0.6975>}
 thumb_01_l {parent=hand_l, loc=<1.3118 -0.3191 -2.459>, rot=<0.6546 -0.4042 0.0664 -0.6354>}
 thumb_02_l {parent=thumb_01_l, loc=<5.9336 -0.0448 0.0401>, rot=<-0.0037 0.0213 0.298 -0.9543>}
 thumb_03_l {parent=thumb_02_l, loc=<2.2223 0.3682 -0.0907>, rot=<0.0118 -0.013 0.0066 -0.9998>}
 index_metacarpal_l {parent=hand_l, loc=<1.6535 0.0183 -2.96>, rot=<0.0158 -0.0124 -0.0173 -0.9996>}
 index_01_l {parent=index_metacarpal_l, loc=<8.3008 0.0357 -0.0524>, rot=<-0.0044 0.0485 0.0924 -0.9945>}
 index_02_l {parent=index_01_l, loc=<3.3587 -0.3004 0.0452>, rot=<-0.002 0.0027 0.3225 -0.9466>}
 index_03_l {parent=index_02_l, loc=<2.5578 -0.123 -0.0252>, rot=<-0.002 0.0067 0.2558 -0.9667>}
 middle_metacarpal_l {parent=hand_l, loc=<2.1856 -0.1725 -0.5984>, rot=<0.006 0.0168 -0.0329 -0.9993>}
 middle_01_l {parent=middle_metacarpal_l, loc=<7.3275 -0.2447 0.0129>, rot=<-0.0208 0.0042 0.0858 -0.9961>}
 middle_02_l {parent=middle_01_l, loc=<4.0123 0.1401 0.1175>, rot=<-0.0218 0.0258 0.3512 -0.9357>}
 middle_03_l {parent=middle_02_l, loc=<2.6077 0.0534 0.1373>, rot=<0.0082 0.0188 0.1708 -0.9851>}
 ring_metacarpal_l {parent=hand_l, loc=<2.2074 -0.3625 1.3638>, rot=<-0.0284 0.0722 -0.0458 -0.9959>}
 ring_01_l {parent=ring_metacarpal_l, loc=<7.0295 -0.6197 -0.1446>, rot=<-0.0489 -0.0252 0.1892 -0.9804>}
 ring_02_l {parent=ring_01_l, loc=<3.6174 0.4647 -0.1027>, rot=<0 0 0.2806 -0.9598>}
 ring_03_l {parent=ring_02_l, loc=<2.4152 0.114 0.0199>, rot=<0 0 0.1189 -0.9929>}
 pinky_metacarpal_l {parent=hand_l, loc=<1.7616 -0.9753 3.3814>, rot=<-0.0397 0.0814 -0.069 -0.9935>}
 pinky_01_l {parent=pinky_metacarpal_l, loc=<7.1709 -0.4831 -0.1094>, rot=<-0.0654 -0.0035 0.2242 -0.9723>}
 pinky_02_l {parent=pinky_01_l, loc=<2.8274 0.0893 -0.1043>, rot=<0.0037 0.0011 0.2854 -0.9584>}
 pinky_03_l {parent=pinky_02_l, loc=<2.2046 0.0307 -0.0078>, rot=<-0.0002 0.0098 0.1445 -0.9895>}
 deform_wrist_dn_base_l {parent=hand_l, loc=<0 -2.2342 0>, rot=<0 0 0 -1>}
 deform_wrist_dn_l {parent=deform_wrist_dn_base_l, loc=<0 0 0>, rot=<0 0 0 1>}
 deform_wrist_up_base_l {parent=hand_l, loc=<0 2.2682 0>, rot=<0 0 0 -1>}
 deform_wrist_up_l {parent=deform_wrist_up_base_l, loc=<0 0 0>, rot=<0 0 0 1>}
 lowerarm_twist_02_l {parent=lowerarm_l, loc=<9.6199 0 0>, rot=<0 0 0 -1>}
 lowerarm_twist_01_l {parent=lowerarm_l, loc=<17.8549 0 0>, rot=<0 0 0 -1>}
 deform_elbow_in_base_l {parent=lowerarm_l, loc=<0.2112 5.259 0.8274>, rot=<0 0 0 -1>}
 deform_elbow_in_l {parent=deform_elbow_in_base_l, loc=<0 0 0>, rot=<0 0 0 1>}
 deform_elbow_out_base_l {parent=lowerarm_l, loc=<0 -4 0>, rot=<0 0 0 -1>}
 deform_elbow_out_l {parent=deform_elbow_out_base_l, loc=<0 0 0>, rot=<0 0 0 1>}
 upperarm_twist_01_l {parent=upperarm_l, loc=<0.0041 0 0>, rot=<0 0 0 -1>}
 deform_upperarm_bk_base_l {parent=upperarm_twist_01_l, loc=<1.0813 -7.2193 -1.9453>, rot=<0 0 0 1>}
 deform_upperarm_bk_l {parent=deform_upperarm_bk_base_l, loc=<0 0 0>, rot=<0 0 0 1>}
 deform_upperarm_fr_base_l {parent=upperarm_twist_01_l, loc=<-0.3691 7.0506 -2.765>, rot=<0 0 0 1>}
 deform_upperarm_fr_l {parent=deform_upperarm_fr_base_l, loc=<0 0 0>, rot=<0 0 0 1>}
 deform_upperarm_up_base_l {parent=upperarm_twist_01_l, loc=<-3.2216 -1.1391 3.3311>, rot=<0 0 0 1>}
 deform_upperarm_up_l {parent=deform_upperarm_up_base_l, loc=<0 0 0>, rot=<0 0 0 1>}
 upperarm_twist_02_l {parent=upperarm_l, loc=<13.7941 0 0>, rot=<0 0 0 -1>}
 clavicle_r {parent=spine_05, loc=<4.3575 -0.1466 2.3223>, rot=<0.7289 0.0465 -0.6829 0.009>}
 upperarm_r {parent=clavicle_r, loc=<-15.5111 0 0>, rot=<0.0226 -0.3234 -0.0088 -0.9459>}
 lowerarm_r {parent=upperarm_r, loc=<-26.6514 0 0>, rot=<0.0391 -0.0099 -0.2453 -0.9686>}
 hand_r {parent=lowerarm_r, loc=<-22.3491 0 0>, rot=<-0.7081 -0.0507 0.0975 -0.6975>}
 thumb_01_r {parent=hand_r, loc=<-1.3118 0.3191 2.459>, rot=<0.6546 -0.4042 0.0664 -0.6354>}
 thumb_02_r {parent=thumb_01_r, loc=<-5.9336 0.0448 -0.0401>, rot=<-0.0037 0.0213 0.298 -0.9543>}
 thumb_03_r {parent=thumb_02_r, loc=<-2.2223 -0.3682 0.0907>, rot=<0.0118 -0.013 0.0066 -0.9998>}
 index_metacarpal_r {parent=hand_r, loc=<-1.6535 -0.0183 2.96>, rot=<0.0158 -0.0124 -0.0173 -0.9996>}
 index_01_r {parent=index_metacarpal_r, loc=<-8.3008 -0.0357 0.0524>, rot=<-0.0044 0.0485 0.0924 -0.9945>}
 index_02_r {parent=index_01_r, loc=<-3.3587 0.3004 -0.0452>, rot=<-0.002 0.0027 0.3225 -0.9466>}
 index_03_r {parent=index_02_r, loc=<-2.5578 0.123 0.0252>, rot=<-0.002 0.0067 0.2558 -0.9667>}
 middle_metacarpal_r {parent=hand_r, loc=<-2.1856 0.1725 0.5984>, rot=<0.006 0.0168 -0.0329 -0.9993>}
 middle_01_r {parent=middle_metacarpal_r, loc=<-7.3275 0.2447 -0.0129>, rot=<-0.0208 0.0042 0.0858 -0.9961>}
 middle_02_r {parent=middle_01_r, loc=<-4.0123 -0.1401 -0.1175>, rot=<-0.0218 0.0258 0.3512 -0.9357>}
 middle_03_r {parent=middle_02_r, loc=<-2.6077 -0.0534 -0.1373>, rot=<0.0082 0.0188 0.1708 -0.9851>}
 ring_metacarpal_r {parent=hand_r, loc=<-2.2074 0.3625 -1.3638>, rot=<-0.0284 0.0722 -0.0458 -0.9959>}
 ring_01_r {parent=ring_metacarpal_r, loc=<-7.0295 0.6197 0.1446>, rot=<-0.0489 -0.0252 0.1892 -0.9804>}
 ring_02_r {parent=ring_01_r, loc=<-3.6174 -0.4647 0.1027>, rot=<0 0 0.2806 -0.9598>}
 ring_03_r {parent=ring_02_r, loc=<-2.4152 -0.114 -0.0199>, rot=<0 0 0.1189 -0.9929>}
 pinky_metacarpal_r {parent=hand_r, loc=<-1.7616 0.9753 -3.3814>, rot=<-0.0397 0.0814 -0.069 -0.9935>}
 pinky_01_r {parent=pinky_metacarpal_r, loc=<-7.1709 0.4831 0.1094>, rot=<-0.0654 -0.0035 0.2242 -0.9723>}
 pinky_02_r {parent=pinky_01_r, loc=<-2.8274 -0.0893 0.1043>, rot=<0.0037 0.0011 0.2854 -0.9584>}
 pinky_03_r {parent=pinky_02_r, loc=<-2.2046 -0.0307 0.0078>, rot=<-0.0002 0.0098 0.1445 -0.9895>}
 deform_wrist_dn_base_r {parent=hand_r, loc=<-0.0004 2.2345 -0.0001>, rot=<0 0 0 -1>}
 deform_wrist_dn_r {parent=deform_wrist_dn_base_r, loc=<0 0 0>, rot=<0 0 0 1>}
 deform_wrist_up_base_r {parent=hand_r, loc=<0.0003 -2.2684 0.0001>, rot=<0 0 0 -1>}
 deform_wrist_up_r {parent=deform_wrist_up_base_r, loc=<0 0 0>, rot=<0 0 0 1>}
 lowerarm_twist_02_r {parent=lowerarm_r, loc=<-9.6199 0 0>, rot=<0 0 0 -1>}
 lowerarm_twist_01_r {parent=lowerarm_r, loc=<-17.8549 0 0>, rot=<0 0 0 -1>}
 deform_elbow_in_base_r {parent=lowerarm_r, loc=<-0.2111 -5.2591 -0.8275>, rot=<0 0 0 -1>}
 deform_elbow_in_r {parent=deform_elbow_in_base_r, loc=<-0.0001 0 0.0002>, rot=<0 0 0 1>}
 deform_elbow_out_base_r {parent=lowerarm_r, loc=<0 4 -0.0001>, rot=<0 0 0 -1>}
 deform_elbow_out_r {parent=deform_elbow_out_base_r, loc=<0 0 0>, rot=<0 0 0 1>}
 upperarm_twist_01_r {parent=upperarm_r, loc=<-0.0041 0 0>, rot=<0 0 0 -1>}
 deform_upperarm_bk_base_r {parent=upperarm_twist_01_r, loc=<-1.0814 7.2192 1.9455>, rot=<0 0 0 1>}
 deform_upperarm_bk_r {parent=deform_upperarm_bk_base_r, loc=<0 0 0>, rot=<0 0 0 1>}
 deform_upperarm_fr_base_r {parent=upperarm_twist_01_r, loc=<0.3689 -7.0506 2.7652>, rot=<0 0 0 1>}
 deform_upperarm_fr_r {parent=deform_upperarm_fr_base_r, loc=<0 0 0>, rot=<0 0 0 1>}
 deform_upperarm_up_base_r {parent=upperarm_twist_01_r, loc=<3.2214 1.1391 -3.3308>, rot=<0 0 0 1>}
 deform_upperarm_up_r {parent=deform_upperarm_up_base_r, loc=<0 0 0>, rot=<0 0 0 -1>}
 upperarm_twist_02_r {parent=upperarm_r, loc=<-13.7941 0 0>, rot=<0 0 0 -1>}
 dyn_shoulderpad_master_rotate_r {parent=upperarm_r, loc=<0 0 0>, rot=<-0.0146 -0.9967 -0.0647 -0.0463>}
 dyn_shoulderpad_master_r {parent=dyn_shoulderpad_master_rotate_r, loc=<-0.0742 0.7046 3.9291>, rot=<0 0 0 1>}
 dyn_shoulderpad_edge_up {parent=dyn_shoulderpad_master_r, loc=<-7.5356 0.0135 0.7611>, rot=<0.0048 0.2088 -0.0082 -0.9779>}
 dyn_shoulderpad_edge_down {parent=dyn_shoulderpad_edge_up, loc=<13.4237 -0.2141 -7.1936>, rot=<-0.0048 -0.2088 0.0082 -0.9779>}
 dyn_collar_side_base_r {parent=spine_05, loc=<12.8935 -0.2352 6.9616>, rot=<-0.0393 -0.1798 0.972 -0.1462>}
 dyn_collar_side_01_r {parent=dyn_collar_side_base_r, loc=<0 0 0>, rot=<0 0 0 1>}
 dyn_collar_fr_base_r {parent=spine_05, loc=<9.0313 7.1224 6.984>, rot=<-0.0183 0.1581 0.9625 -0.2195>}
 dyn_collar_fr_01_r {parent=dyn_collar_fr_base_r, loc=<0 0 0>, rot=<0 0 0 1>}
 dyn_collar_back_base {parent=spine_05, loc=<14.1226 -4.1641 0.0148>, rot=<-0.0455 -0.7041 0.707 0.0482>}
 dyn_collar_back_01 {parent=dyn_collar_back_base, loc=<0 0 0>, rot=<0 0 0 1>}
 dyn_collar_fr_base_l {parent=spine_05, loc=<9.0313 7.1224 -6.984>, rot=<-0.1229 -0.1645 -0.241 -0.9485>}
 dyn_collar_fr_01_l {parent=dyn_collar_fr_base_l, loc=<0 0 0>, rot=<0 0 0 1>}
 dyn_collar_side_base_l {parent=spine_05, loc=<12.8933 -0.2352 -6.9616>, rot=<0.1798 -0.0393 -0.1462 -0.972>}
 dyn_collar_side_01_l {parent=dyn_collar_side_base_l, loc=<0 0 0>, rot=<0 0 0 1>}
 deform_pec_aim_base_l {parent=spine_04, loc=<7.8732 12.6388 -2.4123>, rot=<0.0102 -0.4602 0.3979 -0.7936>}
 deform_pec_aim_l {parent=deform_pec_aim_base_l, loc=<0 0 0>, rot=<0 0 0 1>}
 deform_pec_base_l {parent=deform_pec_aim_l, loc=<10.766 0 0>, rot=<0 0 0 1>}
 deform_pec_l {parent=deform_pec_base_l, loc=<0.055 0 0>, rot=<0 0 0 1>}
 deform_pec_aim_base_r {parent=spine_04, loc=<7.8727 12.6388 2.4123>, rot=<-0.4602 -0.0102 0.7936 0.3979>}
 deform_pec_aim_r {parent=deform_pec_aim_base_r, loc=<0 0 0>, rot=<0 0 0 1>}
 deform_pec_base_r {parent=deform_pec_aim_r, loc=<-10.7662 -0.0006 -0.0007>, rot=<0 0 0 1>}
 deform_pec_r {parent=deform_pec_base_r, loc=<0 0 0>, rot=<0 0 0 1>}
 dyn_lapel_base_r {parent=spine_04, loc=<3.9332 12.4595 6.5484>, rot=<-0.0262 0.4465 0.8943 -0.0151>}
 dyn_lapel_01_r {parent=dyn_lapel_base_r, loc=<0 0 0>, rot=<0 0 0 1>}
 dyn_lapel_base_l {parent=spine_04, loc=<3.933 12.4595 -6.5484>, rot=<-0.4465 -0.0262 -0.0151 -0.8943>}
 dyn_lapel_01_l {parent=dyn_lapel_base_l, loc=<0 0 0>, rot=<0 0 0 1>}
 dyn_strap_side_base_01_r {parent=spine_04, loc=<1.8432 -1.4256 14.6086>, rot=<-0.2337 -0.1084 0.9639 0.0679>}
 dyn_strap_side_01_r {parent=dyn_strap_side_base_01_r, loc=<0 0 0>, rot=<0 0 0 1>}
 dyn_strap_side_base_02_r {parent=spine_04, loc=<-2.3612 -1.309 13.4712>, rot=<-0.1599 -0.1036 0.9801 0.0558>}
 dyn_strap_side_02_r {parent=dyn_strap_side_base_02_r, loc=<0 0 0>, rot=<0 0 0 1>}
 dyn_strap_side_base_02_l {parent=spine_04, loc=<-2.3614 -1.3091 -13.4712>, rot=<0.1036 -0.1599 0.0558 -0.9801>}
 dyn_strap_side_02_l {parent=dyn_strap_side_base_02_l, loc=<0 0 0>, rot=<0 0 0 1>}
 dyn_strap_side_base_01_l {parent=spine_04, loc=<1.8428 -1.4256 -14.6086>, rot=<0.1084 -0.2337 0.0679 -0.9639>}
 dyn_strap_side_01_l {parent=dyn_strap_side_base_01_l, loc=<0 0 0>, rot=<0 0 0 1>}
 dyn_jacket_base_fr_r {parent=spine_04, loc=<-5.6196 8.555 8.9277>, rot=<-0.0336 0.4288 0.8999 -0.0728>}
 dyn_jacket_01_fr_r {parent=dyn_jacket_base_fr_r, loc=<0 0 0>, rot=<0 0 0 1>}
 dyn_jacket_02_fr_r {parent=dyn_jacket_01_fr_r, loc=<12.1536 0.2487 0.2317>, rot=<0.0427 0.045 -0.0203 -0.9979>}
 dyn_jacket_03_fr_r {parent=dyn_jacket_02_fr_r, loc=<8.1397 0.0395 0.5352>, rot=<0.0407 0.029 -0.0143 -0.9986>}
 dyn_jacket_04_fr_r {parent=dyn_jacket_03_fr_r, loc=<7.6953 0.1282 -0.1141>, rot=<0.0001 0.01 -0.0083 -0.9999>}
 dyn_jacket_05_fr_r {parent=dyn_jacket_04_fr_r, loc=<15.4057 1.0215 -0.2152>, rot=<0.039 -0.0199 -0.0634 -0.997>}
 dyn_badge_base {parent=dyn_jacket_02_fr_r, loc=<5.8639 1.4349 2.94>, rot=<0.0057 0.0527 0.092 -0.9943>}
 dyn_badge_01 {parent=dyn_badge_base, loc=<0 0 0>, rot=<0 0 0 1>}
 dyn_jacket_base_fr_l {parent=spine_04, loc=<-5.6195 8.555 -8.9277>, rot=<-0.4288 -0.0336 -0.0728 -0.8999>}
 dyn_jacket_01_fr_l {parent=dyn_jacket_base_fr_l, loc=<0 0 0>, rot=<0 0 0 1>}
 dyn_jacket_02_fr_l {parent=dyn_jacket_01_fr_l, loc=<-12.1541 -0.2487 -0.2318>, rot=<0.0427 0.045 -0.0203 -0.9979>}
 dyn_jacket_03_fr_l {parent=dyn_jacket_02_fr_l, loc=<-8.1392 -0.0395 -0.5353>, rot=<0.0407 0.029 -0.0143 -0.9986>}
 dyn_jacket_04_fr_l {parent=dyn_jacket_03_fr_l, loc=<-7.6953 -0.1282 0.1141>, rot=<0.0001 0.01 -0.0083 -0.9999>}
 dyn_jacket_05_fr_l {parent=dyn_jacket_04_fr_l, loc=<-15.4057 -1.0215 0.2152>, rot=<0.039 -0.0199 -0.0634 -0.997>}
 dyn_jacket_base_side_r {parent=spine_03, loc=<-1.1268 0.4393 11.2677>, rot=<-0.0117 -0.0876 0.9928 -0.0812>}
 dyn_jacket_01_side_r {parent=dyn_jacket_base_side_r, loc=<0 0 0>, rot=<0 0 0 1>}
 dyn_jacket_02_side_r {parent=dyn_jacket_01_side_r, loc=<6.8887 0.2676 0.9868>, rot=<-0.0015 0.1037 -0.0193 -0.9944>}
 dyn_jacket_03_side_r {parent=dyn_jacket_02_side_r, loc=<15.2672 0.2903 0.3159>, rot=<0.0022 0.001 -0.0094 -1>}
 dyn_jacket_04_side_r {parent=dyn_jacket_03_side_r, loc=<14.4667 0.3063 0.3202>, rot=<0.0109 0.0222 -0.0217 -0.9995>}
 dyn_jacket_base_side_l {parent=spine_03, loc=<-1.127 0.4393 -11.2677>, rot=<0.0876 -0.0117 -0.0812 -0.9928>}
 dyn_jacket_01_side_l {parent=dyn_jacket_base_side_l, loc=<0 0 0>, rot=<0 0 0 1>}
 dyn_jacket_02_side_l {parent=dyn_jacket_01_side_l, loc=<-6.8889 -0.2676 -0.9868>, rot=<-0.0015 0.1037 -0.0193 -0.9944>}
 dyn_jacket_03_side_l {parent=dyn_jacket_02_side_l, loc=<-15.2669 -0.2903 -0.316>, rot=<0.0022 0.001 -0.0094 -1>}
 dyn_jacket_04_side_l {parent=dyn_jacket_03_side_l, loc=<-14.4667 -0.3063 -0.3202>, rot=<0.0109 0.0222 -0.0217 -0.9995>}
 dyn_jacket_base_back_l {parent=spine_02, loc=<-1.4787 -7.0263 -5.8896>, rot=<0.5375 0.1385 -0.0972 -0.8261>}
 dyn_jacket_01_back_l {parent=dyn_jacket_base_back_l, loc=<0 0 0>, rot=<0 0 0 1>}
 dyn_jacket_02_back_l {parent=dyn_jacket_01_back_l, loc=<-15.6945 -0.0369 0.2557>, rot=<0 0 0.0064 -1>}
 dyn_jacket_03_back_l {parent=dyn_jacket_02_back_l, loc=<-13.7731 -0.029 -0.1143>, rot=<0 -0.0008 -0.0222 -0.9998>}
 dyn_strap_back_base_l {parent=dyn_jacket_01_back_l, loc=<-0.6496 1.7781 -0.2612>, rot=<-0.0509 0.0024 0.0135 -0.9986>}
 dyn_strap_back_01_l {parent=dyn_strap_back_base_l, loc=<0 0 0>, rot=<0 0 0 -1>}
 dyn_strap_back_02_l {parent=dyn_strap_back_01_l, loc=<-15.0538 -0.0013 0.1256>, rot=<-0.0017 -0.0055 0 -1>}
 dyn_strap_back_03_l {parent=dyn_strap_back_02_l, loc=<-13.8131 -0.1106 -0.1781>, rot=<0.0083 0.0055 0 -1>}
 dyn_strap_back_04_l {parent=dyn_strap_back_03_l, loc=<-11.9428 -0.0859 -0.0056>, rot=<0 0 -0.0037 -1>}
 dyn_strap_back_05_l {parent=dyn_strap_back_04_l, loc=<-11.7743 0.0288 -0.0617>, rot=<0.0008 0.0022 -0.0036 -1>}
 dyn_strap_back_06_l {parent=dyn_strap_back_05_l, loc=<-11.0325 -0.2279 0.4176>, rot=<0 -0.0657 -0.0033 -0.9978>}
 dyn_jacket_base_back_r {parent=spine_02, loc=<-1.4789 -7.0264 5.8896>, rot=<0.1385 -0.5375 0.8261 -0.0972>}
 dyn_jacket_01_back_r {parent=dyn_jacket_base_back_r, loc=<0 0 0>, rot=<0 0 0 1>}
 dyn_jacket_02_back_r {parent=dyn_jacket_01_back_r, loc=<15.6943 0.0369 -0.2556>, rot=<0 0 0.0064 -1>}
 dyn_jacket_03_back_r {parent=dyn_jacket_02_back_r, loc=<13.7731 0.029 0.1143>, rot=<0 -0.0008 -0.0222 -0.9998>}
 dyn_strap_back_base_r {parent=dyn_jacket_01_back_r, loc=<0.6492 -1.7781 0.2613>, rot=<-0.0509 0.0024 0.0135 -0.9986>}
 dyn_strap_back_01_r {parent=dyn_strap_back_base_r, loc=<0 0 0>, rot=<0 0 0 1>}
 dyn_strap_back_02_r {parent=dyn_strap_back_01_r, loc=<15.0555 0.0008 0.0155>, rot=<-0.0017 -0.0055 0 -1>}
 dyn_strap_back_03_r {parent=dyn_strap_back_02_r, loc=<13.8139 -0.0016 0.1>, rot=<0.0083 0.0055 0 -1>}
 dyn_strap_back_04_r {parent=dyn_strap_back_03_r, loc=<11.943 0 0.1849>, rot=<0 0 0 1>}
 dyn_strap_back_05_r {parent=dyn_strap_back_04_r, loc=<11.7236 -0.2382 -0.5399>, rot=<0.0003 -0.0403 0.0087 -0.9992>}
 deform_pelvis {parent=spine_01, loc=<6.3105 -0.0257 0>, rot=<0.0419 -0.7059 0.7059 -0.0419>, scale=<0.5 0.5 0.5>}
 deform_jacket_back_alt_r {parent=spine_01, loc=<5.2423 -4.661 9.6064>, rot=<0.1183 -0.3901 0.9108 -0.0651>}
 deform_jacket_back_alt_l {parent=spine_01, loc=<5.242 -4.661 -9.6063>, rot=<0.3901 0.1183 -0.0651 -0.9108>}
 thigh_l {parent=pelvis, loc=<-2.5693 -0.0109 -7.7501>, rot=<0.0012 0.0572 0.0155 -0.9982>}
 calf_l {parent=thigh_l, loc=<-39.0484 0 0>, rot=<0.0006 -0.0161 -0.0391 -0.9991>}
 foot_l {parent=calf_l, loc=<-42.561 0.1077 0.4292>, rot=<0.0418 -0.0455 0.0343 -0.9975>}
 ball_l {parent=foot_l, loc=<-6.1599 13.8264 0.6806>, rot=<0.0001 -0.0001 -0.7063 -0.7079>}
 deform_foot_bk_l {parent=foot_l, loc=<0.0759 -4.7148 -0.0636>, rot=<0.6743 -0.0018 0.0055 -0.7384>}
 calf_twist_02_l {parent=calf_l, loc=<-18.8064 -0.0967 0.0397>, rot=<0.0237 0.0095 -0.0035 -0.9997>}
 calf_twist_01_l {parent=calf_l, loc=<-35.813 -0.0967 0.0397>, rot=<0.0237 0.0095 -0.0035 -0.9997>}
 deform_knee_in_base_l {parent=calf_l, loc=<1.298 -5.2957 -0.0496>, rot=<0 0 0 -1>}
 deform_knee_in_l {parent=deform_knee_in_base_l, loc=<0 0 0>, rot=<0 0 0 1>}
 deform_knee_out_base_l {parent=calf_l, loc=<0 4 0>, rot=<0 0 0 -1>}
 deform_knee_out_l {parent=deform_knee_out_base_l, loc=<0 0 0>, rot=<0 0 0 1>}
 thigh_twist_01_l {parent=thigh_l, loc=<-21.6303 0 0>, rot=<0.0013 0 0 -1>}
 deform_groin_base_l {parent=thigh_l, loc=<-2.1908 8.1142 -1.7442>, rot=<0 0 0 -1>}
 deform_groin_l {parent=deform_groin_base_l, loc=<0 0 0>, rot=<0 0 0 1>}
 deform_jacket_front_alt_l {parent=thigh_l, loc=<-8.1384 4.6272 -5.0898>, rot=<-0.2762 0.0261 -0.0757 -0.9578>}
 thigh_r {parent=pelvis, loc=<-2.5693 -0.0109 7.7501>, rot=<0.0572 -0.0012 0.9982 0.0155>}
 calf_r {parent=thigh_r, loc=<39.0484 0 0>, rot=<0.0006 -0.0161 -0.0391 -0.9991>}
 foot_r {parent=calf_r, loc=<42.561 -0.1077 -0.4292>, rot=<0.0418 -0.0455 0.0343 -0.9975>}
 ball_r {parent=foot_r, loc=<6.1599 -13.8264 -0.6806>, rot=<0.0001 -0.0001 -0.7063 -0.7079>}
 deform_foot_bk_r {parent=foot_r, loc=<-0.0759 4.7148 0.0636>, rot=<0.6743 -0.0018 0.0055 -0.7384>}
 calf_twist_02_r {parent=calf_r, loc=<18.8064 0.0967 -0.0397>, rot=<0.0237 0.0095 -0.0035 -0.9997>}
 calf_twist_01_r {parent=calf_r, loc=<35.813 0.0967 -0.0397>, rot=<0.0237 0.0095 -0.0035 -0.9997>}
 deform_knee_in_base_r {parent=calf_r, loc=<-1.298 5.2957 0.0496>, rot=<0 0 0 -1>}
 deform_knee_in_r {parent=deform_knee_in_base_r, loc=<0 0 0>, rot=<0 0 0 1>}
 deform_knee_out_base_r {parent=calf_r, loc=<0 -4 0>, rot=<0 0 0 -1>}
 deform_knee_out_r {parent=deform_knee_out_base_r, loc=<0 0 0>, rot=<0 0 0 1>}
 thigh_twist_01_r {parent=thigh_r, loc=<21.6303 0 0>, rot=<0.0013 0 0 -1>}
 deform_groin_base_r {parent=thigh_r, loc=<2.1908 -8.1142 1.7442>, rot=<0 0 0 -1>}
 deform_groin_r {parent=deform_groin_base_r, loc=<0 0 0>, rot=<0 0 0 1>}
 deform_jacket_front_alt_r {parent=thigh_r, loc=<8.1384 -4.6272 5.0898>, rot=<-0.2762 0.0261 -0.0757 -0.9578>}
 deform_glute_base_l {parent=pelvis, loc=<-8.8064 -8.8222 -8.4287>, rot=<0.0275 0.0501 -0.4472 -0.8926>}
 deform_glute_l {parent=deform_glute_base_l, loc=<0 0 0>, rot=<0 0 0 1>}
 deform_glute_base_r {parent=pelvis, loc=<-8.8062 -8.8222 8.4287>, rot=<0.0501 -0.0275 0.8926 -0.4472>}
 deform_glute_r {parent=deform_glute_base_r, loc=<0 0 0>, rot=<0 0 0 1>}
 deform_jacket_side_alt_r {parent=pelvis, loc=<3.5502 2.0555 13.0046>, rot=<0.0846 0.1457 0.9833 -0.0692>}
 deform_jacket_side_alt_l {parent=pelvis, loc=<3.5502 2.0555 -13.0046>, rot=<-0.1457 0.0846 -0.0692 -0.9833>}
]
variables [
 Clavicle_Pitch_R {type=double, default=0}
 LOD1 {type=bool, default=False}
 LOD2 {type=bool, default=False}
 RBAN Off {type=bool, default=False}
 Thigh_TwistZ_L {type=double, default=0}
 Thigh_TwistZ_R {type=double, default=0}
 UpperArm Up {type=double, default=0}
]
