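# Physics asset for the GreenHazard male medium body's coat, defining collision shapes and joints for the dynamic jacket bones.
# Six capsule chains (front, side and back, left and right) hang from small base bodies constrained to spine_01, mostly with limited swing and twist.
# Cross-constraints tie neighboring chains and both back panels into a connected skirt.
# Pelvis, spine and leg shapes plus a wide root box provide body and ground-level collision volumes.
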PhysicsAsset /BRCosmetics/Characters/Player/Male/Medium/Bodies/M_MED_GreenHazard/Meshes/M_MED_GreenHazard_Physics_Coat
bodies {
 calf_l [{shapeType=Capsule, length=43, radius=5.8, shapeName=calf_l_capsule, center=<-23.5848 -0.9162 0>, rotation=<rot 0 87.2876 -90>}]
 calf_r [{shapeType=Capsule, length=43, radius=5.8, shapeName=calf_r_capsule, center=<23.5848 0.9162 0>, rotation=<rot 0 87.2876 90>}]
 dyn_jacket_01_back_l [{shapeType=Capsule, length=13.3, radius=8, shapeName=dyn_jacket_01_back_l_capsule, center=<-6 0 -8>, rotation=<rot 90 90 0>}]
 dyn_jacket_01_back_r [{shapeType=Capsule, length=13.3, radius=8, shapeName=dyn_jacket_01_back_l_capsule, center=<6 0 8>, rotation=<rot 90 90 0>}]
 dyn_jacket_01_fr_l [{
   shapeType=Capsule
   length=25.3
   radius=3.8
   shapeName=dyn_jacket_01_back_l_capsule
   center=<-14.572 -3.5369 -2.7848>
   rotation=<rot 90 -166.2006 0>
  }]
 dyn_jacket_01_fr_r [{
   shapeType=Capsule
   length=25.3
   radius=3.8
   shapeName=dyn_jacket_01_back_l_capsule
   center=<14.572 3.5369 2.7848>
   rotation=<rot 90 -166.2006 0>
  }]
 dyn_jacket_01_side_l [{shapeType=Capsule, length=7, radius=8, shapeName=dyn_jacket_01_back_l_capsule, center=<-6 0 -7>, rotation=<rot 90 90 0>}]
 dyn_jacket_01_side_r [{shapeType=Capsule, length=5.5, radius=10.5, shapeName=dyn_jacket_01_back_l_capsule, center=<6 0 7>, rotation=<rot 90 90 0>}]
 dyn_jacket_02_back_l [{shapeType=Capsule, length=15.7, radius=8, shapeName=dyn_jacket_02_back_l_capsule, center=<-6 0 -4.7502>, rotation=<rot 90 90 0>}]
 dyn_jacket_02_back_r [{shapeType=Capsule, length=15.7, radius=8, shapeName=dyn_jacket_02_back_l_capsule, center=<6 0 5.4206>, rotation=<rot 90 90 0>}]
 dyn_jacket_02_fr_l [{
   shapeType=Capsule
   length=11.6
   radius=7.9
   shapeName=dyn_jacket_01_back_l_capsule
   center=<-6.0047 -0.0003 -7.1028>
   rotation=<rot 87.2047 -176.0845 -0.0002>
  }]
 dyn_jacket_02_fr_r [{shapeType=Capsule, length=12.2, radius=7.3, shapeName=dyn_jacket_01_back_l_capsule, center=<6 0 7.19>, rotation=<rot 90 -179.9667 0>}]
 dyn_jacket_02_side_l [{shapeType=Capsule, length=12.8, radius=8.1, shapeName=dyn_jacket_01_back_l_capsule, center=<-6 0 -7>, rotation=<rot 90 90 0>}]
 dyn_jacket_02_side_r [{shapeType=Capsule, length=4.7, radius=8, shapeName=dyn_jacket_01_back_l_capsule, center=<6 0 7>, rotation=<rot 90 90 0>}]
 dyn_jacket_03_back_l [{shapeType=Capsule, length=19.1, radius=8, shapeName=dyn_jacket_03_back_l_capsule, center=<-6 0 -3.6659>, rotation=<rot 90 90 0>}]
 dyn_jacket_03_back_r [{shapeType=Capsule, length=19.1, radius=8, shapeName=dyn_jacket_03_back_l_capsule, center=<6 0 3.6659>, rotation=<rot 90 90 0>}]
 dyn_jacket_03_fr_l [{shapeType=Capsule, length=10.6, radius=8.6, shapeName=dyn_jacket_03_back_l_capsule, center=<-7.3167 0 -7.987>, rotation=<rot 90 179.9999 0>}]
 dyn_jacket_03_fr_r [{shapeType=Capsule, length=11.2, radius=8, shapeName=dyn_jacket_03_back_l_capsule, center=<7.3167 0 7.987>, rotation=<rot 90 179.9999 0>}]
 dyn_jacket_03_side_l [{shapeType=Capsule, length=13.4, radius=8.4, shapeName=dyn_jacket_02_back_l_capsule, center=<-6 0 -7.2443>, rotation=<rot 90 90 0>}]
 dyn_jacket_03_side_r [{shapeType=Capsule, length=15.1, radius=8, shapeName=dyn_jacket_02_back_l_capsule, center=<6 0 7.1004>, rotation=<rot 90 90 0>}]
 dyn_jacket_04_fr_l [{shapeType=Capsule, length=9.9, radius=9.3, shapeName=dyn_jacket_03_back_l_capsule, center=<-7.3167 0 -7.987>, rotation=<rot 90 179.9999 0>}]
 dyn_jacket_04_fr_r [{shapeType=Capsule, length=11.2, radius=8, shapeName=dyn_jacket_03_back_l_capsule, center=<7.3167 0 7.987>, rotation=<rot 90 179.9999 0>}]
 dyn_jacket_04_side_l [{
   shapeType=Capsule
   length=13.6
   radius=8.4
   shapeName=dyn_jacket_03_back_l_capsule
   center=<-7.652 -0.4598 -6.0698>
   rotation=<rot 90 105.5546 0>
  }]
 dyn_jacket_04_side_r [{shapeType=Capsule, length=16.6, radius=8, shapeName=dyn_jacket_03_back_l_capsule, center=<8.0113 0.4699 5.5>, rotation=<rot 90 103.151 0>}]
 dyn_jacket_05_fr_l [{
   shapeType=Capsule
   length=1.2
   radius=13
   shapeName=dyn_jacket_02_back_l_capsule
   center=<-7.1511 0.0036 -11.5915>
   rotation=<rot 85.8267 0.0334 -179.9998>
  }]
 dyn_jacket_05_fr_r [{
   shapeType=Capsule
   length=1.7
   radius=12
   shapeName=dyn_jacket_02_back_l_capsule
   center=<9.7227 0.0022 10.9568>
   rotation=<rot 90 -179.9668 0>
  }]
 dyn_jacket_base_back_l [{shapeType=Capsule, length=0.505, radius=0.505, shapeName=dyn_jacket_base_back_l_capsule, rotation=<rot 0 -90 0>}]
 dyn_jacket_base_back_r [{shapeType=Capsule, length=0.505, radius=0.505, shapeName=dyn_jacket_base_back_r_capsule, rotation=<rot 0 -90 0>}]
 dyn_jacket_base_fr_l [{shapeType=Capsule, length=0.505, radius=0.505, shapeName=dyn_jacket_base_fr_l_capsule, rotation=<rot 0 -90 0>}]
 dyn_jacket_base_fr_r [{shapeType=Capsule, length=0.505, radius=0.505, shapeName=dyn_jacket_base_fr_r_capsule, rotation=<rot 0 -90 0>}]
 dyn_jacket_base_side_l [{shapeType=Capsule, length=0.505, radius=0.505, shapeName=dyn_jacket_base_side_l_capsule, rotation=<rot 0 -90 0>}]
 dyn_jacket_base_side_r [{shapeType=Capsule, length=0.505, radius=0.505, shapeName=dyn_jacket_base_side_r_capsule, rotation=<rot 0 -90 0>}]
 foot_l [{shapeType=Capsule, length=12.0411, radius=7.5921, shapeName=foot_l_capsule, center=<-7.5284 7.2191 -0.4879>, rotation=<rot 0 3.275 -89.3193>}]
 foot_r [{shapeType=Capsule, length=10.8411, radius=8.7921, shapeName=foot_r_capsule, center=<7.4197 -5.3187 0.4653>, rotation=<rot 0 3.275 90.6807>}]
 pelvis [{shapeType=Capsule, length=9.9, radius=10.5, shapeName=pelvis_capsule, rotation=<rot 0 90 0>}]
 root [{shapeType=Box, extentX=153.61, extentY=166.21, extentZ=35.01, shapeName=root_box, center=<0 0 -17.1594>, rotation=<rot 0 -90 -0.1526>}]
 spine_01 [{shapeType=Capsule, length=2.1048, radius=12.7857, shapeName=spine_01_capsule, rotation=<rot 90 0 0>}]
 thigh_l [
  {shapeType=Sphere, radius=9.3668, shapeName=thigh_l_sphere, center=<-26.6537 1.8636 0.7184>}
  {shapeType=Capsule, length=34.4677, radius=7.7, shapeName=thigh_r_capsule, center=<-19.3063 -3.5511 0>, rotation=<rot 0 94.2351 90>}
  {
   shapeType=Capsule
   length=36.7178
   radius=7.7
   shapeName=thigh_l_capsule
   center=<-15.6253 2.483 -0.0001>
   rotation=<rot -0.0024 88.291 90.0001>
  }
 ]
 thigh_r [{shapeType=Capsule, length=29.8, radius=7.7, shapeName=thigh_r_capsule, center=<20.0472 -0.6369 0>, rotation=<rot 0 94.2351 90>}]
}
constraints [
 {bone1Name=dyn_jacket_01_back_l, bone2Name=dyn_jacket_base_back_l, swing1LimitDegrees=3, swing2LimitDegrees=45, twistLimitDegrees=3}
 {bone1Name=dyn_jacket_03_back_l, bone2Name=dyn_jacket_02_back_l, swing1LimitDegrees=4, swing2LimitDegrees=45, twistLimitDegrees=4}
 {bone1Name=dyn_jacket_02_back_l, bone2Name=dyn_jacket_01_back_l, swing1LimitDegrees=4, swing2LimitDegrees=45, twistLimitDegrees=4}
 {bone1Name=dyn_jacket_03_back_r, bone2Name=dyn_jacket_02_back_r, swing1LimitDegrees=4, swing2LimitDegrees=45, twistLimitDegrees=4}
 {bone1Name=dyn_jacket_02_back_r, bone2Name=dyn_jacket_01_back_r, swing1LimitDegrees=4, swing2LimitDegrees=45, twistLimitDegrees=4}
 {bone1Name=dyn_jacket_01_back_r, bone2Name=dyn_jacket_base_back_r, swing1LimitDegrees=3, swing2LimitDegrees=45, twistLimitDegrees=3}
 {bone1Name=dyn_jacket_base_back_l, bone2Name=spine_01, swing1LimitDegrees=3, swing2LimitDegrees=35, twistLimitDegrees=3}
 {bone1Name=dyn_jacket_base_back_r, bone2Name=spine_01, swing1LimitDegrees=3, swing2LimitDegrees=35, twistLimitDegrees=3}
 {
  bone1Name=thigh_l
  bone2Name=pelvis
  swing1LimitDegrees=45
  swing1Motion=Free
  swing2LimitDegrees=45
  swing2Motion=Free
  twistLimitDegrees=45
  twistMotion=Free
 }
 {
  bone1Name=thigh_r
  bone2Name=pelvis
  swing1LimitDegrees=45
  swing1Motion=Free
  swing2LimitDegrees=45
  swing2Motion=Free
  twistLimitDegrees=45
  twistMotion=Free
 }
 {
  bone1Name=calf_r
  bone2Name=thigh_r
  swing1LimitDegrees=45
  swing1Motion=Free
  swing2LimitDegrees=45
  swing2Motion=Free
  twistLimitDegrees=45
  twistMotion=Free
 }
 {
  bone1Name=calf_l
  bone2Name=thigh_l
  swing1LimitDegrees=45
  swing1Motion=Free
  swing2LimitDegrees=45
  swing2Motion=Free
  twistLimitDegrees=45
  twistMotion=Free
 }
 {bone1Name=dyn_jacket_base_side_l, bone2Name=spine_01, swing1LimitDegrees=45, swing2LimitDegrees=45, twistLimitDegrees=45}
 {bone1Name=dyn_jacket_01_side_l, bone2Name=dyn_jacket_base_side_l, swing1LimitDegrees=4, swing2LimitDegrees=45, twistLimitDegrees=3}
 {bone1Name=dyn_jacket_02_side_l, bone2Name=dyn_jacket_01_side_l, swing1LimitDegrees=4, swing2LimitDegrees=35, twistLimitDegrees=3}
 {bone1Name=dyn_jacket_03_side_l, bone2Name=dyn_jacket_02_side_l, swing1LimitDegrees=15, swing2LimitDegrees=35, twistLimitDegrees=3}
 {bone1Name=dyn_jacket_04_side_l, bone2Name=dyn_jacket_03_side_l, swing1LimitDegrees=15, swing2LimitDegrees=35, twistLimitDegrees=4}
 {bone1Name=dyn_jacket_base_side_r, bone2Name=spine_01, swing1LimitDegrees=4, swing2LimitDegrees=35, twistLimitDegrees=4}
 {bone1Name=dyn_jacket_01_side_r, bone2Name=dyn_jacket_base_side_r, swing1LimitDegrees=4, swing2LimitDegrees=35, twistLimitDegrees=3}
 {bone1Name=dyn_jacket_02_side_r, bone2Name=dyn_jacket_01_side_r, swing1LimitDegrees=4, swing2LimitDegrees=35, twistLimitDegrees=3}
 {bone1Name=dyn_jacket_03_side_r, bone2Name=dyn_jacket_02_side_r, swing1LimitDegrees=15, swing2LimitDegrees=35, twistLimitDegrees=3}
 {bone1Name=dyn_jacket_04_side_r, bone2Name=dyn_jacket_03_side_r, swing1LimitDegrees=15, swing2LimitDegrees=35, twistLimitDegrees=3}
 {
  bone1Name=dyn_jacket_base_fr_l
  bone2Name=spine_01
  swing1LimitDegrees=0
  swing1Motion=Free
  swing2LimitDegrees=35
  swing2Motion=Free
  twistLimitDegrees=35
  twistMotion=Free
 }
 {bone1Name=dyn_jacket_01_fr_l, bone2Name=dyn_jacket_base_fr_l, swing1LimitDegrees=60, swing2LimitDegrees=60, twistLimitDegrees=45}
 {bone1Name=dyn_jacket_03_fr_l, bone2Name=dyn_jacket_02_fr_l, swing1LimitDegrees=45, swing2LimitDegrees=45, twistLimitDegrees=45}
 {bone1Name=dyn_jacket_02_fr_l, bone2Name=dyn_jacket_01_fr_l, swing1LimitDegrees=45, swing2LimitDegrees=45, twistLimitDegrees=45}
 {bone1Name=dyn_jacket_04_fr_l, bone2Name=dyn_jacket_03_fr_l, swing1LimitDegrees=25, swing2LimitDegrees=55, twistLimitDegrees=15}
 {bone1Name=dyn_jacket_base_fr_r, bone2Name=spine_01, swing1LimitDegrees=25, swing2LimitDegrees=45, twistLimitDegrees=15}
 {bone1Name=dyn_jacket_01_fr_r, bone2Name=dyn_jacket_base_fr_r, swing1LimitDegrees=45, swing2LimitDegrees=45, twistLimitDegrees=45}
 {bone1Name=dyn_jacket_02_fr_r, bone2Name=dyn_jacket_01_fr_r, swing1LimitDegrees=45, swing2LimitDegrees=45, twistLimitDegrees=45}
 {bone1Name=dyn_jacket_03_fr_r, bone2Name=dyn_jacket_02_fr_r, swing1LimitDegrees=45, swing2LimitDegrees=45, twistLimitDegrees=45}
 {bone1Name=dyn_jacket_04_fr_r, bone2Name=dyn_jacket_03_fr_r, swing1LimitDegrees=25, swing2LimitDegrees=25, twistLimitDegrees=15}
 {bone1Name=dyn_jacket_03_back_l, bone2Name=dyn_jacket_04_side_l, swing1LimitDegrees=45, swing2LimitDegrees=45, twistLimitDegrees=45}
 {bone1Name=dyn_jacket_03_back_r, bone2Name=dyn_jacket_04_side_r, swing1LimitDegrees=45, swing2LimitDegrees=45, twistLimitDegrees=45}
 {bone1Name=dyn_jacket_05_fr_l, bone2Name=dyn_jacket_04_fr_l, swing1LimitDegrees=10, swing2LimitDegrees=55, twistLimitDegrees=15}
 {bone1Name=dyn_jacket_05_fr_r, bone2Name=dyn_jacket_04_fr_r, swing1LimitDegrees=25, swing2LimitDegrees=35, twistLimitDegrees=15}
 {bone1Name=dyn_jacket_04_side_l, bone2Name=dyn_jacket_05_fr_l, swing1LimitDegrees=45, swing2LimitDegrees=45, twistLimitDegrees=45}
 {bone1Name=dyn_jacket_04_side_r, bone2Name=dyn_jacket_05_fr_r, swing1LimitDegrees=45, swing2LimitDegrees=45, twistLimitDegrees=45}
 {bone1Name=dyn_jacket_03_side_l, bone2Name=dyn_jacket_04_fr_l, swing1LimitDegrees=55, swing2LimitDegrees=55, twistLimitDegrees=55}
 {
  bone1Name=dyn_jacket_03_back_r
  bone2Name=dyn_jacket_03_back_l
  swing1LimitDegrees=45
  swing1Motion=Free
  swing2LimitDegrees=45
  swing2Motion=Free
  twistLimitDegrees=45
  twistMotion=Free
 }
 {bone1Name=dyn_jacket_03_side_r, bone2Name=dyn_jacket_04_fr_r, swing1LimitDegrees=45, swing2LimitDegrees=45, twistLimitDegrees=45}
 {
  bone1Name=spine_01
  bone2Name=pelvis
  swing1LimitDegrees=45
  swing1Motion=Free
  swing2LimitDegrees=45
  swing2Motion=Free
  twistLimitDegrees=45
  twistMotion=Free
 }
 {
  bone1Name=dyn_jacket_01_side_l
  bone2Name=dyn_jacket_02_fr_l
  swing1LimitDegrees=45
  swing1Motion=Free
  swing2LimitDegrees=45
  swing2Motion=Free
  twistLimitDegrees=45
  twistMotion=Free
 }
 {
  bone1Name=dyn_jacket_01_side_r
  bone2Name=dyn_jacket_02_fr_r
  swing1LimitDegrees=45
  swing1Motion=Free
  swing2LimitDegrees=45
  swing2Motion=Free
  twistLimitDegrees=45
  twistMotion=Free
 }
 {
  bone1Name=foot_r
  bone2Name=calf_r
  swing1LimitDegrees=45
  swing1Motion=Free
  swing2LimitDegrees=45
  swing2Motion=Free
  twistLimitDegrees=45
  twistMotion=Free
 }
 {
  bone1Name=foot_l
  bone2Name=calf_l
  swing1LimitDegrees=45
  swing1Motion=Free
  swing2LimitDegrees=45
  swing2Motion=Free
  twistLimitDegrees=45
  twistMotion=Free
 }
 {
  bone1Name=dyn_jacket_02_side_r
  bone2Name=dyn_jacket_03_fr_r
  swing1LimitDegrees=45
  swing2LimitDegrees=45
  twistLimitDegrees=45
  twistMotion=Locked
 }
 {
  bone1Name=dyn_jacket_02_side_l
  bone2Name=dyn_jacket_03_fr_l
  swing1LimitDegrees=45
  swing1Motion=Free
  swing2LimitDegrees=45
  swing2Motion=Free
  twistLimitDegrees=45
  twistMotion=Free
 }
 {
  bone1Name=pelvis
  bone2Name=root
  swing1LimitDegrees=45
  swing1Motion=Free
  swing2LimitDegrees=45
  swing2Motion=Free
  twistLimitDegrees=45
  twistMotion=Free
 }
 {
  bone1Name=dyn_jacket_05_fr_l
  bone2Name=dyn_jacket_04_fr_l
  swing1LimitDegrees=45
  swing1Motion=Free
  swing2LimitDegrees=45
  swing2Motion=Free
  twistLimitDegrees=45
  twistMotion=Free
 }
]
