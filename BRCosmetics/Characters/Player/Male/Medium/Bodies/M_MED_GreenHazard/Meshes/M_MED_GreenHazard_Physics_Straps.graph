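# Physics asset for the dangling back straps on the GreenHazard medium male player body.
# Left and right capsule chains hang from the back jacket bones with limited swing and twist, ending in box bodies whose last joints lock twist and one swing axis; the left chain has one extra segment.
# Jacket, pelvis and calf capsules, plus a wide floor-level root box, form the surrounding body shapes, linked by angularly free constraints.
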PhysicsAsset /BRCosmetics/Characters/Player/Male/Medium/Bodies/M_MED_GreenHazard/Meshes/M_MED_GreenHazard_Physics_Straps
bodies {
 calf_l [{shapeType=Capsule, length=35, radius=8, shapeName=calf_l_capsule, center=<-22 0 0>, rotation=<rot 90 0 0>}]
 calf_r [{shapeType=Capsule, length=35, radius=8, shapeName=calf_r_capsule, center=<22 0 0>, rotation=<rot 90 0 0>}]
 dyn_jacket_01_back_l [{
   shapeType=Capsule
   length=8
   radius=8
   shapeName=dyn_jacket_01_back_r_capsule
   center=<-7.5591 2.7136 9.1355>
   rotation=<rot 74.9999 89.9983 179.9985>
  }]
 dyn_jacket_01_back_r [{
   shapeType=Capsule
   length=8.9
   radius=8
   shapeName=dyn_jacket_01_back_l_capsule
   center=<7.5591 -4.8603 -8.5603>
   rotation=<rot 74.9999 89.9983 -0.0015>
  }]
 dyn_jacket_02_back_l [{
   shapeType=Capsule
   length=10.7
   radius=8
   shapeName=dyn_jacket_02_back_r_capsule
   center=<-6 1.9104 8.7761>
   rotation=<rot 74.9999 89.9984 179.9985>
  }]
 dyn_jacket_02_back_r [{
   shapeType=Capsule
   length=10.7
   radius=8
   shapeName=dyn_jacket_02_back_l_capsule
   center=<6 -3.817 -7.3048>
   rotation=<rot 74.9999 89.9984 -0.0015>
  }]
 dyn_jacket_03_back_l [{
   shapeType=Capsule
   length=8.7
   radius=8.9
   shapeName=dyn_jacket_03_back_r_capsule
   center=<-6.7662 3.1494 7.9159>
   rotation=<rot 82.3818 89.9966 179.9967>
  }]
 dyn_jacket_03_back_r [{
   shapeType=Capsule
   length=8.7
   radius=8.9
   shapeName=dyn_jacket_03_back_l_capsule
   center=<6.7662 -3.1494 -7.9159>
   rotation=<rot 82.3818 89.9966 -0.0033>
  }]
 dyn_jacket_04_side_l [{shapeType=Capsule, length=5.7, radius=10.5, shapeName=dyn_jacket_03_back_r_capsule, center=<-7.9335 5.0591 3.7184>, rotation=<rot 90 -90 0>}]
 dyn_jacket_04_side_r [{shapeType=Capsule, length=5.7, radius=10.5, shapeName=dyn_jacket_03_back_l_capsule, center=<7.9335 -5.0591 -3.7184>, rotation=<rot 90 90 0>}]
 dyn_strap_back_01_l [{shapeType=Capsule, length=6.9, radius=4.6, shapeName=dyn_strap_back_01_l_capsule, center=<-11.3972 0 -2.5>, rotation=<rot 90 0 0>}]
 dyn_strap_back_01_r [{shapeType=Capsule, length=6.4, radius=4.6, shapeName=dyn_strap_back_01_l_capsule, center=<11.4624 0 2.5>, rotation=<rot 90 0 0>}]
 dyn_strap_back_02_l [{shapeType=Capsule, length=4.5, radius=4.5, shapeName=dyn_strap_back_02_l_capsule, center=<-5 0 -2.5>, rotation=<rot 90 0 0>}]
 dyn_strap_back_02_r [{shapeType=Capsule, length=4.5, radius=4.5, shapeName=dyn_strap_back_02_l_capsule, center=<5 0 2.5>, rotation=<rot 90 0 0>}]
 dyn_strap_back_03_l [{shapeType=Capsule, length=2.9, radius=5.1, shapeName=dyn_strap_back_03_l_capsule, center=<-4 0 -2.5>, rotation=<rot 90 0 0>}]
 dyn_strap_back_03_r [{shapeType=Capsule, length=2.6, radius=5.4, shapeName=dyn_strap_back_03_l_capsule, center=<4 0 3.1932>, rotation=<rot 90 0 0>}]
 dyn_strap_back_04_l [{shapeType=Capsule, length=3, radius=5, shapeName=dyn_strap_back_04_l_capsule, center=<-4 0 -2.5>, rotation=<rot 90 0 0>}]
 dyn_strap_back_04_r [{shapeType=Capsule, length=2.9, radius=5.1, shapeName=dyn_strap_back_04_l_capsule, center=<4 0 3.1622>, rotation=<rot 90 0 0>}]
 dyn_strap_back_05_l [{shapeType=Capsule, length=3.7, radius=4.3, shapeName=dyn_strap_back_05_l_capsule, center=<-4 0 -2.5>, rotation=<rot 90 0 0>}]
 dyn_strap_back_05_r [{
   shapeType=Box
   extentX=6.138
   extentY=1.4079
   extentZ=7.8139
   shapeName=dyn_strap_back_05_r_box
   center=<2.4724 -0.1549 0.0293>
   rotation=<rot 0 -2.7838 -89.3012>
  }]
 dyn_strap_back_06_l [{
   shapeType=Box
   extentX=6.0324
   extentY=1.4118
   extentZ=7.8292
   shapeName=dyn_strap_back_06_l_box
   center=<-2.1218 0.0068 -0.0605>
   rotation=<rot 0 179.2647 -88.4139>
  }]
 dyn_strap_back_base_l [{shapeType=Capsule, length=0, radius=0, shapeName=dyn_strap_back_base_l_capsule, center=<-2 0 0>, rotation=<rot 90 0 0>}]
 dyn_strap_back_base_r [{shapeType=Capsule, length=0.505, radius=0.505, shapeName=dyn_strap_back_base_r_capsule, center=<2 0 0>, rotation=<rot 90 -90 0>}]
 pelvis [{shapeType=Capsule, length=18.3, radius=9.7, shapeName=pelvis_capsule, center=<0 1.0086 0>}]
 root [{shapeType=Box, extentX=153.61, extentY=166.21, extentZ=35.01, shapeName=root_box, center=<0 0 -17.1594>, rotation=<rot 0 -90 -0.1526>}]
}
constraints [
 {
  bone1Name=dyn_strap_back_06_l
  bone2Name=dyn_strap_back_05_l
  swing1LimitDegrees=25
  swing1Motion=Locked
  swing2LimitDegrees=80
  twistLimitDegrees=5
  twistMotion=Locked
 }
 {bone1Name=dyn_strap_back_05_l, bone2Name=dyn_strap_back_04_l, swing1LimitDegrees=25, swing2LimitDegrees=35, twistLimitDegrees=3}
 {bone1Name=dyn_strap_back_04_l, bone2Name=dyn_strap_back_03_l, swing1LimitDegrees=25, swing2LimitDegrees=35, twistLimitDegrees=3}
 {bone1Name=dyn_strap_back_03_l, bone2Name=dyn_strap_back_02_l, swing1LimitDegrees=25, swing2LimitDegrees=35, twistLimitDegrees=3}
 {bone1Name=dyn_strap_back_02_l, bone2Name=dyn_strap_back_01_l, swing1LimitDegrees=15, swing2LimitDegrees=35, twistLimitDegrees=3}
 {bone1Name=dyn_strap_back_01_l, bone2Name=dyn_strap_back_base_l, swing1LimitDegrees=10, swing2LimitDegrees=45, twistLimitDegrees=3}
 {bone1Name=dyn_strap_back_01_r, bone2Name=dyn_strap_back_base_r, swing1LimitDegrees=10, swing2LimitDegrees=45, twistLimitDegrees=3}
 {bone1Name=dyn_strap_back_02_r, bone2Name=dyn_strap_back_01_r, swing1LimitDegrees=15, swing2LimitDegrees=35, twistLimitDegrees=3}
 {bone1Name=dyn_strap_back_03_r, bone2Name=dyn_strap_back_02_r, swing1LimitDegrees=25, swing2LimitDegrees=35, twistLimitDegrees=3}
 {bone1Name=dyn_strap_back_04_r, bone2Name=dyn_strap_back_03_r, swing1LimitDegrees=25, swing2LimitDegrees=35, twistLimitDegrees=3}
 {
  bone1Name=dyn_strap_back_05_r
  bone2Name=dyn_strap_back_04_r
  swing1LimitDegrees=25
  swing1Motion=Locked
  swing2LimitDegrees=80
  twistLimitDegrees=5
  twistMotion=Locked
 }
 {
  bone1Name=dyn_strap_back_base_l
  bone2Name=dyn_jacket_01_back_l
  swing1LimitDegrees=45
  swing1Motion=Free
  swing2LimitDegrees=45
  swing2Motion=Free
  twistLimitDegrees=45
  twistMotion=Free
 }
 {
  bone1Name=dyn_strap_back_base_r
  bone2Name=dyn_jacket_01_back_r
  swing1LimitDegrees=45
  swing1Motion=Free
  swing2LimitDegrees=45
  swing2Motion=Free
  twistLimitDegrees=45
  twistMotion=Free
 }
 {
  bone1Name=calf_l
  bone2Name=pelvis
  swing1LimitDegrees=45
  swing1Motion=Free
  swing2LimitDegrees=45
  swing2Motion=Free
  twistLimitDegrees=45
  twistMotion=Free
 }
 {
  bone1Name=calf_r
  bone2Name=pelvis
  swing1LimitDegrees=45
  swing1Motion=Free
  swing2LimitDegrees=45
  swing2Motion=Free
  twistLimitDegrees=45
  twistMotion=Free
 }
 {
  bone1Name=pelvis
  bone2Name=root
  swing1LimitDegrees=45
  swing1Motion=Free
  swing2LimitDegrees=45
  swing2Motion=Free
  twistLimitDegrees=45
  twistMotion=Free
 }
 {
  bone1Name=dyn_jacket_03_back_r
  bone2Name=dyn_jacket_02_back_r
  swing1LimitDegrees=45
  swing1Motion=Free
  swing2LimitDegrees=45
  swing2Motion=Free
  twistLimitDegrees=45
  twistMotion=Free
 }
 {
  bone1Name=dyn_jacket_02_back_r
  bone2Name=dyn_jacket_01_back_r
  swing1LimitDegrees=45
  swing1Motion=Free
  swing2LimitDegrees=45
  swing2Motion=Free
  twistLimitDegrees=45
  twistMotion=Free
 }
 {
  bone1Name=dyn_jacket_01_back_r
  bone2Name=pelvis
  swing1LimitDegrees=45
  swing1Motion=Free
  swing2LimitDegrees=45
  swing2Motion=Free
  twistLimitDegrees=45
  twistMotion=Free
 }
 {
  bone1Name=dyn_jacket_03_back_l
  bone2Name=dyn_jacket_02_back_l
  swing1LimitDegrees=45
  swing1Motion=Free
  swing2LimitDegrees=45
  swing2Motion=Free
  twistLimitDegrees=45
  twistMotion=Free
 }
 {
  bone1Name=dyn_jacket_02_back_l
  bone2Name=dyn_jacket_01_back_l
  swing1LimitDegrees=45
  swing1Motion=Free
  swing2LimitDegrees=45
  swing2Motion=Free
  twistLimitDegrees=45
  twistMotion=Free
 }
 {
  bone1Name=dyn_jacket_01_back_l
  bone2Name=pelvis
  swing1LimitDegrees=45
  swing1Motion=Free
  swing2LimitDegrees=45
  swing2Motion=Free
  twistLimitDegrees=45
  twistMotion=Free
 }
 {
  bone1Name=dyn_jacket_04_side_r
  bone2Name=pelvis
  swing1LimitDegrees=45
  swing1Motion=Free
  swing2LimitDegrees=45
  swing2Motion=Free
  twistLimitDegrees=45
  twistMotion=Free
 }
 {
  bone1Name=dyn_jacket_04_side_l
  bone2Name=pelvis
  swing1LimitDegrees=45
  swing1Motion=Free
  swing2LimitDegrees=45
  swing2Motion=Free
  twistLimitDegrees=45
  twistMotion=Free
 }
]
