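# Animation Blueprint for the GreenHazard male body cosmetic, adding secondary motion to coat, straps, collar and badge.
# The copied parent pose runs through Control Rig, coat and strap RigidBody nodes (strap tuning switches on cape back bling), then strap and collar AnimDynamics.
# RBAN_On selects between jacket-driven and body-driven badge setups before an additive backpack-attach offset.
# EventGraph updates rigid-body, cape, LOD and gravity state, though GravityOverride feeds no anim node.
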AnimBlueprint /BRCosmetics/Characters/Player/Male/Medium/Bodies/M_MED_GreenHazard/Meshes/M_MED_GreenHazard_AnimBP
target_skeleton=&/M_MED_GreenHazard_Skeleton
graphs [
 AnimGraph {nodes [
   "dyn_badge_base.rotateZ = thigh_r.rotateY * -1 - Bone Driven Controller"!(ComponentPose:Component Space Pose Link Structure, Alpha:float=0.5) -> (Pose:Component Space Pose Link Structure) {
    is=title AnimGraphNode_BoneDrivenController
    bAffectTargetRotationZ=true
    multiplier=-1
    sourceComponent=RotationY
    sourceBone {boneName=thigh_r}
    targetBone {boneName=dyn_badge_base}
   }
   n1 = "Constraint - dyn_badge_base (PPPPPPPP)"!(
    ComponentPose=Local To Component(LocalPose=Use cached pose 'PreSim2'())
    Alpha=0
    ConstraintWeights_0=0.0001
    ConstraintWeights_1:float=0.0082
    ConstraintWeights_2:float=0.0026
    ConstraintWeights_3:float=0.1121
    ConstraintWeights_4:float=0.0036
    ConstraintWeights_5:float=0.0643
    ConstraintWeights_6:float=0.768
    ConstraintWeights_7:float=0.041
   ) -> (Pose) {
    boneToModify {boneName=dyn_badge_base}
    constraintSetup [
     {targetBone {boneName=dyn_jacket_02_back_r}}
     {targetBone {boneName=dyn_jacket_01_back_r}}
     {targetBone {boneName=dyn_jacket_03_side_r}}
     {targetBone {boneName=dyn_jacket_02_side_r}}
     {targetBone {boneName=dyn_jacket_01_side_r}}
     {targetBone {boneName=dyn_jacket_04_fr_r}}
     {targetBone {boneName=dyn_jacket_02_fr_r}}
     {targetBone {boneName=dyn_jacket_01_fr_r}}
    ]
    constraintWeights [1,1,1,1,1,1,1,1]
   }
   n2 = "dyn_badge_base.rotateY = dyn_jacket_03_fr_r.rotateY - Bone Driven Controller"!(ComponentPose:Component Space Pose Link Structure=@n1, Alpha:float=0.8) -> (Pose:Component Space Pose Link Structure) {
    is=title AnimGraphNode_BoneDrivenController
    bAffectTargetRotationY=true
    sourceComponent=RotationY
    sourceBone {boneName=dyn_jacket_03_fr_r}
    targetBone {boneName=dyn_badge_base}
   }
   n3 = Anim Dynamics - Bone: dyn_badge_01!(ComponentPose=@n2, Alpha=1, GravityOverride:Vector="0,0,-980", ExternalForce="0,0,0") -> (Pose) {
    angularDampingOverride=1
    angularSpringConstant=50
    bAngularSpring=true
    bOverrideAngularDamping=true
    bUseGravityOverride=true
    bUsePlanarLimit=false
    simulationSpace=RootRelative
    alphaBoolBlend {blendOption=Linear}
    boundBone {boneName=dyn_badge_01}
    componentAppliedLinearAccClamp=<5000 5000 5000>
    componentLinearAccScale=<0.5 0.5 0.5>
    componentLinearVelScale=<0.5 0.5 0.5>
    physicsBodyDefinitions [{
      boundBone {boneName=dyn_badge_01}
      boxExtents=<10 10 10>
      constraintSetup {angularLimitsMax=<1 3 7>, angularLimitsMin=<-1 -35 -7>, angularTarget=<1 0 0>}
      localJointOffset=<1 0 0>
     }]
    sphericalLimits [{limitRadius=11, drivingBone {boneName=dyn_jacket_04_fr_r}, sphereLocalOffset=<0 0 -8>}]
   }
   n4 = Component To Local(ComponentPose=@n3)
   n5 = "Constraint - dyn_badge_base (PPPPPPPP)"(
    ComponentPose=Local To Component(LocalPose=Use cached pose 'PreSim2'())
    Alpha=1
    ConstraintWeights_0=0.8507
    ConstraintWeights_1=0.0728
    ConstraintWeights_2=0.0398
    ConstraintWeights_3=0.0192
    ConstraintWeights_4=0.0001
    ConstraintWeights_5=0.0007
    ConstraintWeights_6=0.009
    ConstraintWeights_7=0.0077
   ) {
    boneToModify {boneName=dyn_badge_base}
    constraintSetup [
     {targetBone {boneName=pelvis}}
     {targetBone {boneName=spine_01}}
     {targetBone {boneName=spine_02}}
     {targetBone {boneName=spine_03}}
     {targetBone {boneName=thigh_twist_01_l}}
     {targetBone {boneName=thigh_r}}
     {targetBone {boneName=thigh_twist_01_r}}
     {targetBone {boneName=deform_groin_r}}
    ]
    constraintWeights [1,1,1,1,1,1,1,1]
   }
   n6 = "dyn_badge_base.rotateY = thigh_r.rotateZ * -1 - Bone Driven Controller"!(ComponentPose:Component Space Pose Link Structure=@n5, Alpha:float=0.85) -> (Pose:Component Space Pose Link Structure) {
    is=title AnimGraphNode_BoneDrivenController
    bAffectTargetRotationY=true
    multiplier=-1
    sourceComponent=RotationZ
    sourceBone {boneName=thigh_r}
    targetBone {boneName=dyn_badge_base}
   }
   n7 = "dyn_badge_base.rotateX = thigh_r.rotateX - Bone Driven Controller"!(ComponentPose:Component Space Pose Link Structure=@n6, Alpha:float=0.2) -> (Pose:Component Space Pose Link Structure) {
    is=title AnimGraphNode_BoneDrivenController
    bAffectTargetRotationX=true
    sourceComponent=RotationX
    sourceBone {boneName=thigh_r}
    targetBone {boneName=dyn_badge_base}
   }
   n8 = Anim Dynamics - Bone: dyn_badge_01(ComponentPose=@n7, Alpha=1, GravityOverride="0,0,-980", ExternalForce="0,0,0") {
    angularDampingOverride=1
    angularSpringConstant=50
    bAngularSpring=true
    bOverrideAngularDamping=true
    bUseGravityOverride=true
    bUsePlanarLimit=false
    simulationSpace=RootRelative
    alphaBoolBlend {blendOption=Linear}
    boundBone {boneName=dyn_badge_01}
    componentAppliedLinearAccClamp=<5000 5000 5000>
    componentLinearAccScale=<0.5 0.5 0.5>
    componentLinearVelScale=<0.5 0.5 0.5>
    physicsBodyDefinitions [{
      boundBone {boneName=dyn_badge_01}
      boxExtents=<10 10 10>
      constraintSetup {angularLimitsMax=<1 2 3>, angularLimitsMin=<-1 -25 -3>, angularTarget=<1 0 0>}
      localJointOffset=<1 0 0>
     }]
   }
   n9 = Component To Local(ComponentPose=@n8)
   n10 = Get RBAN_On() -> (RBAN_On:bool)
   n11 = "Transform (Modify) Bone - Bone: attach_backpack"!(
    ComponentPose=Local To Component(LocalPose=Blend Poses by bool(BlendPose_0=@n4, BlendPose_1=@n9, bActiveValue=@n10))
    bAlphaBoolEnabled:bool=True
    Translation="0,1.5,1"
    Rotation="0,0,3"
    Scale="1,1,1"
   ) -> (Pose) {
    alphaInputType=Bool
    rotationMode=BMM_Additive
    rotationSpace=BCS_BoneSpace
    translationMode=BMM_Additive
    translationSpace=BCS_BoneSpace
    boneToModify {boneName=attach_backpack}
    rotation {roll=3}
    translation {y=1.5, z=1}
   }
   Output Pose(Result=Component To Local(ComponentPose=@n11)) {layerGroup=None}
   Save cached pose 'PreSim'(Pose=Copy Pose From Mesh() {bUseAttachedParent=true}) {cacheName=PreSim}
   n12 = Control Rig!(Source=Use cached pose 'PreSim'(), Alpha, LOD1:bool=Get LOD1() -> (LOD1:bool), RBAN Off:bool=NOT Boolean(A=Get RBAN_On())) -> (Pose) {controlRigAssetReference {blueprintRigClass=&/M_MED_GreenHazard_CtrlRig.M_MED_GreenHazard_CtrlRig_C}}
   n13 = RigidBody!(ComponentPose=Local To Component(LocalPose=@n12), Alpha, OverrideWorldGravity:Vector="0,0,0", ExternalForce) -> (Pose) {
    bClampLinearTranslationLimitToRefPose=true
    bOverrideWorldGravity=true
    overridePhysicsAsset=&/M_MED_GreenHazard_Physics_Coat
    simulationSpace=BaseBoneSpace
    alphaBoolBlend {blendOption=Linear}
    baseBoneRef {boneName=spine_03}
    simSpaceSettings {maxAngularAcceleration=25, maxAngularVelocity=3, maxLinearAcceleration=100, maxLinearVelocity=75, velocityScaleZ=0, worldAlpha=1}
   }
   n14 = Get Wear_cape() -> (Wear_cape:bool)
   n15 = RigidBody!(ComponentPose=@n13, bAlphaBoolEnabled:bool=@n14, OverrideWorldGravity="0,0,0", ExternalForce) -> (Pose) {
    alphaInputType=Bool
    bClampLinearTranslationLimitToRefPose=true
    bOverrideWorldGravity=true
    overridePhysicsAsset=&/M_MED_GreenHazard_Physics_Straps
    simulationSpace=BaseBoneSpace
    alphaBoolBlend {blendOption=Linear}
    baseBoneRef {boneName=spine_02}
    simSpaceSettings {
     maxAngularAcceleration=30
     maxAngularVelocity=3
     maxLinearAcceleration=50
     maxLinearVelocity=50
     worldAlpha=0.45
     externalLinearDragV=<1 0.2 0.2>
    }
   }
   n16 = RigidBody(ComponentPose=@n15, bAlphaBoolEnabled=NOT Boolean(A=@n14), OverrideWorldGravity="0,0,0") {
    alphaInputType=Bool
    bClampLinearTranslationLimitToRefPose=true
    bOverrideWorldGravity=true
    overridePhysicsAsset=&M_MED_GreenHazard_Physics_Straps
    simulationSpace=BaseBoneSpace
    alphaBoolBlend {blendOption=Linear}
    baseBoneRef {boneName=spine_02}
    simSpaceSettings {
     maxAngularAcceleration=40
     maxAngularVelocity=4
     maxLinearAcceleration=95
     maxLinearVelocity=250
     velocityScaleZ=2
     worldAlpha=1
     externalLinearDragV=<2 0.5 0.5>
    }
   }
   n17 = Anim Dynamics - Bone: dyn_strap_side_01_r!(ComponentPose=@n16, Alpha=1, GravityOverride:Vector="0,0,-980", ExternalForce="0,0,0") -> (Pose) {
    angularBiasOverride=1
    angularDampingOverride=1
    angularSpringConstant=50
    bAngularSpring=true
    bLinearSpring=true
    bOverrideAngularBias=true
    bOverrideAngularDamping=true
    bUseGravityOverride=true
    bUsePlanarLimit=false
    linearSpringConstant=100
    simulationSpace=RootRelative
    alphaBoolBlend {blendOption=Linear}
    boundBone {boneName=dyn_strap_side_01_r}
    componentAppliedLinearAccClamp=<1000 1000 1000>
    componentLinearAccScale=<0.5 0.5 0.5>
    componentLinearVelScale=<0.5 0.5 0.5>
    physicsBodyDefinitions [{
      boundBone {boneName=dyn_strap_side_01_r}
      boxExtents=<10 10 10>
      constraintSetup {angularLimitsMax=<5 5 5>, angularLimitsMin=<-5 -5 -5>, angularTarget=<1 0 0>, linearAxesMax=<1.5 0 3>, linearAxesMin=<-2.5 0 0>}
      localJointOffset=<0.1 0 0>
     }]
    simSpaceSettings {maxAngularAcceleration=1000, maxAngularVelocity=1000}
   }
   n18 = Anim Dynamics - Bone: dyn_strap_side_02_r!(ComponentPose=@n17, Alpha=1, GravityOverride:Vector="0,0,-980", ExternalForce="0,0,0") -> (Pose) {
    angularBiasOverride=1
    angularDampingOverride=1
    angularSpringConstant=50
    bAngularSpring=true
    bLinearSpring=true
    bOverrideAngularBias=true
    bOverrideAngularDamping=true
    bUseGravityOverride=true
    bUsePlanarLimit=false
    linearSpringConstant=100
    simulationSpace=RootRelative
    alphaBoolBlend {blendOption=Linear}
    boundBone {boneName=dyn_strap_side_02_r}
    componentAppliedLinearAccClamp=<1000 1000 1000>
    componentLinearAccScale=<0.5 0.5 0.5>
    componentLinearVelScale=<0.5 0.5 0.5>
    physicsBodyDefinitions [{
      boundBone {boneName=dyn_strap_side_02_r}
      boxExtents=<10 10 10>
      constraintSetup {angularLimitsMax=<5 5 5>, angularLimitsMin=<-5 -5 -5>, angularTarget=<1 0 0>, linearAxesMax=<2 0 2>, linearAxesMin=<-3 0 0>}
      localJointOffset=<0.1 0 0>
     }]
    simSpaceSettings {maxAngularAcceleration=1000, maxAngularVelocity=1000}
   }
   n19 = Anim Dynamics - Bone: dyn_strap_side_01_l!(ComponentPose=@n18, Alpha=1, GravityOverride:Vector="0,0,-980", ExternalForce="0,0,0") -> (Pose) {
    angularBiasOverride=1
    angularDampingOverride=1
    angularSpringConstant=50
    bAngularSpring=true
    bLinearSpring=true
    bOverrideAngularBias=true
    bOverrideAngularDamping=true
    bUseGravityOverride=true
    bUsePlanarLimit=false
    linearSpringConstant=100
    simulationSpace=RootRelative
    alphaBoolBlend {blendOption=Linear}
    boundBone {boneName=dyn_strap_side_01_l}
    componentAppliedLinearAccClamp=<1000 1000 1000>
    componentLinearAccScale=<0.5 0.5 0.5>
    componentLinearVelScale=<0.5 0.5 0.5>
    physicsBodyDefinitions [{
      boundBone {boneName=dyn_strap_side_01_l}
      boxExtents=<10 10 10>
      constraintSetup {angularLimitsMax=<5 5 5>, angularLimitsMin=<-5 -5 -5>, angularTarget=<1 0 0>, linearAxesMax=<1.5 0 3>, linearAxesMin=<-2.5 0 0>}
      localJointOffset=<0.1 0 0>
     }]
    simSpaceSettings {maxAngularAcceleration=1000, maxAngularVelocity=1000}
   }
   n20 = Anim Dynamics - Bone: dyn_strap_side_02_l!(ComponentPose=@n19, Alpha=1, GravityOverride:Vector="0,0,-980", ExternalForce="0,0,0") -> (Pose) {
    angularBiasOverride=1
    angularDampingOverride=1
    angularSpringConstant=50
    bAngularSpring=true
    bLinearSpring=true
    bOverrideAngularBias=true
    bOverrideAngularDamping=true
    bUseGravityOverride=true
    bUsePlanarLimit=false
    linearSpringConstant=100
    simulationSpace=RootRelative
    alphaBoolBlend {blendOption=Linear}
    boundBone {boneName=dyn_strap_side_02_l}
    componentAppliedLinearAccClamp=<1000 1000 1000>
    componentLinearAccScale=<0.5 0.5 0.5>
    componentLinearVelScale=<0.5 0.5 0.5>
    physicsBodyDefinitions [{
      boundBone {boneName=dyn_strap_side_02_l}
      boxExtents=<10 10 10>
      constraintSetup {angularLimitsMax=<5 5 5>, angularLimitsMin=<-5 -5 -5>, angularTarget=<1 0 0>, linearAxesMax=<2 0 2>, linearAxesMin=<-3 0 0>}
      localJointOffset=<0.1 0 0>
     }]
    simSpaceSettings {maxAngularAcceleration=1000, maxAngularVelocity=1000}
   }
   n21 = Get Collar Dynamic() -> (Collar Dynamic:double)
   n22 = Anim Dynamics - Bone: dyn_collar_fr_01_r!(ComponentPose=@n20, Alpha=@n21, GravityOverride:Vector="0,0,0", ExternalForce="0,0,0") -> (Pose) {
    angularDampingOverride=1
    angularSpringConstant=1000
    bAngularSpring=true
    bOverrideAngularDamping=true
    bUseGravityOverride=true
    simulationSpace=RootRelative
    alphaBoolBlend {blendOption=Linear}
    boundBone {boneName=dyn_collar_fr_01_r}
    componentAppliedLinearAccClamp=<5000 5000 5000>
    componentLinearAccScale=<0.05 0.05 0.05>
    componentLinearVelScale=<0.05 0.05 0.05>
    physicsBodyDefinitions [{
      boundBone {boneName=dyn_collar_fr_01_r}
      boxExtents=<10 10 10>
      constraintSetup {angularLimitsMax=<3 25 3>, angularLimitsMin=<-3 -1 -3>, angularTarget=<1 0 0>}
      localJointOffset=<-0.25 0 0>
     }]
    simSpaceSettings {maxAngularAcceleration=100, maxAngularVelocity=100, simSpaceAngularAlpha=0.2}
   }
   n23 = Anim Dynamics - Bone: dyn_collar_side_01_r!(ComponentPose=@n22, Alpha=@n21, GravityOverride:Vector="0,0,0", ExternalForce="0,0,0") -> (Pose) {
    angularDampingOverride=1
    angularSpringConstant=1000
    bAngularSpring=true
    bOverrideAngularDamping=true
    bUseGravityOverride=true
    simulationSpace=RootRelative
    alphaBoolBlend {blendOption=Linear}
    boundBone {boneName=dyn_collar_side_01_r}
    componentAppliedLinearAccClamp=<5000 5000 5000>
    componentLinearAccScale=<0.05 0.05 0.05>
    componentLinearVelScale=<0.05 0.05 0.05>
    physicsBodyDefinitions [{
      boundBone {boneName=dyn_collar_side_01_r}
      boxExtents=<10 10 10>
      constraintSetup {angularLimitsMax=<3 10 3>, angularLimitsMin=<-3 -1 -3>, angularTarget=<1 0 0>}
      localJointOffset=<-0.1 0 0>
     }]
    simSpaceSettings {maxAngularAcceleration=100, maxAngularVelocity=100, simSpaceAngularAlpha=0.2}
   }
   n24 = Anim Dynamics - Bone: dyn_collar_back_01!(ComponentPose=@n23, Alpha=@n21, GravityOverride:Vector="0,0,0", ExternalForce="0,0,0") -> (Pose) {
    angularDampingOverride=1
    angularSpringConstant=1000
    bOverrideAngularDamping=true
    bUseGravityOverride=true
    simulationSpace=RootRelative
    alphaBoolBlend {blendOption=Linear}
    boundBone {boneName=dyn_collar_back_01}
    componentAppliedLinearAccClamp=<5000 5000 5000>
    componentLinearAccScale=<0.05 0.05 0.05>
    componentLinearVelScale=<0.05 0.05 0.05>
    physicsBodyDefinitions [{
      boundBone {boneName=dyn_collar_back_01}
      boxExtents=<10 10 10>
      constraintSetup {angularLimitsMax=<2 9 2>, angularLimitsMin=<-2 0 -2>, angularTarget=<1 0 0>}
      localJointOffset=<-0.1 0 0>
     }]
    simSpaceSettings {maxAngularAcceleration=100, maxAngularVelocity=100, simSpaceAngularAlpha=0.2}
   }
   n25 = Anim Dynamics - Bone: dyn_collar_fr_01_l!(ComponentPose=@n24, Alpha=@n21, GravityOverride:Vector="0,0,0", ExternalForce="0,0,0") -> (Pose) {
    angularDampingOverride=1
    angularSpringConstant=1000
    bAngularSpring=true
    bOverrideAngularDamping=true
    bUseGravityOverride=true
    simulationSpace=RootRelative
    alphaBoolBlend {blendOption=Linear}
    boundBone {boneName=dyn_collar_fr_01_l}
    componentAppliedLinearAccClamp=<5000 5000 5000>
    componentLinearAccScale=<0.05 0.05 0.05>
    componentLinearVelScale=<0.05 0.05 0.05>
    physicsBodyDefinitions [{
      boundBone {boneName=dyn_collar_fr_01_l}
      boxExtents=<10 10 10>
      constraintSetup {angularLimitsMax=<3 25 3>, angularLimitsMin=<-3 -1 -3>, angularTarget=<1 0 0>}
      localJointOffset=<0.25 0 0>
     }]
    simSpaceSettings {maxAngularAcceleration=100, maxAngularVelocity=100, simSpaceAngularAlpha=0.2}
   }
   n26 = Anim Dynamics - Bone: dyn_collar_side_01_l!(ComponentPose=@n25, Alpha=@n21, GravityOverride:Vector="0,0,0", ExternalForce="0,0,0") -> (Pose) {
    angularDampingOverride=1
    angularSpringConstant=1000
    bAngularSpring=true
    bOverrideAngularDamping=true
    bUseGravityOverride=true
    simulationSpace=RootRelative
    alphaBoolBlend {blendOption=Linear}
    boundBone {boneName=dyn_collar_side_01_l}
    componentAppliedLinearAccClamp=<5000 5000 5000>
    componentLinearAccScale=<0.05 0.05 0.05>
    componentLinearVelScale=<0.05 0.05 0.05>
    physicsBodyDefinitions [{
      boundBone {boneName=dyn_collar_side_01_l}
      boxExtents=<10 10 10>
      constraintSetup {angularLimitsMax=<3 10 3>, angularLimitsMin=<-3 -1 -3>, angularTarget=<1 0 0>}
      localJointOffset=<0.1 0 0>
     }]
    simSpaceSettings {maxAngularAcceleration=100, maxAngularVelocity=100, simSpaceAngularAlpha=0.2}
   }
   Save cached pose 'PreSim2'(Pose=Component To Local(ComponentPose=@n26)) {cacheName=PreSim2}
  ]}
 EventGraph {nodes [
   Event BlueprintUpdateAnimation()
   > n1 = Sequence!(execute) -> (then_0, then_1, then_2:Exec, then_3:Exec, then_4:Exec, then_5:Exec, then_6:Exec)
   > n2 = Gravity Override State Function DEPRECATED(
    CharacterPartAnimInstance=Self-Reference()
    GameTimeInSeconds=GetGameTimeInSeconds()
    GravityOverrideParams=Get Gravity Override Params()
    RBAN Params=Get RBAN Params()
   )
   > Set GravityOverride(GravityOverride=@n2)
   Set Strap Dynamic(execute=@n1.then_1, Strap Dynamic:double=0.5) -> (Output_Get:double)
   Set Collar Dynamic(execute=@n1.then_2, Collar Dynamic:double=0.7) -> (Output_Get:double)
   Set Wear_cape(execute=@n1.then_3, Wear_cape:bool=CheckForBackbling(self=Self-Reference()).bIsBackblingACape) -> (Output_Get:bool)
   Set RBAN_On(execute=@n1.then_4, RBAN_On:bool=Get bRigidBodyEnabled()) -> (Output_Get:bool)
   n3 = Get CurrentLOD()
   Set LOD2(execute=@n1.then_5, LOD2:bool="integer >= integer"(A=@n3, B=2)) -> (Output_Get:bool)
   Set LOD1(execute=@n1.then_6, LOD1:bool="integer >= integer"(A=@n3, B=1)) -> (Output_Get:bool)
   n4 = Event OnCharacterCustomizationLoaded()
   > n5 = For Each Loop(Array:Array of Custom Character Part Object References=@n4.CustomParts) -> (Array Element:Custom Character Part Object Reference)
   > Branch(Condition="Equal (Enum)"(
     A:EFortCustomPartType Enum=GetPartType!(self:Custom Character Part Object Reference=@n5) -> (ReturnValue:EFortCustomPartType Enum=Head)
     B:EFortCustomPartType Enum=Backpack
    ))
   > n6 = Branch(Condition=Contains(SearchIn=GetObjectName(Object=@n5), Substring=Greenhazard))
   > Set IsThisMyBackBling(IsThisMyBackBling:bool=true) -> (Output_Get:bool)
   Set IsThisMyBackBling(execute=@n6.else, IsThisMyBackBling=false)
   Event BlueprintInitializeAnimation()
   > Set SkelMesh(SkelMesh=GetOwnerSkeletalMeshComponent())
   n7 = TryGetPawnOwner()
   > Is Valid(InputObject=@n7)
   > n8 = Cast To FortPlayerPawnAthena(Object=@n7)
   > Set PlayerPawn(PlayerPawn:Fort Player Pawn Athena Object Reference=@n8) -> (Output_Get:Fort Player Pawn Athena Object Reference)
  ]}
]
variables [
 Collar Dynamic {type=number}
 Gravity Override Params {
  type=GravityOverrideParamsStruct
  default {
   dBNO {jointName=root, gravityOverride=<0 -450 -980>}
   skydive_DivingDown {windAmplitude=<50 100 100>}
   skydive_FloatingDown {gravityOverride=<0 -200 300>, windAmplitude=<150 -100 150>, windFrequency=<1.5 10 12>, windOffset=<0 -500 70>}
   skydive_GoingUp {gravityOverride=<0 0 -400>, windAmplitude=<50 -50 50>, windOffset=<0 -600 50>}
   skydive_Parachuting {gravityOverride=<0 -500 -980>, windAmplitude=<100 -100 100>}
  }
 }
 GravityOverride {type=Vector}
 IsThisMyBackBling {type=boolean}
 LOD1 {type=boolean}
 LOD2 {type=boolean}
 PlayerPawn {type=FortPlayerPawnAthena}
 RBAN Params {
  type=RBANParamsStruct
  default {
   dBNO {simSpaceSettings {
     maxAngularAcceleration=1000
     maxAngularVelocity=3
     maxLinearAcceleration=300
     maxLinearVelocity=700
     velocityScaleZ=0
     externalLinearDragV=<0 0.5 0>
    }}
   dBNO_Carried {simSpaceSettings {
     maxAngularAcceleration=1000
     maxAngularVelocity=3
     maxLinearAcceleration=300
     maxLinearVelocity=700
     velocityScaleZ=0
     externalLinearDragV=<0 0.5 0>
    }}
   emoteOrMelee {simSpaceSettings {
     maxAngularAcceleration=1000
     maxAngularVelocity=3
     maxLinearAcceleration=300
     maxLinearVelocity=700
     velocityScaleZ=0
     externalLinearDragV=<0 0.5 0>
    }}
   front_End {simSpaceSettings {
     maxAngularAcceleration=1000
     maxAngularVelocity=3
     maxLinearAcceleration=300
     maxLinearVelocity=700
     velocityScaleZ=0
     externalLinearDragV=<0 0.5 0>
    }}
   onGround_Crouch_Moving {simSpaceSettings {
     maxAngularAcceleration=1000
     maxAngularVelocity=3
     maxLinearAcceleration=300
     maxLinearVelocity=700
     velocityScaleZ=0
     externalLinearDragV=<0 0.5 0>
    }}
   onGround_Falling {simSpaceSettings {
     maxAngularAcceleration=1000
     maxAngularVelocity=3
     maxLinearAcceleration=300
     maxLinearVelocity=700
     velocityScaleZ=0
     externalLinearDragV=<0 0.5 0>
    }}
   onGround_Moving {simSpaceSettings {
     maxAngularAcceleration=1000
     maxAngularVelocity=3
     maxLinearAcceleration=300
     maxLinearVelocity=700
     velocityScaleZ=0
     externalLinearDragV=<0 0.5 0>
    }}
   onGround_Standing {simSpaceSettings {
     maxAngularAcceleration=1000
     maxAngularVelocity=3
     maxLinearAcceleration=300
     maxLinearVelocity=700
     velocityScaleZ=0
     externalLinearDragV=<0 0.5 0>
    }}
   rail_Grinding {simSpaceSettings {
     maxAngularAcceleration=1000
     maxAngularVelocity=3
     maxLinearAcceleration=300
     maxLinearVelocity=700
     velocityScaleZ=0
     worldAlpha=0.1
     externalLinearDragV=<0 0.5 0>
    }}
   riding_Moving {simSpaceSettings {
     maxAngularAcceleration=1000
     maxAngularVelocity=3
     maxLinearAcceleration=300
     maxLinearVelocity=700
     velocityScaleZ=0
     externalLinearDragV=<0 0.5 0>
    }}
   riding_Moving_Targeting {simSpaceSettings {
     maxAngularAcceleration=1000
     maxAngularVelocity=3
     maxLinearAcceleration=300
     maxLinearVelocity=700
     velocityScaleZ=0
     externalLinearDragV=<0 0.5 0>
    }}
   riding_Standing {simSpaceSettings {
     maxAngularAcceleration=1000
     maxAngularVelocity=3
     maxLinearAcceleration=300
     maxLinearVelocity=700
     velocityScaleZ=0
     externalLinearDragV=<0 0.5 0>
    }}
   skydive_DivingDown {simSpaceSettings {
     maxAngularAcceleration=1000
     maxAngularVelocity=3
     maxLinearAcceleration=300
     maxLinearVelocity=700
     velocityScaleZ=0
     externalLinearDragV=<0 0.5 0>
    }}
   skydive_FloatingDown {
    componentLinearAccScale=<0 0.5 0>
    componentLinearVelScale=<0 0.5 0>
    simSpaceSettings {
     maxAngularAcceleration=1000
     maxAngularVelocity=3
     maxLinearAcceleration=300
     maxLinearVelocity=700
     velocityScaleZ=0
     externalLinearDragV=<0 0.5 0>
    }
   }
   skydive_GoingUp {simSpaceSettings {
     maxAngularAcceleration=1000
     maxAngularVelocity=3
     maxLinearAcceleration=300
     maxLinearVelocity=700
     velocityScaleZ=0
     externalLinearDragV=<0 0.5 0>
    }}
   skydive_Parachuting {
    componentLinearAccScale=<0 0.1 0.3>
    componentLinearVelScale=<0 0.1 0.3>
    simSpaceSettings {
     maxAngularAcceleration=1000
     maxAngularVelocity=3
     maxLinearAcceleration=300
     maxLinearVelocity=700
     velocityScaleZ=0
     externalLinearDragV=<0 0.5 0>
    }
   }
   skydive_Surfing {simSpaceSettings {
     maxAngularAcceleration=1000
     maxAngularVelocity=3
     maxLinearAcceleration=300
     maxLinearVelocity=700
     velocityScaleZ=0
     externalLinearDragV=<0 0.5 0>
    }}
   swimming {simSpaceSettings {
     maxAngularAcceleration=1000
     maxAngularVelocity=3
     maxLinearAcceleration=300
     maxLinearVelocity=700
     velocityScaleZ=0
     externalLinearDragV=<0 0.5 0>
    }}
   tactical_Sprint {simSpaceSettings {
     maxAngularAcceleration=1000
     maxAngularVelocity=3
     maxLinearAcceleration=300
     maxLinearVelocity=700
     velocityScaleZ=0
     externalLinearDragV=<0 0.5 0>
    }}
   ziplining {simSpaceSettings {
     maxAngularAcceleration=1000
     maxAngularVelocity=3
     maxLinearAcceleration=300
     maxLinearVelocity=700
     velocityScaleZ=0
     externalLinearDragV=<0 0.5 0>
    }}
  }
 }
 RBAN_On {type=boolean}
 SkelMesh {type=SkeletalMeshComponent}
 Strap Dynamic {type=number}
 Wear_cape {type=boolean}
]
