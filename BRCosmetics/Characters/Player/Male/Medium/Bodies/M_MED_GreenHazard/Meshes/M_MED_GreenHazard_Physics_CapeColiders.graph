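# Physics asset of cape colliders for the GreenHazard medium male player body cosmetic, built entirely from capsules.
# Capsules cover pelvis, lower-to-upper spine, calves and feet, plus dynamic jacket back and side bones, linked by constraints with free angular motion.
# Left and right back-strap chains hang from the first jacket back bones with limited swing and twist.
# Each chain's last joint locks one swing axis and twist, leaving a wider second swing.
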PhysicsAsset /BRCosmetics/Characters/Player/Male/Medium/Bodies/M_MED_GreenHazard/Meshes/M_MED_GreenHazard_Physics_CapeColiders
bodies {
 calf_l [{shapeType=Capsule, length=35, radius=8, shapeName=calf_l_capsule, center=<-22 0 0>, rotation=<rot 90 0 0>}]
 calf_r [{shapeType=Capsule, length=35, radius=8, shapeName=calf_r_capsule, center=<22 0 0>, rotation=<rot 90 0 0>}]
 dyn_jacket_01_back_l [{
   shapeType=Capsule
   length=5.8
   radius=10.2
   shapeName=dyn_jacket_01_back_r_capsule
   center=<-7.5592 2.087 6.7971>
   rotation=<rot 74.9999 89.9983 179.9985>
  }]
 dyn_jacket_01_back_r [{
   shapeType=Capsule
   length=5.8
   radius=10.2
   shapeName=dyn_jacket_01_back_l_capsule
   center=<7.5592 -2.087 -6.7971>
   rotation=<rot 74.9999 89.9983 -0.0015>
  }]
 dyn_jacket_02_back_l [{
   shapeType=Capsule
   length=8.5
   radius=10.2
   shapeName=dyn_jacket_02_back_r_capsule
   center=<-6 1.4847 7.1872>
   rotation=<rot 74.9999 89.9984 179.9985>
  }]
 dyn_jacket_02_back_r [{
   shapeType=Capsule
   length=8.5
   radius=10.2
   shapeName=dyn_jacket_02_back_l_capsule
   center=<6 -1.4847 -7.1872>
   rotation=<rot 74.9999 89.9984 -0.0015>
  }]
 dyn_jacket_03_back_l [{
   shapeType=Capsule
   length=6.5
   radius=11.1
   shapeName=dyn_jacket_03_back_r_capsule
   center=<-6.7662 3.0425 7.1169>
   rotation=<rot 82.3818 89.9966 179.9967>
  }]
 dyn_jacket_03_back_r [{
   shapeType=Capsule
   length=6.5
   radius=11.1
   shapeName=dyn_jacket_03_back_l_capsule
   center=<6.7662 -3.0425 -7.1169>
   rotation=<rot 82.3818 89.9966 -0.0033>
  }]
 dyn_jacket_04_side_l [{shapeType=Capsule, length=5.7, radius=10.5, shapeName=dyn_jacket_03_back_r_capsule, center=<-7.9335 5.0591 3.7184>, rotation=<rot 90 -90 0>}]
 dyn_jacket_04_side_r [{shapeType=Capsule, length=5.7, radius=10.5, shapeName=dyn_jacket_03_back_l_capsule, center=<7.9335 -5.0591 -3.7184>, rotation=<rot 90 90 0>}]
 dyn_strap_back_01_l [{shapeType=Capsule, length=6.2, radius=5.7, shapeName=dyn_strap_back_01_l_capsule, center=<-11.3972 0 2>, rotation=<rot 90 0 0>}]
 dyn_strap_back_01_r [{shapeType=Capsule, length=4.7, radius=6.5, shapeName=dyn_strap_back_01_l_capsule, center=<11.4624 0 -2>, rotation=<rot 90 0 0>}]
 dyn_strap_back_02_l [{shapeType=Capsule, length=6.3, radius=6.5, shapeName=dyn_strap_back_02_l_capsule, center=<-5 0 2>, rotation=<rot 90 0 0>}]
 dyn_strap_back_02_r [{shapeType=Capsule, length=3.9, radius=6.5, shapeName=dyn_strap_back_02_l_capsule, center=<5 0 -2>, rotation=<rot 90 0 0>}]
 dyn_strap_back_03_l [{shapeType=Capsule, length=4.7, radius=6.5, shapeName=dyn_strap_back_03_l_capsule, center=<-4 0 2>, rotation=<rot 90 0 0>}]
 dyn_strap_back_03_r [{shapeType=Capsule, length=5.1, radius=6.5, shapeName=dyn_strap_back_03_l_capsule, center=<4 0 -2>, rotation=<rot 90 0 0>}]
 dyn_strap_back_04_l [{shapeType=Capsule, length=4.8, radius=6.5, shapeName=dyn_strap_back_04_l_capsule, center=<-4 0 2>, rotation=<rot 90 0 0>}]
 dyn_strap_back_04_r [{shapeType=Capsule, length=4.3, radius=6.5, shapeName=dyn_strap_back_04_l_capsule, center=<4 0 -2>, rotation=<rot 90 0 0>}]
 dyn_strap_back_05_l [{shapeType=Capsule, length=2, radius=6.5, shapeName=dyn_strap_back_05_l_capsule, center=<-4 0 2>, rotation=<rot 90 0 0>}]
 dyn_strap_back_05_r [{shapeType=Capsule, length=0, radius=6.8, shapeName=dyn_strap_back_05_r_capsule, center=<2.4261 -0.137 -2>, rotation=<rot 0 -2.4394 -89.5829>}]
 dyn_strap_back_06_l [{shapeType=Capsule, length=0, radius=6.5, shapeName=dyn_strap_back_06_l_capsule, center=<-2.0334 0.0676 2>, rotation=<rot 0 176.985 -88.5358>}]
 foot_l [{shapeType=Capsule, length=14.4327, radius=7.993, shapeName=foot_l_capsule, center=<-4.6749 6.3802 -0.1399>, rotation=<rot 0 2.0678 -89.3216>}]
 foot_r [{shapeType=Capsule, length=14.4327, radius=7.993, shapeName=foot_r_capsule, center=<4.6749 -6.3802 0.1399>, rotation=<rot 0 2.0678 90.6784>}]
 pelvis [{shapeType=Capsule, length=17.3, radius=9.7, shapeName=pelvis_capsule, center=<0 1.0086 0>}]
 spine_01 [{shapeType=Capsule, length=10.8, radius=12, shapeName=spine_01_capsule, center=<5 -0.7388 0>}]
 spine_02 [{shapeType=Capsule, length=9.4, radius=12, shapeName=spine_02_capsule, center=<5 -0.4878 0>}]
 spine_03 [{shapeType=Capsule, length=9, radius=12, shapeName=spine_03_capsule, center=<5 -0.5855 0>}]
 spine_04 [{shapeType=Capsule, length=9.9, radius=12, shapeName=spine_04_capsule, center=<5 -0.7756 0>}]
}
constraints [
 {
  bone1Name=dyn_strap_back_06_l
  bone2Name=dyn_strap_back_05_l
  swing1LimitDegrees=25
  swing1Motion=Locked
  swing2LimitDegrees=80
  twistLimitDegrees=5
  twistMotion=Locked
 }
 {bone1Name=dyn_strap_back_05_l, bone2Name=dyn_strap_back_04_l, swing1LimitDegrees=25, swing2LimitDegrees=35, twistLimitDegrees=3}
 {bone1Name=dyn_strap_back_04_l, bone2Name=dyn_strap_back_03_l, swing1LimitDegrees=25, swing2LimitDegrees=35, twistLimitDegrees=3}
 {bone1Name=dyn_strap_back_03_l, bone2Name=dyn_strap_back_02_l, swing1LimitDegrees=25, swing2LimitDegrees=35, twistLimitDegrees=3}
 {bone1Name=dyn_strap_back_02_l, bone2Name=dyn_strap_back_01_l, swing1LimitDegrees=15, swing2LimitDegrees=35, twistLimitDegrees=3}
 {bone1Name=dyn_strap_back_01_l, bone2Name=dyn_jacket_01_back_l, swing1LimitDegrees=10, swing2LimitDegrees=45, twistLimitDegrees=3}
 {bone1Name=dyn_strap_back_01_r, bone2Name=dyn_jacket_01_back_r, swing1LimitDegrees=10, swing2LimitDegrees=45, twistLimitDegrees=3}
 {bone1Name=dyn_strap_back_02_r, bone2Name=dyn_strap_back_01_r, swing1LimitDegrees=15, swing2LimitDegrees=35, twistLimitDegrees=3}
 {bone1Name=dyn_strap_back_03_r, bone2Name=dyn_strap_back_02_r, swing1LimitDegrees=25, swing2LimitDegrees=35, twistLimitDegrees=3}
 {bone1Name=dyn_strap_back_04_r, bone2Name=dyn_strap_back_03_r, swing1LimitDegrees=25, swing2LimitDegrees=35, twistLimitDegrees=3}
 {
  bone1Name=dyn_strap_back_05_r
  bone2Name=dyn_strap_back_04_r
  swing1LimitDegrees=25
  swing1Motion=Locked
  swing2LimitDegrees=80
  twistLimitDegrees=5
  twistMotion=Locked
 }
 {
  bone1Name=calf_l
  bone2Name=pelvis
  swing1LimitDegrees=45
  swing1Motion=Free
  swing2LimitDegrees=45
  swing2Motion=Free
  twistLimitDegrees=45
  twistMotion=Free
 }
 {
  bone1Name=calf_r
  bone2Name=pelvis
  swing1LimitDegrees=45
  swing1Motion=Free
  swing2LimitDegrees=45
  swing2Motion=Free
  twistLimitDegrees=45
  twistMotion=Free
 }
 {
  bone1Name=dyn_jacket_03_back_r
  bone2Name=dyn_jacket_02_back_r
  swing1LimitDegrees=45
  swing1Motion=Free
  swing2LimitDegrees=45
  swing2Motion=Free
  twistLimitDegrees=45
  twistMotion=Free
 }
 {
  bone1Name=dyn_jacket_02_back_r
  bone2Name=dyn_jacket_01_back_r
  swing1LimitDegrees=45
  swing1Motion=Free
  swing2LimitDegrees=45
  swing2Motion=Free
  twistLimitDegrees=45
  twistMotion=Free
 }
 {
  bone1Name=dyn_jacket_01_back_r
  bone2Name=spine_02
  swing1LimitDegrees=45
  swing1Motion=Free
  swing2LimitDegrees=45
  swing2Motion=Free
  twistLimitDegrees=45
  twistMotion=Free
 }
 {
  bone1Name=dyn_jacket_03_back_l
  bone2Name=dyn_jacket_02_back_l
  swing1LimitDegrees=45
  swing1Motion=Free
  swing2LimitDegrees=45
  swing2Motion=Free
  twistLimitDegrees=45
  twistMotion=Free
 }
 {
  bone1Name=dyn_jacket_02_back_l
  bone2Name=dyn_jacket_01_back_l
  swing1LimitDegrees=45
  swing1Motion=Free
  swing2LimitDegrees=45
  swing2Motion=Free
  twistLimitDegrees=45
  twistMotion=Free
 }
 {
  bone1Name=dyn_jacket_01_back_l
  bone2Name=spine_02
  swing1LimitDegrees=45
  swing1Motion=Free
  swing2LimitDegrees=45
  swing2Motion=Free
  twistLimitDegrees=45
  twistMotion=Free
 }
 {
  bone1Name=dyn_jacket_04_side_r
  bone2Name=spine_03
  swing1LimitDegrees=45
  swing1Motion=Free
  swing2LimitDegrees=45
  swing2Motion=Free
  twistLimitDegrees=45
  twistMotion=Free
 }
 {
  bone1Name=dyn_jacket_04_side_l
  bone2Name=spine_03
  swing1LimitDegrees=45
  swing1Motion=Free
  swing2LimitDegrees=45
  swing2Motion=Free
  twistLimitDegrees=45
  twistMotion=Free
 }
 {
  bone1Name=spine_01
  bone2Name=pelvis
  swing1LimitDegrees=45
  swing1Motion=Free
  swing2LimitDegrees=45
  swing2Motion=Free
  twistLimitDegrees=45
  twistMotion=Free
 }
 {
  bone1Name=spine_02
  bone2Name=spine_01
  swing1LimitDegrees=45
  swing1Motion=Free
  swing2LimitDegrees=45
  swing2Motion=Free
  twistLimitDegrees=45
  twistMotion=Free
 }
 {
  bone1Name=spine_03
  bone2Name=spine_02
  swing1LimitDegrees=45
  swing1Motion=Free
  swing2LimitDegrees=45
  swing2Motion=Free
  twistLimitDegrees=45
  twistMotion=Free
 }
 {
  bone1Name=spine_04
  bone2Name=spine_03
  swing1LimitDegrees=45
  swing1Motion=Free
  swing2LimitDegrees=45
  swing2Motion=Free
  twistLimitDegrees=45
  twistMotion=Free
 }
 {
  bone1Name=foot_r
  bone2Name=calf_r
  swing1LimitDegrees=45
  swing1Motion=Free
  swing2LimitDegrees=45
  swing2Motion=Free
  twistLimitDegrees=45
  twistMotion=Free
 }
 {
  bone1Name=foot_l
  bone2Name=calf_l
  swing1LimitDegrees=45
  swing1Motion=Free
  swing2LimitDegrees=45
  swing2Motion=Free
  twistLimitDegrees=45
  twistMotion=Free
 }
]
