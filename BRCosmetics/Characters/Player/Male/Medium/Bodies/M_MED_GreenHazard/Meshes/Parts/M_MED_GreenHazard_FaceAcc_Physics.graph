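# Physics asset for the face accessory mesh of the male medium GreenHazard player body, defining physics for its dynamic hair bones.
# A head capsule anchors twelve small, identically sized hair capsules covering front, top, sides and back.
# Each hair body is constrained directly to the head rather than chained, with tight per-bone swing and twist limits, loosest at the front and left.
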PhysicsAsset /BRCosmetics/Characters/Player/Male/Medium/Bodies/M_MED_GreenHazard/Meshes/Parts/M_MED_GreenHazard_FaceAcc_Physics
bodies {
 dyn_hair_01_back [{shapeType=Capsule, length=1, radius=3, shapeName=dyn_hair_03_mid_capsule, center=<1 0 0>, rotation=<rot 90 0 0>}]
 dyn_hair_01_fr [{shapeType=Capsule, length=1, radius=3, shapeName=dyn_hair_03_mid_capsule, center=<1 0 0>, rotation=<rot 90 0 0>}]
 dyn_hair_01_r [{shapeType=Capsule, length=1, radius=3, shapeName=dyn_hair_03_mid_capsule, center=<1 0 0>, rotation=<rot 90 0 0>}]
 dyn_hair_01_top [{shapeType=Capsule, length=1, radius=3, shapeName=dyn_hair_03_mid_capsule, center=<1 0 0>, rotation=<rot 90 0 0>}]
 dyn_hair_02_fr [{shapeType=Capsule, length=1, radius=3, shapeName=dyn_hair_03_mid_capsule, center=<1 0 0>, rotation=<rot 90 0 0>}]
 dyn_hair_02_r [{shapeType=Capsule, length=1, radius=3, shapeName=dyn_hair_03_mid_capsule, center=<1 0 0>, rotation=<rot 90 0 0>}]
 dyn_hair_02_top [{shapeType=Capsule, length=1, radius=3, shapeName=dyn_hair_03_mid_capsule, center=<1 0 0>, rotation=<rot 90 0 0>}]
 dyn_hair_03_fr [{shapeType=Capsule, length=1, radius=3, shapeName=dyn_hair_03_mid_capsule, center=<1 0 0>, rotation=<rot 90 0 0>}]
 dyn_hair_04_l [{shapeType=Capsule, length=1, radius=3, shapeName=dyn_hair_03_mid_capsule, center=<1 0 0>, rotation=<rot 90 0 0>}]
 dyn_hair_dw_back_l [{shapeType=Capsule, length=1, radius=3, shapeName=dyn_hair_03_mid_capsule, center=<1 0 0>}]
 dyn_hair_dw_back_mid [{shapeType=Capsule, length=1, radius=3, shapeName=dyn_hair_03_mid_capsule, center=<1 0 0>}]
 dyn_hair_dw_back_r [{shapeType=Capsule, length=1, radius=3, shapeName=dyn_hair_03_mid_capsule, center=<1 0 0>}]
 head [{shapeType=Capsule, length=2.7853, radius=6.3826, center=<6.5161 2.5137 0>, rotation=<rot 90 -28.1197 0>}]
}
constraints [
 {bone1Name=dyn_hair_dw_back_r, bone2Name=head, swing1LimitDegrees=10, swing2LimitDegrees=10, twistLimitDegrees=10}
 {bone1Name=dyn_hair_dw_back_mid, bone2Name=head, swing1LimitDegrees=10, swing2LimitDegrees=10, twistLimitDegrees=10}
 {bone1Name=dyn_hair_dw_back_l, bone2Name=head, swing1LimitDegrees=10, swing2LimitDegrees=10, twistLimitDegrees=10}
 {bone1Name=dyn_hair_03_fr, bone2Name=head, swing1LimitDegrees=15, swing2LimitDegrees=15, twistLimitDegrees=15}
 {bone1Name=dyn_hair_02_r, bone2Name=head, swing1LimitDegrees=7, swing2LimitDegrees=7, twistLimitDegrees=7}
 {bone1Name=dyn_hair_01_fr, bone2Name=head, swing1LimitDegrees=15, swing2LimitDegrees=15, twistLimitDegrees=15}
 {bone1Name=dyn_hair_04_l, bone2Name=head, swing1LimitDegrees=15, swing2LimitDegrees=15, twistLimitDegrees=15}
 {bone1Name=dyn_hair_01_back, bone2Name=head, swing1LimitDegrees=8, swing2LimitDegrees=6, twistLimitDegrees=5}
 {bone1Name=dyn_hair_02_fr, bone2Name=head, swing1LimitDegrees=15, swing2LimitDegrees=15, twistLimitDegrees=15}
 {bone1Name=dyn_hair_01_top, bone2Name=head, swing1LimitDegrees=10, swing2LimitDegrees=10, twistLimitDegrees=6}
 {bone1Name=dyn_hair_02_top, bone2Name=head, swing1LimitDegrees=8, swing2LimitDegrees=10, twistLimitDegrees=5}
 {bone1Name=dyn_hair_01_r, bone2Name=head, swing1LimitDegrees=8, swing2LimitDegrees=8, twistLimitDegrees=6}
]
